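# Translucent UI material for a season pass reward track, drawing a grid of rounded tiles with a progress indication.
# MF_UI_Pattern and SDF boxes build aspect-corrected tiles, cropped by a box mask sized from column and row counts with odd/even offset correction.
# OverlayProgress dims tile alpha along U and positions a vertical marker line.
# The FilledProgress static switch replaces that marker with RGB tinting shaped by top indent gradients.
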Material /SeasonPass/MSEE/Materials/M_RewardTrack_Base
graph {
 properties {
  bUsedWithStaticMesh=false
  blendMode=BLEND_Translucent
  materialDomain=MD_UI
  pixelDepthOffsetMode=PDOM_Legacy
  usedShadingModels=Default Lit
 }
 nodes [
  Constant2Vector() {g=0.05, r=0.5}
  Constant2Vector() {g=0.95, r=0.5}
  n1 = AppendVector(A=ScalarParameter() {parameterName=DotX_Offset}, B=ScalarParameter() {parameterName=DotY_Offset})
  n2 = NamedRerouteUsage()
  n3 = ScalarParameter() {defaultValue=8, parameterName=DotColumns}
  n4 = Round(None=@n3)
  n5 = LinearInterpolate(B=Multiply(A=Divide(B=@n2)) {constB=0.5}, Alpha=Fmod(A=@n4, B=Constant() {r=2}))
  n6 = NamedRerouteUsage()
  n7 = Round(None=ScalarParameter() {defaultValue=2, parameterName=DotRows})
  n8 = LinearInterpolate(B=Multiply(A=Divide(B=@n6)) {constB=0.5}, Alpha=Fmod(A=@n7, B=Constant() {r=2})) {constB=0}
  NamedRerouteDeclaration(None=Add(A=@n1, B=AppendVector(A=@n5, B=@n8)) {constB=0}) {name=PatternUV_Offset, nodeColor=<rgba 0 0 0 1>}
  n9 = ScalarParameter() {defaultValue=0.1, parameterName=TileEdgeFalloff}
  n10 = Constant() {r=1}
  n11 = AppendVector(A=@n10, B=Multiply(A=@n10) {constB=1.25})
  n12 = TextureCoordinate() {coordinateIndex=3}
  n13 = AppendVector(A=Constant() {r=1}, B=Divide(A=ComponentMask(None=@n12) {r=true}, B=ComponentMask(None=@n12) {g=true}))
  n14 = NamedRerouteDeclaration(None=MaterialFunctionCall(UVs=TextureCoordinate(), Texture Scale=Multiply(A=@n11, B=@n13)) {materialFunction=&ScaleUVsByCenter}.UVs) {name=ScaledUV_1, nodeColor=<rgba 0 0 0 1>}
  n15 = NamedRerouteUsage()
  n16 = MaterialFunctionCall(
   UVs=MaterialFunctionCall!(
    UV=Add(A=@n14, B=@n15) {constB=0}
    Tiling X=NamedRerouteDeclaration(None=@n3) {name=PatternX_1, nodeColor=<rgba 0 0 0 1>}
    Tiling Y=NamedRerouteDeclaration(None=@n3) {name=PatternY_1, nodeColor=<rgba 0 0 0 1>}
   ) {materialFunction=/FortUILibrary/MaterialFunctions/Basics/MF_UI_Pattern}.Pattern UVs
   Size=ScalarParameter() {defaultValue=0.85, parameterName=SpaceBetweenTiles}
   CornerRoundness=ScalarParameter() {defaultValue=0.025, parameterName=TileCornerRoundness}
  ) {materialFunction=&MF_UI_Box}
  n17 = NamedRerouteUsage()
  n18 = Divide(A=@n4, B=@n2)
  n19 = Divide(A=@n7, B=@n6)
  n20 = Multiply(
   A=SmoothStep(Min=OneMinus(None=@n9), Value=OneMinus(None=@n16.SDF)) {constMin=0.9}
   B=MaterialFunctionCall!(A=Add(A=@n17, B=@n1) {constB=0}, B=Constant() {r=0.5}, Bounds=AppendVector(A=@n18, B=@n19), Edge Falloff=Constant() {r=0.001}) {materialFunction=/Engine/Functions/Engine_MaterialFunctions01/Gradient/BoxMask-2D}.Result
  )
  n21 = VectorParameter() {parameterName=RGB, defaultValue=<rgba 0.6654 0.1746 0.0009 1>}
  n22 = ScalarParameter() {defaultValue=0.44, group=Top, parameterName=IndentWidth_Top, sliderMax=0.51, sliderMin=0.175, sortPriority=0}
  n23 = Subtract(A=@n22, B=ScalarParameter() {defaultValue=0.02, group=Top, parameterName=IndentTightness_Top, sliderMax=1, sliderMin=0.01, sortPriority=1})
  n24 = Constant() {r=-0.06}
  n25 = OneMinus(None=@n22)
  n26 = OneMinus(None=@n23)
  n27 = Ceil(None=Add(
    A=MaterialFunctionCall(GradientStart=@n22, GradientEnd=@n23, GradientOffset=@n24) {materialFunction=&MF_UI_LinearGradient}.Result
    B=MaterialFunctionCall(GradientStart=@n25, GradientEnd=@n26, GradientOffset=@n24) {materialFunction=&MF_UI_LinearGradient}.Result
   ))
  n28 = TextureCoordinate() {coordinateIndex=3}
  n29 = ComponentMask(None=@n28) {g=true}
  n30 = ComponentMask(None=@n28) {r=true}
  n31 = MaterialFunctionCall(UVs=TextureCoordinate(), Texture Scale=AppendVector(A=Divide(A=@n29, B=@n30), B=Constant() {r=1})) {materialFunction=&ScaleUVsByCenter}
  n32 = ScalarParameter() {defaultValue=0.725, group=Shape, parameterName=OverlayProgress, sliderMax=1, sortPriority=2}
  n33 = Constant() {r=0.05}
  n34 = MaterialFunctionCall!(
   Coordinates=@n31.UVs
   P0=AppendVector(A=@n32, B=@n33)
   P1=AppendVector(A=@n32, B=OneMinus(None=@n33))
   Line Width=Constant() {r=0.035}
   Edge Falloff=Constant() {r=0.25}
   Square Corners?=StaticBool()
  ) {materialFunction=/Engine/Functions/Engine_MaterialFunctions01/Coordinates/DrawLine-2D}
  n35 = Ceil(None=@n34.Result)
  StaticSwitchParameter(
   True=LinearInterpolate(A=@n20, B=@n21.RGB, Alpha=OneMinus(None=@n27)) {constA=1}
   False=LinearInterpolate(A=@n20, B=LinearInterpolate(B=@n21.RGB, Alpha=@n35), Alpha=@n34.Result) {constA=1}
  ) {parameterName=FilledProgress}
  n36 = ScalarParameter() {defaultValue=0.725, group=Shape, parameterName=OverlayProgress, sliderMax=1, sortPriority=2}
  n37 = MaterialFunctionCall(UV Channel=@n31.UVs) {materialFunction=&LinearGradient}
  n38 = Multiply(A=@n20, B=LinearInterpolate(Alpha=Step(Y=@n36, X=@n37.UGradient)) {constA=1, constB=0.5})
  StaticSwitchParameter(True=@n38, False=Max(A=@n34.Result, B=@n38)) {parameterName=FilledProgress}
 ]
}
