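# Season Pass reward card widget, built on the UIKit item card base, showing item art, locked/owned state, progress bars and free or theme-colored backgrounds.
# Focus sets the selection viewmodel's selected reward, which drives the card's selected look.
# Lock, ownership, quest and purchasable-count checks gate hold-to-purchase.
# Hold progress is written to that shared viewmodel; helpers pick claim-all or single-reward progress and feed thresholded values into dynamic materials.
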WidgetBlueprint /SeasonPass/MSEE/WBP_MSEE_Reward_Itemcard
animations [Focus {
  binding_count=1
  length_seconds=0.05
  object=&self:Focus
  sequence {bindings [SizeBox {
     id=#1
     possessed_class=SizeBox
     tracks [{
       type=MovieSceneFloatTrack
       display_name=Height Override
       sections [{
         type=MovieSceneFloatSection
         end_frame=1
         start_frame=0
         channels [None {type=MovieSceneScriptingFloatChannel, keys [[0,160],[0.9997,200],[1,180]]}]
        }]
      }]
    }]}
 }]
graphs [
 CheckIfInClaimAll {nodes [
   CheckIfInClaimAll!() -> (then, IsClaimAll:bool=~) {is=title K2Node_FunctionEntry}
   > Set InClaimAllArray(InClaimAllArray:bool=false) -> (Output_Get:bool)
   n1 = Get SeasonPassSelectionVM() -> (SeasonPassSelectionVM:Season Pass Selection VM Object Reference)
   > n2 = For Each Loop with Break!(
    Exec:Exec
    Array:Array of Season Pass Reward VM Object References=GetPurchaseableRewards!(self:Season Pass Element Base VM Object Reference=Get SelectedPage!(self:Season Pass Selection VM Object Reference=@n1) -> (SelectedPage:Season Pass Page VM Object Reference)) -> (ReturnValue:Array of Season Pass Reward VM Object References)
    Break:Exec=@n4.then
   ) -> (LoopBody:Exec, Array Element:Season Pass Reward VM Object Reference, Array Index:int, Completed:Exec)
   > n3 = Set InClaimAllArray(InClaimAllArray="Equal (Object)"(A=@n2, B=Get SeasonPassRewardVM() -> (SeasonPassRewardVM:Season Pass Reward VM Object Reference)))
   > n4 = Branch(Condition=@n3)
  ]}
 EventGraph {nodes [
   Event On Focused() {is=title K2Node_Event}
   > Set SelectedReward!(
    execute
    SelectedReward:Season Pass Reward VM Object Reference=Get SeasonPassRewardVM()
    self:Season Pass Selection VM Object Reference=Get SeasonPassSelectionVM()
   ) -> (then, Output_Get:Season Pass Reward VM Object Reference)
   Event On Clicked()
   > PrintString!(execute, InString=----, bPrintToScreen, bPrintToLog, TextColor, Duration, Key) -> (then)
   > Parent: On Clicked() {is=title K2Node_CallParentFunction}
   > Call Item Clicked!(execute:Exec, self:WBP MSEE Reward Itemcard Object Reference, RewardVM:Season Pass Reward VM Object Reference=Get SeasonPassRewardVM()) -> (then:Exec) {is=title K2Node_CallDelegate}
   Event Construct()
   n1 = Get CommonInputSubsystem()
   > HandleInputMethodChanged!(execute, self, NewInputMethod:ECommonInputType Enum=GetCurrentInputType(self=@n1)) -> (then)
   n2 = Create Event()
   > Bind Event to On Input Method Changed(self=@n1, Delegate=@n2)
   > SetFortBrushSize!(
    execute
    self:Fort Lazy Image Object Reference=Get LazyImage_Item() -> (LazyImage_Item:Fort Lazy Image Object Reference)
    FortBrushSize:EFortBrushSize Enum=Large
   ) -> (then)
   n3 = Event PreConstruct()
   > Parent: PreConstruct(IsDesignTime=@n3.IsDesignTime)
   > SetAsyncBackground!(
    execute
    self:WBP UIKit Item Card Base Object Reference
    MaterialInterfaceValue:Material Interface Soft Object Reference=/SeasonPass/Widgets/Reboot/Materials/M_UI_BPReboot_RewardTrack_TileBackground
   ) -> (then)
   > Branch(Condition=@n3.IsDesignTime)
   > UpdateProgressBarColors!(execute, self, ShouldUseExpMaterial:bool=Get Preview_UseAlterProgressBar() -> (Preview_UseAlterProgressBar:bool)) -> (then)
   > SetItemIsEquipped!(execute, self:WBP UIKit Item Card Base Object Reference, Is Equipped:bool=false) -> (then)
   Event On Unhovered()
   n4 = Get SelectedReward!(self:Season Pass Selection VM Object Reference=Get SeasonPassSelectionVM()) -> (SelectedReward:Season Pass Reward VM Object Reference)
   n5 = Get SeasonPassRewardVM()
   > Branch(Condition=AND Boolean(
     A=Get IsButtonFocused() -> (IsButtonFocused:bool)
     B="Not Equal (Object)"(A:Object Reference=@n4, B:Object Reference=@n5) -> (ReturnValue:bool=~) {is=title K2Node_PromotableOperator}
    ))
   > Parent: On Unhovered() {is=title K2Node_CallParentFunction}
   Event Destruct()
   > Parent: Destruct()
   > Unbind Event from On Input Method Changed(self=@n1, Delegate=@n2)
   n6 = Event OnActionProgress!() -> (OutputDelegate, then, HeldPercent:float) {is=title K2Node_Event}
   > Parent: OnActionProgress!(execute, HeldPercent:float=@n6.HeldPercent) -> (then) {is=title K2Node_CallParentFunction}
   n7 = Get SeasonPassSelectionVM()
   > Branch(Condition="Equal (Object)"(A=Get SelectedReward(self=@n7), B=Get SeasonPassRewardVM()))
   > Set OfferPurchaseHoldProgress!(execute, OfferPurchaseHoldProgress:float=@n6.HeldPercent, self:Season Pass Selection VM Object Reference=Get SeasonPassSelectionVM()) -> (then, Output_Get:float)
   Event OnInitialized()
   > Parent: OnInitialized() {is=title K2Node_CallParentFunction}
   > n8 = GetDynamicMaterial(self=Get Img_RewardEffect() -> (Img_RewardEffect:Image Object Reference))
   > Set DynamicRewardMaterial(DynamicRewardMaterial:Material Instance Dynamic Object Reference=@n8) -> (Output_Get:Material Instance Dynamic Object Reference)
   Event On Hovered()
   > Parent: On Hovered() {is=title K2Node_CallParentFunction}
   n9 = Branch(Condition=Get IsModeSpecificReward() -> (IsModeSpecificReward:bool))
   n10 = Sequence(execute=@n9.else)
   > Branch(Condition=IsInputType!(InputType:ECommonInputType Enum=Touch, IsNot:bool=false) -> (Result:bool))
   > Set SelectedReward(SelectedReward=Get SeasonPassRewardVM(), self=Get SeasonPassSelectionVM())
   n11 = Get SeasonPassRewardVM()
   n12 = Get SelectedReward(self=Get SeasonPassSelectionVM())
   Branch(
    execute=@n10.then_1
    Condition=AND Boolean(A=GetRequiresHold!(self:Common Button Base Object Reference) -> (ReturnValue:bool), B="Equal (Object)"(A=@n12, B=@n11))
   )
   > PurchaseItem!(
    execute:Exec
    Item:Season Pass Reward VM Object Reference=Get SeasonPassRewardVM()
    SeasonPass:Season Pass VM Object Reference=Get SeasonPassVM() -> (SeasonPassVM:Season Pass VM Object Reference)
    bTryPurchasePrerequisites:bool=false
   ) -> (then:Exec, OnOperationCompleted:Exec, Success:bool=~) {is=title K2Node_AsyncAction}
   n13 = Get!(execute:Exec) -> (then:Exec, else:Exec, SeasonPassRewardVM:Season Pass Reward VM Object Reference)
   > n14 = Get!(execute) -> (then, else, SeasonPassSelectionVM:Season Pass Selection VM Object Reference)
   > Get!(execute, self:Season Pass Selection VM Object Reference=@n14) -> (then, else, SelectedPage:Season Pass Page VM Object Reference)
   n15 = Get SeasonPassVM()
   n16 = Get SelectedPage(self=Get SeasonPassSelectionVM())
   > SetRequiresHold!(
    execute
    self:Common Button Base Object Reference
    bInRequiresHold:bool=GetCanBeHeld!(
     self
     IsRewardLocked:bool=Get bIsLocked!(self:Season Pass Element Base VM Object Reference=@n13) -> (bIsLocked:bool)
     IsRewardOwned:bool=Get bIsRewardOwned!(self:Season Pass Reward VM Object Reference=@n13) -> (bIsRewardOwned:bool)
     IsPageLocked:bool=Get bIsLocked(self=@n16)
     IsQuestDriven:bool=Get bIsQuestDriven!(self:Season Pass Reward VM Object Reference=@n13) -> (bIsQuestDriven:bool)
     PurchasableRewards:int=GetNumPurchaseableRewards!(self:Season Pass Element Base VM Object Reference=@n15) -> (ReturnValue:int=0)
    ) -> (CanBeHeld:bool)
   ) -> (then)
   n17 = Get Focus() -> (Focus:Widget Animation Object Reference)
   PlayAnimationForward(InAnimation=@n17)
   PlayAnimationReverse(InAnimation=@n17)
   HandleInputMethodChanged!() -> (OutputDelegate, then, NewInputMethod:ECommonInputType Enum) {is=title K2Node_CustomEvent}
  ]}
 Get Item VM {nodes [
   n1 = Get Item VM!() -> (then, Reward:Season Pass Reward VM Object Reference) {is=title K2Node_FunctionEntry}
   > Is Valid(InputObject=@n1)
   > Return Node!(
    execute
    Result:Fort Item VM Object Reference="Get (a copy)"(Array:Array of Fort Item VM Object References=Get ItemVMs!(self:Season Pass Reward VM Object Reference=@n1) -> (ItemVMs:Array of Fort Item VM Object References)) -> (Output:Fort Item VM Object Reference)
   )
  ]}
 GetCanBeHeld {nodes [
   n1 = GetCanBeHeld!() -> (then, IsRewardLocked:bool=~, IsRewardOwned:bool=~, IsPageLocked:bool=~, IsQuestDriven:bool=~, PurchasableRewards:int) {is=title K2Node_FunctionEntry}
   n2 = NOT Boolean(A=OR Boolean!(A=@n1, B=@n1.IsRewardOwned, C:bool=@n1.IsPageLocked, D:bool=@n1.IsQuestDriven, E:bool="Equal (Integer)"(A=@n1.PurchasableRewards)) -> (ReturnValue))
   > Set Visibility(Visibility=Select(Option 0=Hidden, Option 1=SelfHitTestInvisible, Index=@n2), self=Get Img_RewardEffect())
   > Return Node!(execute, CanBeHeld:bool=@n2)
  ]}
 GetProgressVisibility {nodes [
   GetProgressVisibility() {is=title K2Node_FunctionEntry}
   > Return Node!(execute, NewParam:ESlateVisibility Enum=Visible)
  ]}
 GetProgressVisuals {nodes [
   n1 = GetProgressVisuals!() -> (then, Progress:double=~, SelectedReward:Season Pass Reward VM Object Reference, IsClaimAll:bool=~) {is=title K2Node_FunctionEntry}
   n2 = "Equal (Object)"(A=Get SelectedReward!() -> (SelectedReward), B=Get SeasonPassRewardVM())
   n3 = Get Progress() -> (Progress:double)
   n4 = Get Progress()
   n5 = Get InClaimAllArray() -> (InClaimAllArray:bool)
   > Return Node(NewParam:double=SelectFloat(A=SelectFloat(A=@n4, B=0, bPickA=@n5), B=SelectFloat(A=@n3, B=0, bPickA=@n2), bPickA=@n1.IsClaimAll))
  ]}
 HandleProgressVisuals {nodes [
   n1 = Get Background() -> (Background:Image Object Reference)
   n2 = HandleProgressVisuals!() -> (then, Progress:double=~) {is=title K2Node_FunctionEntry}
   n3 = SelectFloat(
    A=MapRangeClamped(Value=@n2, InRangeA=Get HoldProgressPercentThreshold() -> (HoldProgressPercentThreshold:double), InRangeB=1)
    B=0
    bPickA="float > float"(A=@n2, B=Get HoldProgressPercentThreshold())
   )
   n4 = SetScalarParameterValue(
    self=GetDynamicMaterial!(Brush:"Slate Brush Structure (by ref)"=Get Brush!(self:Image Object Reference=@n1) -> (Brush:Slate Brush Structure)) -> (ReturnValue)
    ParameterName=Progress
    Value=@n3
   )
   n5 = GetDynamicMaterial(Brush=Get Brush(self=Get LazyImage_Item()))
   SetScalarParameterValue(execute=[@n4.then, @n2.then], self=@n5, ParameterName=HoldEffect, Value=@n3)
   > SetScalarParameterValue(self=@n5, ParameterName=HoldAnimation, Value=@n3)
   > SetScalarParameterValue(self=Get DynamicRewardMaterial() -> (DynamicRewardMaterial:Material Instance Dynamic Object Reference), ParameterName=HoldRatio, Value=@n3)
  ]}
 ItemClicked {nodes [ItemClicked!() -> (then, RewardVM:Season Pass Reward VM Object Reference) {is=title K2Node_FunctionEntry}]}
 OwnerShipChanged {nodes [
   n1 = OwnerShipChanged!() -> (then, IsOwned:bool=~) {is=title K2Node_FunctionEntry}
   > n2 = Branch(Condition=@n1)
   > Set Status Indicator Visibility!(execute, self:WBP UIKit Item Card Base Object Reference, Indicator:E_UIKit_ItemCardStatusIndicator Enum=NewEnumerator5, Visibility:bool=true) -> (then) {is=title K2Node_CallFunction}
   Set Status Indicator Visibility(execute=@n2.else, Indicator=NewEnumerator5, Visibility=false)
  ]}
 RewardVM_Updated {nodes [
   RewardVM_Updated!() -> (then, RewardVM:Season Pass Reward VM Object Reference) {is=title K2Node_FunctionEntry}
   > VM_SelectedRewardUpdated!(
    execute
    self
    SelectedReward:Season Pass Reward VM Object Reference=Get SelectedReward!(self:Season Pass Selection VM Object Reference=Get SeasonPassSelectionVM()) -> (SelectedReward)
   ) -> (then)
  ]}
 SetItemImage {nodes [
   n1 = SetItemImage!() -> (then, Item Image:Soft Object Reference) {is=title K2Node_FunctionEntry}
   > Parent: SetItemImage!(execute, Item Image:Soft Object Reference=@n1) -> (then) {is=title K2Node_CallParentFunction}
   n2 = SlotAsGridSlot(Widget=Get LazyImage_Item())
   n3 = OR Boolean!(A=NOT Boolean(A=IsAssetNull!(Asset:Soft Object Reference=@n1) -> (ReturnValue:bool)), B=InEditor!() -> (ReturnValue:bool)) -> (ReturnValue)
   > SetHorizontalAlignment(
    self:Grid Slot Object Reference=@n2
    InHorizontalAlignment=Select(Option 0:EHorizontalAlignment Enum=HAlign_Center, Option 1:EHorizontalAlignment Enum=HAlign_Fill, Index=@n3) -> (ReturnValue:EHorizontalAlignment Enum)
   )
   > SetVerticalAlignment(
    self:Grid Slot Object Reference=@n2
    InVerticalAlignment=Select(Option 0:EVerticalAlignment Enum=VAlign_Center, Option 1:EVerticalAlignment Enum=VAlign_Fill, Index=@n3) -> (ReturnValue:EVerticalAlignment Enum)
   )
  ]}
 SetLockedState {nodes [
   SetLockedState!() -> (then, IsLocked:bool=~) {is=title K2Node_FunctionEntry}
   > SetVisibility(
    self=Get StatusIndicator_Locked() -> (StatusIndicator_Locked:WBP UIKit Status Indicator Object Reference)
    InVisibility=Select(Option 0:ESlateVisibility Enum=Collapsed, Option 1:ESlateVisibility Enum=Visible, Index=Get IsLocked() -> (IsLocked:bool)) -> (ReturnValue:ESlateVisibility Enum)
   )
  ]}
 UpdateProgressBarColors {nodes [
   n1 = UpdateProgressBarColors!() -> (then, ShouldUseExpMaterial:bool=~) {is=title K2Node_FunctionEntry}
   > SetActiveWidget(
    self=Get ProgressBar_Switcher() -> (ProgressBar_Switcher:Common Visibility Switcher Object Reference)
    Widget=Select(
     Option 0:Widget Object Reference=Get ProgressBar() -> (ProgressBar:WBP UIKit Progress Bar Object Reference)
     Option 1:Widget Object Reference=Get ProgressBar_Alter() -> (ProgressBar_Alter:WBP UIKit Progress Bar Object Reference)
     Index=@n1
    ) -> (ReturnValue:Widget Object Reference)
   )
  ]}
 UpdateSelectedState {nodes [
   n1 = UpdateSelectedState!() -> (then, ShouldSelect:bool=~) {is=title K2Node_FunctionEntry}
   > n2 = Branch(Condition=@n1)
   > Parent: On Selected() {is=title K2Node_CallParentFunction}
   Branch(execute=@n2.else, Condition=Get IsButtonFocused())
   > Parent: On Deselected() {is=title K2Node_CallParentFunction}
  ]}
 VM_HandleIsFree {nodes [
   n1 = VM_HandleIsFree!() -> (then, IsFree:bool=~) {is=title K2Node_FunctionEntry}
   > n2 = Set Is Free(Is Free:bool=@n1) -> (Output_Get:bool)
   n3 = Get BackgroundMaterial_Free() -> (BackgroundMaterial_Free:Material Instance Object Reference)
   n4 = Get ThemedBackground() -> (ThemedBackground:Material Instance Dynamic Object Reference)
   n5 = IsValid(Object=@n4)
   n6 = Make Soft Reference!(Input:Material Instance Dynamic Object Reference=@n4) -> (Output:Material Instance Dynamic Soft Object Reference)
   n7 = Make Soft Reference(Input:Material Object Reference=Get BackgroundMaterial() -> (BackgroundMaterial:Material Object Reference)) -> (Output:Material Soft Object Reference)
   > SetAsyncBackground(MaterialInterfaceValue=Select(
     Option 0:Material Interface Soft Object Reference=Select(Option 0=@n7, Option 1=@n6, Index=@n5)
     Option 1:Material Interface Soft Object Reference=Make Soft Reference(Input:Material Instance Object Reference=@n3) -> (Output:Material Instance Soft Object Reference)
     Index=@n2
    ) -> (ReturnValue:Material Interface Soft Object Reference))
  ]}
 VM_HandleThemeData {nodes [
   n1 = VM_HandleThemeData!() -> (then, Theme:Season Pass Category Theme Data Structure) {is=title K2Node_FunctionEntry}
   > n2 = Sequence()
   > n3 = Branch(Condition=Get Is Free() -> (Is Free:bool))
   n4 = CreateDynamicMaterialInstance!(execute=@n3.else, Parent=Get BackgroundMaterial(), OptionalName=None, CreationFlags=None) -> (then, ReturnValue)
   > n5 = Set ThemedBackground(ThemedBackground:Material Instance Dynamic Object Reference=@n4) -> (Output_Get:Material Instance Dynamic Object Reference)
   n6 = TryLookupColor!(
    Table:"Fort Color Reference Table Structure (by ref)"=Break Season Pass Category Theme Data!(SeasonPassCategoryThemeData:"Season Pass Category Theme Data Structure (by ref)"=@n1) -> (
     ImageReferenceTable:Fort Image Reference Table Structure
     ColorReferenceTable:Fort Color Reference Table Structure
     TextReferenceTable:Fort Text Reference Table Structure
    ) {is=title K2Node_BreakStruct}.ColorReferenceTable
    Name:String=CharacterColor
   ) -> (OutColor:Linear Color Structure, ReturnValue:bool)
   > SetVectorParameterValue(
    self=@n5
    ParameterName=CharacterColor
    Value=Select(Option 0:Linear Color Structure=(R=0.0802,G=1,B=0,A=1), Option 1:Linear Color Structure=@n6, Index=@n6.ReturnValue) -> (ReturnValue:Linear Color Structure)
   )
   > SetAsyncBackground(MaterialInterfaceValue=Make Soft Reference(Input:Material Instance Dynamic Object Reference=@n5) -> (Output:Material Instance Dynamic Soft Object Reference))
   n7 = TryLookupColor(Table=Break Season Pass Category Theme Data(SeasonPassCategoryThemeData=@n1).ColorReferenceTable, Name=ProgressionColor)
   SetVectorParameterValue(
    execute=@n2.then_1
    self=Get ProgressBarMID!(self:WBP UIKit Progress Bar Base Object Reference=Get ProgressBar()) -> (ProgressBarMID:Material Instance Dynamic Object Reference)
    ParameterName=Primary Color
    Value=Select(Option 0=(R=0.0802,G=1,B=0,A=1), Option 1=@n7, Index=@n7.ReturnValue)
   )
  ]}
 VM_SelectedRewardUpdated {nodes [
   n1 = VM_SelectedRewardUpdated!() -> (then, SelectedReward:Season Pass Reward VM Object Reference) {is=title K2Node_FunctionEntry}
   > UpdateSelectedState!(execute, self, ShouldSelect:bool="Equal (Object)"(A=@n1, B=Get SeasonPassRewardVM())) -> (then) {is=title K2Node_CallFunction}
  ]}
 VM_SetBarsProgress {nodes [
   n1 = VM_SetBarsProgress!() -> (then, Progress:double=~) {is=title K2Node_FunctionEntry}
   > SetProgress!(execute, self:WBP UIKit Progress Bar Base Object Reference=[Get ProgressBar(), Get ProgressBar_Alter()], Progress:double=@n1) -> (then)
  ]}
 ViewmodelsUpdated {nodes [
   n1 = ViewmodelsUpdated!() -> (then, MyRewardVM:Season Pass Selection VM Object Reference, SelectedReward:Season Pass Selection VM Object Reference) {is=title K2Node_FunctionEntry}
   > Return Node!(execute, AreTheSameReward:bool="Equal (Object)"(A=@n1, B=@n1.SelectedReward))
  ]}
]
variables [
 BackgroundMaterial {type=Material, default=&M_UI_BPReboot_RewardTrack_TileBackground, class=~, graph=~}
 BackgroundMaterial_Free {type=MaterialInstance, default=/SeasonPass/Widgets/Reboot/Materials/MI_UI_BPReboot_RewardTrack_TileBackground_Free}
 CardIndex {type=integer}
 DynamicRewardMaterial {type=MaterialInstanceDynamic}
 HideEnhanced {type=boolean, default=true}
 HoldProgressPercentThreshold {type=number, default=0.05}
 InClaimAllArray {type=boolean}
 Is Free {type=boolean}
 IsModeSpecificReward {type=boolean}
 ItemClicked {}
 Preview_UseAlterProgressBar {type=boolean, default=true}
 ThemedBackground {type=MaterialInstanceDynamic}
 WidgetSeasonPassCategory {type=SeasonPassCategory_C, default=null}
]
widgets {
 info {
  inheritedWidgetCount=0
  namedSlotCount=2
  parentClass=/FortUILibrary/Systems/Items/WBP_UIKit_ItemCard_Base.WBP_UIKit_ItemCard_Base_C
  rootWidgetClass=null
  widgetCount=13
 }
 widgets [
  Status {widgetClassPath=&Overlay, widget=Status}
  StatusIndicator_Owned {
   widgetClassPath=/FortUILibrary/UIKit/StatusIndicator/WBP_UIKit_StatusIndicator
   bIsVariable=true
   parent=Status
   widget=StatusIndicator_Owned
   properties {icon=/FortUILibrary/Textures/MaterialInstances/MI_UIKit_Icon_Check, iconSize=Compact, visibility=Collapsed}
   slot {class=OverlaySlot, properties {padding=<ltrb 8 8 8 8>}}
  }
  StatusIndicator_Locked {
   widgetClassPath=&WBP_UIKit_StatusIndicator
   bIsVariable=true
   parent=Status
   widget=StatusIndicator_Locked
   properties {type=Utility, icon=/FortUILibrary/Textures/MaterialInstances/MI_UIKit_Icon_Locked, iconSize=Compact, visibility=Collapsed}
   slot {class=OverlaySlot, properties {padding=<ltrb 8 8 8 8>}}
  }
  VerticalBox_0 {
   widgetClassPath=&VerticalBox
   parent=Status
   widget=VerticalBox_0
   slot {class=OverlaySlot, properties {horizontalAlignment=HAlign_Fill, verticalAlignment=VAlign_Bottom, padding=<ltrb 8 8 8 8>}}
  }
  StatusIndicator_Included {
   widgetClassPath=&CommonTextBlock
   bIsVariable=true
   parent=VerticalBox_0
   widget=StatusIndicator_Included
   properties {
    text=INCLUDED
    visibility=Collapsed
    font {
     bForceMonospaced=false
     bMaterialIsStencil=false
     fontMaterial=null
     fontObject=&HeadingNow
     letterSpacing=0
     monospacedWidth=1
     size=12
     skewAmount=0
     typefaceFontName=75 Medium
     outlineSettings {
      bApplyOutlineToDropShadows=false
      bMiteredCorners=false
      bSeparateFillAlpha=false
      outlineMaterial=null
      outlineSize=0
      outlineColor=<rgba 0 0 0 1>
     }
    }
    shadowOffset=<0 0>
   }
   slot {class=VerticalBoxSlot, properties {horizontalAlignment=HAlign_Left, verticalAlignment=VAlign_Bottom}}
  }
  ProgressBar_Switcher {
   widgetClassPath=&CommonVisibilitySwitcher
   bIsVariable=true
   parent=VerticalBox_0
   widget=ProgressBar_Switcher
   properties {visibility=Collapsed}
   slot {class=VerticalBoxSlot, properties {padding=<ltrb 0 8 0 0>}}
  }
  ProgressBar {
   widgetClassPath=/FortUILibrary/UIKit/ProgressBar/WBP_UIKit_ProgressBar
   bIsVariable=true
   parent=ProgressBar_Switcher
   widget=ProgressBar
   properties {
    height=16
    linearMaterial=/SeasonPass/Materials/ProgressBar/MI_UI_NexusProgerssBar
    outline_PX=6
    progressValue=0.5488
    useCircular=false
    useOutline=true
   }
   slot {class=CommonVisibilitySwitcherSlot}
  }
  ProgressBar_Alter {
   widgetClassPath=&WBP_UIKit_ProgressBar
   bIsVariable=true
   parent=ProgressBar_Switcher
   widget=ProgressBar_Alter
   properties {
    height=16
    linearMaterial=/SeasonPass/Materials/ProgressBar/MI_UI_NexusProgerssBar_Exp
    outline_PX=6
    progressValue=0.5488
    useCircular=false
    useOutline=true
   }
   slot {class=CommonVisibilitySwitcherSlot}
  }
  CommonVisualAttachment_71 {
   widgetClassPath=&CommonVisualAttachment
   parent=Status
   widget=CommonVisualAttachment_71
   properties {contentAnchor=<0.5 0>}
   slot {class=OverlaySlot, properties {horizontalAlignment=HAlign_Center, verticalAlignment=VAlign_Bottom, padding=<ltrb 0 0 0 15>}}
  }
  Tag_FreeItem {
   widgetClassPath=&Border
   bIsVariable=true
   parent=CommonVisualAttachment_71
   widget=Tag_FreeItem
   properties {
    visibility=Collapsed
    background {
     bIsDynamicallyLoaded=false
     drawAs=Image
     imageType=FullColor
     mirroring=NoMirror
     resourceName=None
     resourceObject=/FortUILibrary/Systems/Badge/Materials/M_UIKit_Badge_Background
     tiling=NoTile
     imageSize=<32 32>
     margin=<ltrb 0 0 0 0>
     outlineSettings {
      bUseBrushTransparency=false
      roundingType=HalfHeightRadius
      width=0
      color {colorUseRule=UseColor_Specified, specifiedColor=<rgba 0 0 0 0>}
      cornerRadii=<0 0 0 0>
     }
     tintColor {colorUseRule=UseColor_Specified, specifiedColor=<rgba 0.0048 0.004 0.0116 1>}
     uVRegion {bIsValid=false, max=<0 0>, min=<0 0>}
    }
    padding=<ltrb 8 8 8 8>
   }
   slot {class=SizeBoxSlot, properties {horizontalAlignment=HAlign_Center}}
  }
  FreeText {
   widgetClassPath=&CommonTextBlock
   parent=Tag_FreeItem
   widget=FreeText
   properties {
    justification=Center
    mobileFontSizeMultiplier=1.14
    scrollStyle=&TextScrollStyle-Base_C
    text=FREE
    textTransformPolicy=ToUpper
    font {
     bForceMonospaced=false
     bMaterialIsStencil=false
     fontMaterial=null
     fontObject=&HeadingNow
     letterSpacing=0
     monospacedWidth=1
     size=12
     skewAmount=0
     typefaceFontName=86 Bold Italic
     outlineSettings {
      bApplyOutlineToDropShadows=false
      bMiteredCorners=false
      bSeparateFillAlpha=false
      outlineMaterial=null
      outlineSize=0
      outlineColor=<rgba 0 0 0 1>
     }
    }
    shadowOffset=<0 0>
   }
   slot {class=BorderSlot, properties {padding=<ltrb 8 8 8 8>}}
  }
  Overlay_134 {widgetClassPath=&Overlay, widget=Overlay_134}
  Img_RewardEffect {
   widgetClassPath=&Image
   bIsVariable=true
   parent=Overlay_134
   widget=Img_RewardEffect
   properties {
    visibility=Hidden
    brush {
     bIsDynamicallyLoaded=false
     drawAs=Image
     imageType=FullColor
     mirroring=NoMirror
     resourceName=None
     resourceObject=/SeasonPass/Materials/MI_UI_Pass_RewardState
     tiling=NoTile
     imageSize=<32 32>
     margin=<ltrb 0 0 0 0>
     outlineSettings {
      bUseBrushTransparency=false
      roundingType=HalfHeightRadius
      width=0
      color {colorUseRule=UseColor_Specified, specifiedColor=<rgba 0 0 0 0>}
      cornerRadii=<0 0 0 0>
     }
     tintColor {colorUseRule=UseColor_Specified, specifiedColor=<rgba 1 1 1 1>}
     uVRegion {bIsValid=false, max=<0 0>, min=<0 0>}
    }
   }
   slot {class=OverlaySlot, properties {horizontalAlignment=HAlign_Fill, verticalAlignment=VAlign_Fill}}
  }
 ]
}
