Material /FortUILibrary/Systems/Items/Materials/M_UIKit_Item_Background
graph {
 properties {
  bUsedWithStaticMesh=false
  blendMode=BLEND_Translucent
  materialDomain=MD_UI
  pixelDepthOffsetMode=PDOM_Legacy
  usedShadingModels=Default Lit
 }
 nodes [
  n1 = ScalarParameter() {
   defaultValue=1
   desc="Enables default background gradient for Item Card.\r\nbool"
   group=DefaultBackground
   parameterName=UseDefaultBg
   sliderMax=1
   sortPriority=0
  }
  LinearInterpolate(A=LinearInterpolate(A=NamedRerouteUsage(), B=NamedRerouteUsage(), Alpha=@n1), B=NamedRerouteUsage(), Alpha=NamedRerouteUsage()) {constAlpha=0.8}
  n2 = ScalarParameter() {defaultValue=20, group=Layout, parameterName=TopLeftCornerRadius, sortPriority=4}
  n3 = ScalarParameter() {defaultValue=20, group=Layout, parameterName=BottomLeftCornerRadius, sortPriority=7}
  n4 = ScalarParameter() {defaultValue=20, group=Layout, parameterName=TopRightCornerRadius, sortPriority=5}
  n5 = ScalarParameter() {defaultValue=20, group=Layout, parameterName=BottomRightCornerRadius, sortPriority=6}
  n6 = ScalarParameter() {
   defaultValue=1
   desc="If enabled reads corner radius values in pixels.\r\nbool"
   group=Layout
   parameterName=UseValueInPixels
   sliderMax=1
   sortPriority=1
  }
  NamedRerouteDeclaration(None=MaterialFunctionCall(
    Corner Radius=MaterialFunctionCall(R=@n2, G=@n3, B=@n4, A=@n5) {materialFunction=&AppendMany}.RGBA
    Radius in Pixels=@n6
   ) {materialFunction=&MF_UI_CoreUI_RoundedMask}.Shape Mask) {name=Shape Mask, nodeColor=<rgba 0 0 0 1>}
  NamedRerouteDeclaration(None=MaterialFunctionCall!(Post Buffer Blur=MaterialFunctionCall!("UVs (Optional)") {materialFunction=/Engine/Functions/UserInterface/GetSlatePost0}.LinearRGB) {materialFunction=/FortUILibrary/MaterialFunctions/CoreUI/MF_UI_CoreUI_BackplateB}.Result) {name=Default Bg, nodeColor=<rgba 0 0 0 1>}
  n7 = VectorParameter() {group=Archived, parameterName=ArchivedColor_1, sortPriority=1, defaultValue=<rgba 0 0 0 1>}
  n8 = VectorParameter() {group=Archived, parameterName=ArchivedColor_2, sortPriority=2, defaultValue=<rgba 0 0.0633 0.2083 1>}
  n9 = MaterialFunctionCall(
   GradientStart=ScalarParameter() {group=Archived, parameterName=ArchivedGradientStart, sortPriority=10}
   GradientEnd=ScalarParameter() {defaultValue=2, group=Archived, parameterName=ArchivedGradientEnd, sortPriority=11}
   RotationAngle=ScalarParameter() {defaultValue=0.75, group=Archived, parameterName=ArchivedGradientAngle, sortPriority=12}
  ) {materialFunction=&MF_UI_LinearGradient}
  n10 = LinearInterpolate(A=@n7.A, B=@n8.A, Alpha=@n9.Result)
  n11 = ScalarParameter() {
   defaultValue=0.95
   desc=Multiplies Archived Colors alpha. A quick way to set desired threshold in applying the archived effect.
   group=Archived
   parameterName=ArchivedAlphaMultiplier
   sliderMax=1
   sortPriority=3
  }
  n12 = ScalarParameter() {
   desc="(0 - 1) Animate item Archiving effect.\r\n0: Item not archived.\r\n1: Item archived."
   group=Archived
   parameterName=ArchivedAnimation
   sliderMax=1
   sortPriority=0
  }
  NamedRerouteDeclaration(None=LinearInterpolate(B=Multiply(A=@n10, B=@n11), Alpha=@n12)) {name=Archived Alpha, nodeColor=<rgba 0 0 0 1>}
  n13 = VectorParameter() {group=DefaultBackground, parameterName=BgColor_2, sortPriority=2, defaultValue=<rgba 0.02 0 0.4 0.7>}
  n14 = VectorParameter() {group=DefaultBackground, parameterName=BgColor_1, sortPriority=1, defaultValue=<rgba 1 0 0.7396 1>}
  n15 = Multiply(A=@n14.RGB) {constB=0.5}
  n16 = Multiply(A=@n15, B=@n15)
  n17 = ScalarParameter() {group=DefaultBackground, parameterName=SingleColorBg, sliderMax=1}
  n18 = ScalarParameter() {group=DefaultBackground, parameterName=BgGradientStart, sortPriority=10}
  n19 = ScalarParameter() {defaultValue=0.9, group=DefaultBackground, parameterName=BgGradientEnd, sortPriority=11}
  n20 = ScalarParameter() {
   defaultValue=0.875
   desc="Angle for linear gradient:\r\n0: from left to right\r\n0.25: from bottom to top\r\n0.5: from right to left\r\n0.75: from top to bottom"
   group=DefaultBackground
   parameterName=BgGradientAngle
   sortPriority=12
  }
  NamedRerouteDeclaration(None=LinearInterpolate(
    A=@n14.RGB
    B=LinearInterpolate(A=@n13.RGB, B=@n16, Alpha=@n17)
    Alpha=MaterialFunctionCall(GradientStart=@n18, GradientEnd=@n19, RotationAngle=@n20) {materialFunction=&MF_UI_LinearGradient}.Result
   )) {name=Colored Bg, nodeColor=<rgba 0 0 0 1>}
  NamedRerouteDeclaration(None=LinearInterpolate(A=@n7.RGB, B=@n8.RGB, Alpha=@n9.Result)) {name=Archived, nodeColor=<rgba 0 0 0 1>}
  NamedRerouteUsage()
 ]
}
