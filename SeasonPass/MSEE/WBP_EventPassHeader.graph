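# Event pass header widget showing time remaining, pass name, scrollable description and a Go To Quests button that broadcasts OnQuestButtonClicked.
# OnPassEndDateUpdated starts a looping minute timer that refreshes the days/hours countdown, cleared on Destruct.
# Theme updates only store text colors from the color reference table; UpdateTextColors applies them, and is called here only on PreConstruct.
# UpdateDecorators passes decorator classes to all three rich text blocks.
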WidgetBlueprint /SeasonPass/MSEE/WBP_EventPassHeader
graphs [
 EventGraph {nodes [
   "On Clicked (Button_GoToQuests)"!() -> (OutputDelegate, then, Button:Common Button Base Object Reference) {is=title K2Node_ComponentBoundEvent}
   > Call On Quest Button Clicked!(execute:Exec, self:WBP Event Pass Header Object Reference) -> (then:Exec) {is=title K2Node_CallDelegate}
   n1 = OnEndDateTimerFinishedLoop() {is=title K2Node_CustomEvent}
   n2 = OnPassEndDateUpdated!() -> (OutputDelegate, then, EndDate:Date Time Structure) {is=title K2Node_CustomEvent}
   n3 = Get Duration!(A:Timespan Structure=DateTime - DateTime!(A:Date Time Structure=@n2.EndDate, B:Date Time Structure=Now!() -> (ReturnValue:Date Time Structure)) -> (ReturnValue:Timespan Structure)) -> (ReturnValue_Days:int=0, ReturnValue_Hours:int=0, ReturnValue_Minutes:int=0, ReturnValue_Seconds:int=0, ReturnValue_Milliseconds:int=0) {is=title K2Node_CallFunction}
   SetText(
    execute=[@n1.then, @n5.then]
    self=Get RichText_TimeRemaining() -> (RichText_TimeRemaining:Common Rich Text Block Object Reference)
    InText=Format Text!(Format="LOCTEXT(\"Event ends in {Days}D {Hours}H\")", Days:int=@n3, Hours:int=@n3.ReturnValue_Hours) -> (Result)
   )
   K2Node_Event_2 = UpdateDecorators!() -> (OutputDelegate, then, InDecoratorClasses:Array of Rich Text Block Decorator Class References) {is=title K2Node_CustomEvent}
   > SetDecorators!(
    execute
    self:Rich Text Block Object Reference=[
     Get RichText_TimeRemaining()
     Get RichText_Description() -> (RichText_Description:Common Rich Text Block Object Reference)
     Get RichText_PassName() -> (RichText_PassName:Common Rich Text Block Object Reference)
    ]
    InDecoratorClasses:Array of Rich Text Block Decorator Class References=@K2Node_Event_2.InDecoratorClasses
   ) -> (then)
   Event PreConstruct()
   > UpdateTextColors()
   Event Destruct()
   > Clear and Invalidate Timer by Handle(Handle=Get EndDateTimerHandle() -> (EndDateTimerHandle:Timer Handle Structure))
   n4 = Set Timer by Event(execute=@n2.then, Delegate=@n1, Time=60, bLooping=true)
   > n5 = Set EndDateTimerHandle(EndDateTimerHandle:Timer Handle Structure=@n4) -> (Output_Get:Timer Handle Structure)
  ]}
 GetEndDate {nodes [
   n1 = GetEndDate!() -> (
    then
    CMSdata_PurchaseBackgroundURL:String
    CMSdata_PurchaseConfirmBackgroundURL:String
    CMSdata_PurchaseDisclaimer:Text
    CMSdata_PurchaseDescription:Text
    CMSdata_PurchaseConfirmDescription:Text
    CMSdata_LandingPageDisclaimer:Text
    CMSdata_LevelPurchaseDisclaimer:Text
    CMSdata_SeasonEndDateText:Text
    CMSdata_SeasonEndDateTime:Date Time Structure
   ) {is=title K2Node_FunctionEntry}
   > Return Node!(execute, EndDate:Date Time Structure=@n1.CMSdata_SeasonEndDateTime)
  ]}
 OnQuestButtonClicked {nodes [OnQuestButtonClicked() {is=title K2Node_FunctionEntry}]}
 OnThemeDataUpdated_VM {nodes [
   n1 = OnThemeDataUpdated_VM!() -> (then, SeasonPassThemeData:"Season Pass Theme Data Structure (by ref)") {is=title K2Node_FunctionEntry}
   n2 = Break Season Pass Theme Data!(SeasonPassThemeData:"Season Pass Theme Data Structure (by ref)"=@n1) -> (
    TitleColor:Linear Color Structure
    TitleMaterial:Soft Object Reference
    TitleOutline:Font Outline Settings Structure
    SeasonColor:Linear Color Structure
    ProgressionColor:Linear Color Structure
    OverlayColor:Linear Color Structure
    VaultBackgroundMaterial:Soft Object Reference
    GiftboxVaultBackgroundMaterial:Soft Object Reference
    UIBackgroundMaterial:Soft Object Reference
    UpsellBackgroundMaterial:Soft Object Reference
    CrewCategoryBackgroundMaterial:Soft Object Reference
    CustomLogo:Soft Object Reference
    CustomSeasonTitle:Text
    CustomSeasonTitleColor:Linear Color Structure
    Upsell2BackgrounMaterial:Soft Object Reference
    ImageReferenceTable:Fort Image Reference Table Structure
    ColorReferenceTable:Fort Color Reference Table Structure
    TextReferenceTable:Fort Text Reference Table Structure
   ) {is=title K2Node_BreakStruct}
   > Set TimeRemainingTextColor(TimeRemainingTextColor:Linear Color Structure=LookupColor!(Table:"Fort Color Reference Table Structure (by ref)"=@n2.ColorReferenceTable, Name:String=TimeRemaining) -> (ReturnValue:Linear Color Structure)) -> (Output_Get:Linear Color Structure)
   > Set EventNameTextColor(EventNameTextColor:Linear Color Structure=LookupColor(Table=@n2.ColorReferenceTable, Name=PassName)) -> (Output_Get:Linear Color Structure)
   > Set EventDescriptionTextColor(EventDescriptionTextColor:Linear Color Structure=LookupColor(Table=@n2.ColorReferenceTable, Name=PassName)) -> (Output_Get:Linear Color Structure)
  ]}
 UpdateTextColors {nodes [
   UpdateTextColors() {is=title K2Node_FunctionEntry}
   > SetDefaultColorAndOpacity!(
    execute
    self:Rich Text Block Object Reference=Get RichText_TimeRemaining()
    InColorAndOpacity:Slate Color Structure=Make Slate Color(SpecifiedColor=Get TimeRemainingTextColor() -> (TimeRemainingTextColor:Linear Color Structure))
   ) -> (then)
   > SetDefaultColorAndOpacity(
    self=Get RichText_PassName()
    InColorAndOpacity=Make Slate Color(SpecifiedColor=Get EventNameTextColor() -> (EventNameTextColor:Linear Color Structure))
   )
   > SetDefaultColorAndOpacity(
    self=Get RichText_Description()
    InColorAndOpacity=Make Slate Color(SpecifiedColor=Get EventDescriptionTextColor() -> (EventDescriptionTextColor:Linear Color Structure))
   )
  ]}
]
variables [
 AccentColor {type=LinearColor, default=<rgba 1 1 1 1>}
 EndDateTimerHandle {type=TimerHandle}
 EventDescriptionTextColor {type=LinearColor, default=<rgba 1 1 1 1>}
 EventNameTextColor {type=LinearColor, default=<rgba 1 1 1 1>}
 OnQuestButtonClicked {}
 Rich Text Decorators {type="array<Class>", default []}
 Rich Text Style Set {type=DataTable, default=null}
 TimeRemainingTextColor {type=LinearColor, default=<rgba 0.2582 0.7157 0.624 1>}
]
widgets {
 info {inheritedWidgetCount=0, parentClass=&UserWidget, rootWidgetClass=&VerticalBox, widgetCount=6}
 widgets [
  VB_Content {widgetClassPath=&VerticalBox, widget=VB_Content}
  RichText_TimeRemaining {
   widgetClassPath=&CommonRichTextBlock
   bIsVariable=true
   parent=VB_Content
   widget=RichText_TimeRemaining
   properties {
    autoWrapText=true
    bOverrideDefaultStyle=true
    mobileTextBlockScale=1.15
    text="[PH] Event Ends in..."
    textTransformPolicy=ToUpper
    defaultTextStyleOverride {
     overflowPolicy=Clip
     transformPolicy=None
     colorAndOpacity {colorUseRule=UseColor_Specified, specifiedColor=<rgba 1 0 1 1>}
     font {
      bForceMonospaced=false
      bMaterialIsStencil=false
      fontMaterial=null
      fontObject=&HeadingNow
      letterSpacing=0
      monospacedWidth=1
      size=18
      skewAmount=0
      typefaceFontName=75 Medium
      outlineSettings {
       bApplyOutlineToDropShadows=false
       bMiteredCorners=false
       bSeparateFillAlpha=false
       outlineMaterial=null
       outlineSize=0
       outlineColor=<rgba 0 0 0 1>
      }
     }
     highlightColor {colorUseRule=UseColor_Specified, specifiedColor=<rgba 0 0 0 1>}
     highlightShape {
      bIsDynamicallyLoaded=false
      drawAs=Image
      imageType=NoImage
      mirroring=NoMirror
      resourceName=None
      resourceObject=null
      tiling=NoTile
      imageSize=<32 32>
      margin=<ltrb 0 0 0 0>
      outlineSettings {
       bUseBrushTransparency=false
       roundingType=HalfHeightRadius
       width=0
       color {colorUseRule=UseColor_Specified, specifiedColor=<rgba 0 0 0 0>}
       cornerRadii=<0 0 0 0>
      }
      tintColor {colorUseRule=UseColor_Specified, specifiedColor=<rgba 1 1 1 1>}
      uVRegion {bIsValid=false, max=<0 0>, min=<0 0>}
     }
     selectedBackgroundColor {colorUseRule=UseColor_Foreground, specifiedColor=<rgba 1 0 1 1>}
     shadowColorAndOpacity=<rgba 0 0 0 1>
     shadowOffset=<0 0>
     strikeBrush {
      bIsDynamicallyLoaded=false
      drawAs=Image
      imageType=NoImage
      mirroring=NoMirror
      resourceName=None
      resourceObject=null
      tiling=NoTile
      imageSize=<32 32>
      margin=<ltrb 0 0 0 0>
      outlineSettings {
       bUseBrushTransparency=false
       roundingType=HalfHeightRadius
       width=0
       color {colorUseRule=UseColor_Specified, specifiedColor=<rgba 0 0 0 0>}
       cornerRadii=<0 0 0 0>
      }
      tintColor {colorUseRule=UseColor_Specified, specifiedColor=<rgba 1 1 1 1>}
      uVRegion {bIsValid=false, max=<0 0>, min=<0 0>}
     }
     underlineBrush {
      bIsDynamicallyLoaded=false
      drawAs=Image
      imageType=NoImage
      mirroring=NoMirror
      resourceName=None
      resourceObject=null
      tiling=NoTile
      imageSize=<32 32>
      margin=<ltrb 0 0 0 0>
      outlineSettings {
       bUseBrushTransparency=false
       roundingType=HalfHeightRadius
       width=0
       color {colorUseRule=UseColor_Specified, specifiedColor=<rgba 0 0 0 0>}
       cornerRadii=<0 0 0 0>
      }
      tintColor {colorUseRule=UseColor_Specified, specifiedColor=<rgba 1 1 1 1>}
      uVRegion {bIsValid=false, max=<0 0>, min=<0 0>}
     }
    }
   }
   slot {class=VerticalBoxSlot}
  }
  RichText_PassName {
   widgetClassPath=&CommonRichTextBlock
   bIsVariable=true
   parent=VB_Content
   widget=RichText_PassName
   properties {
    bOverrideDefaultStyle=true
    mobileTextBlockScale=0.8
    text="[PH] Event Name"
    textTransformPolicy=ToUpper
    defaultTextStyleOverride {
     overflowPolicy=Clip
     transformPolicy=None
     colorAndOpacity {colorUseRule=UseColor_Specified, specifiedColor=<rgba 1 1 1 1>}
     font {
      bForceMonospaced=false
      bMaterialIsStencil=false
      fontMaterial=null
      fontObject=&HeadingNow
      letterSpacing=0
      monospacedWidth=1
      size=36
      skewAmount=0
      typefaceFontName=86 Bold Italic
      outlineSettings {
       bApplyOutlineToDropShadows=false
       bMiteredCorners=false
       bSeparateFillAlpha=false
       outlineMaterial=null
       outlineSize=0
       outlineColor=<rgba 0 0 0 1>
      }
     }
     highlightColor {colorUseRule=UseColor_Specified, specifiedColor=<rgba 0 0 0 1>}
     highlightShape {
      bIsDynamicallyLoaded=false
      drawAs=Image
      imageType=NoImage
      mirroring=NoMirror
      resourceName=None
      resourceObject=null
      tiling=NoTile
      imageSize=<32 32>
      margin=<ltrb 0 0 0 0>
      outlineSettings {
       bUseBrushTransparency=false
       roundingType=HalfHeightRadius
       width=0
       color {colorUseRule=UseColor_Specified, specifiedColor=<rgba 0 0 0 0>}
       cornerRadii=<0 0 0 0>
      }
      tintColor {colorUseRule=UseColor_Specified, specifiedColor=<rgba 1 1 1 1>}
      uVRegion {bIsValid=false, max=<0 0>, min=<0 0>}
     }
     selectedBackgroundColor {colorUseRule=UseColor_Foreground, specifiedColor=<rgba 1 0 1 1>}
     shadowColorAndOpacity=<rgba 0 0 0 1>
     shadowOffset=<0 0>
     strikeBrush {
      bIsDynamicallyLoaded=false
      drawAs=Image
      imageType=NoImage
      mirroring=NoMirror
      resourceName=None
      resourceObject=null
      tiling=NoTile
      imageSize=<32 32>
      margin=<ltrb 0 0 0 0>
      outlineSettings {
       bUseBrushTransparency=false
       roundingType=HalfHeightRadius
       width=0
       color {colorUseRule=UseColor_Specified, specifiedColor=<rgba 0 0 0 0>}
       cornerRadii=<0 0 0 0>
      }
      tintColor {colorUseRule=UseColor_Specified, specifiedColor=<rgba 1 1 1 1>}
      uVRegion {bIsValid=false, max=<0 0>, min=<0 0>}
     }
     underlineBrush {
      bIsDynamicallyLoaded=false
      drawAs=Image
      imageType=NoImage
      mirroring=NoMirror
      resourceName=None
      resourceObject=null
      tiling=NoTile
      imageSize=<32 32>
      margin=<ltrb 0 0 0 0>
      outlineSettings {
       bUseBrushTransparency=false
       roundingType=HalfHeightRadius
       width=0
       color {colorUseRule=UseColor_Specified, specifiedColor=<rgba 0 0 0 0>}
       cornerRadii=<0 0 0 0>
      }
      tintColor {colorUseRule=UseColor_Specified, specifiedColor=<rgba 1 1 1 1>}
      uVRegion {bIsValid=false, max=<0 0>, min=<0 0>}
     }
    }
   }
   slot {class=VerticalBoxSlot, properties {horizontalAlignment=HAlign_Left, verticalAlignment=VAlign_Top}}
  }
  ScrollBox_Description {
   widgetClassPath=&ScrollBox
   bIsVariable=true
   parent=VB_Content
   widget=ScrollBox_Description
   properties {allowOverscroll=false, bAllowRightClickDragScrolling=false, scrollBarVisibility=Collapsed, visibility=SelfHitTestInvisible}
   slot {class=VerticalBoxSlot, properties {verticalAlignment=VAlign_Top, padding=<ltrb 0 16 0 0>, size {sizeRule=Fill, value=1}}}
  }
  RichText_Description {
   widgetClassPath=&CommonRichTextBlock
   bIsVariable=true
   parent=ScrollBox_Description
   widget=RichText_Description
   properties {
    bOverrideDefaultStyle=true
    lineHeightPercentage=1.2
    mobileTextBlockScale=1.2
    text="[PH] <text color=\"FCE104FF\">Event Description</>"
    wrapTextAt=500
    defaultTextStyleOverride {
     overflowPolicy=Clip
     transformPolicy=None
     colorAndOpacity {colorUseRule=UseColor_Specified, specifiedColor=<rgba 1 1 1 1>}
     font {
      bForceMonospaced=false
      bMaterialIsStencil=false
      fontMaterial=null
      fontObject=&HeadingNow
      letterSpacing=0
      monospacedWidth=1
      size=16.5
      skewAmount=0
      typefaceFontName=75 Medium
      outlineSettings {
       bApplyOutlineToDropShadows=false
       bMiteredCorners=false
       bSeparateFillAlpha=false
       outlineMaterial=null
       outlineSize=0
       outlineColor=<rgba 0 0 0 1>
      }
     }
     highlightColor {colorUseRule=UseColor_Specified, specifiedColor=<rgba 0 0 0 1>}
     highlightShape {
      bIsDynamicallyLoaded=false
      drawAs=Image
      imageType=NoImage
      mirroring=NoMirror
      resourceName=None
      resourceObject=null
      tiling=NoTile
      imageSize=<32 32>
      margin=<ltrb 0 0 0 0>
      outlineSettings {
       bUseBrushTransparency=false
       roundingType=HalfHeightRadius
       width=0
       color {colorUseRule=UseColor_Specified, specifiedColor=<rgba 0 0 0 0>}
       cornerRadii=<0 0 0 0>
      }
      tintColor {colorUseRule=UseColor_Specified, specifiedColor=<rgba 1 1 1 1>}
      uVRegion {bIsValid=false, max=<0 0>, min=<0 0>}
     }
     selectedBackgroundColor {colorUseRule=UseColor_Foreground, specifiedColor=<rgba 1 0 1 1>}
     shadowColorAndOpacity=<rgba 0 0 0 1>
     shadowOffset=<0 0>
     strikeBrush {
      bIsDynamicallyLoaded=false
      drawAs=Image
      imageType=NoImage
      mirroring=NoMirror
      resourceName=None
      resourceObject=null
      tiling=NoTile
      imageSize=<32 32>
      margin=<ltrb 0 0 0 0>
      outlineSettings {
       bUseBrushTransparency=false
       roundingType=HalfHeightRadius
       width=0
       color {colorUseRule=UseColor_Specified, specifiedColor=<rgba 0 0 0 0>}
       cornerRadii=<0 0 0 0>
      }
      tintColor {colorUseRule=UseColor_Specified, specifiedColor=<rgba 1 1 1 1>}
      uVRegion {bIsValid=false, max=<0 0>, min=<0 0>}
     }
     underlineBrush {
      bIsDynamicallyLoaded=false
      drawAs=Image
      imageType=NoImage
      mirroring=NoMirror
      resourceName=None
      resourceObject=null
      tiling=NoTile
      imageSize=<32 32>
      margin=<ltrb 0 0 0 0>
      outlineSettings {
       bUseBrushTransparency=false
       roundingType=HalfHeightRadius
       width=0
       color {colorUseRule=UseColor_Specified, specifiedColor=<rgba 0 0 0 0>}
       cornerRadii=<0 0 0 0>
      }
      tintColor {colorUseRule=UseColor_Specified, specifiedColor=<rgba 1 1 1 1>}
      uVRegion {bIsValid=false, max=<0 0>, min=<0 0>}
     }
    }
   }
   slot {class=ScrollBoxSlot}
  }
  Button_GoToQuests {
   widgetClassPath=/FortUILibrary/UIKit/Buttons/WBP_UIKit_Button_Loud
   bIsVariable=true
   parent=VB_Content
   widget=Button_GoToQuests
   properties {text=Go To Quests}
   slot {class=VerticalBoxSlot, properties {horizontalAlignment=HAlign_Left, verticalAlignment=VAlign_Top, padding=<ltrb 0 24 0 0>}}
  }
 ]
}
