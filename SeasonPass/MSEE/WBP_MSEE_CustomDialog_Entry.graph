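# Season Pass MSEE custom dialog entry widget that stacks a rich-text header, an optional lazy-loaded image and a rich-text body.
# PopulateEntry sets both texts and applies the supplied decorator classes to each.
# A valid image reference shows the image at the given size, overriding only the brush size before loading, while an invalid one collapses it.
# PreConstruct fills placeholder content at design time only, for preview.
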WidgetBlueprint /SeasonPass/MSEE/WBP_MSEE_CustomDialog_Entry
graphs [
 EventGraph {nodes [
   n1 = Event PreConstruct()
   > Branch(Condition=@n1.IsDesignTime)
   > PopulateEntry!(
    execute
    self
    Header:"Text (by ref)"=MakeLiteralText(Value="INVTEXT(\"[PH] Header\")")
    Body:"Text (by ref)"=MakeLiteralText(Value="INVTEXT(\"Normal Text, <text color=\\\"FFF\\\" fontface=\\\"76 Bold\\\">Bold Text,</> \\r\\n<text color=\\\"3CADFFFF\\\">Color Text, </><text color=\\\"3CADFFFF\\\" fontface=\\\"76 Bold\\\">Bold Color Text</>  \\r\\n\\r\\nIcon <img id=\\\"ResourceIcon\\\"/>\")")
    ImageReference:"Fort Image Reference Structure (by ref)"=Make Fort Image Reference!(
     Image:Soft Object Reference=/FortUILibrary/Textures/MaterialInstances/MI_UIKit_Icon_ImageError
     ImageUrl:String
     MaterialTextureParamName:Name=None
    ) -> (FortImageReference:Fort Image Reference Structure) {is=title K2Node_MakeStruct}
    Image Size:Vector 2D Structure=MakeVector2D(X=128, Y=128)
    InDecoratorClasses:Array of Rich Text Block Decorator Class References
   ) -> (then)
  ]}
 PopulateEntry {nodes [
   n1 = PopulateEntry!() -> (
    then
    Header:"Text (by ref)"
    Body:"Text (by ref)"
    ImageReference:"Fort Image Reference Structure (by ref)"
    Image Size:Vector 2D Structure
    InDecoratorClasses:Array of Rich Text Block Decorator Class References
   ) {is=title K2Node_FunctionEntry}
   n2 = Get Text_Header() -> (Text_Header:Common Rich Text Block Object Reference)
   > SetText(self=@n2, InText=@n1)
   > SetDecorators!(
    execute
    self:Rich Text Block Object Reference=@n2
    InDecoratorClasses:Array of Rich Text Block Decorator Class References=@n1.InDecoratorClasses
   ) -> (then)
   n3 = Get Text_Body() -> (Text_Body:Common Rich Text Block Object Reference)
   > SetText(self=@n3, InText=@n1.Body)
   > SetDecorators(self=@n3, InDecoratorClasses=@n1.InDecoratorClasses)
   > n4 = Branch(Condition=IsValidImageReference!(ImageReference:"Fort Image Reference Structure (by ref)"=@n1.ImageReference) -> (ReturnValue:bool))
   > SetVisibility(self=Get LazyImage() -> (LazyImage:Fort Lazy Image Object Reference), InVisibility=HitTestInvisible)
   n5 = Get LazyImage()
   n6 = Get!(self:Image Object Reference=@n5) -> (
    Brush_TintColor:Slate Color Structure
    Brush_DrawAs:ESlateBrushDrawType Enum
    Brush_Tiling:ESlateBrushTileType Enum
    Brush_Mirroring:ESlateBrushMirrorType Enum
    Brush_ImageSize:Vector 2D Structure
    Brush_Margin:Margin Structure
    Brush_ResourceObject:Object Reference
    Brush_OutlineSettings:Slate Brush Outline Settings Structure
   )
   > SetBrush!(
    execute
    self=@n5
    InBrush_TintColor:Slate Color Structure=@n6
    InBrush_DrawAs:ESlateBrushDrawType Enum=@n6.Brush_DrawAs
    InBrush_Tiling:ESlateBrushTileType Enum=@n6.Brush_Tiling
    InBrush_Mirroring:ESlateBrushMirrorType Enum=@n6.Brush_Mirroring
    InBrush_ImageSize:Vector 2D Structure=@n1.Image Size
    InBrush_Margin:Margin Structure=@n6.Brush_Margin
    InBrush_ResourceObject:Object Reference=@n6.Brush_ResourceObject
    InBrush_OutlineSettings:Slate Brush Outline Settings Structure=@n6.Brush_OutlineSettings
   ) -> (then)
   > SetBrushFromImageReference!(
    execute
    self:Fort Lazy Image Object Reference=@n5
    ImageReference:"Fort Image Reference Structure (by ref)"=Get ImageReference() -> (ImageReference:"Fort Image Reference Structure (by ref)")
    bMatchSize:bool=false
    DownloadPriority:ECMSDownloadPriority Enum=Normal
   ) -> (then)
   SetVisibility(execute=@n4.else, self=Get LazyImage(), InVisibility=Collapsed)
  ]}
]
variables []
widgets {
 info {inheritedWidgetCount=0, parentClass=&UserWidget, rootWidgetClass=&VerticalBox, widgetCount=4}
 widgets [
  VerticalBox_38 {widgetClassPath=&VerticalBox, widget=VerticalBox_38}
  Text_Header {
   widgetClassPath=&CommonRichTextBlock
   bIsVariable=true
   parent=VerticalBox_38
   widget=Text_Header
   properties {
    autoWrapText=true
    bAutoCollapseWithEmptyText=true
    bOverrideDefaultStyle=true
    lineHeightPercentage=1.15
    mobileTextBlockScale=1.2
    defaultTextStyleOverride {
     overflowPolicy=Clip
     transformPolicy=None
     colorAndOpacity {colorUseRule=UseColor_Specified, specifiedColor=<rgba 1 1 1 1>}
     font {
      bForceMonospaced=false
      bMaterialIsStencil=false
      fontMaterial=null
      fontObject=&HeadingNow
      letterSpacing=0
      monospacedWidth=1
      size=19.5
      skewAmount=0
      typefaceFontName=76 Bold
      outlineSettings {
       bApplyOutlineToDropShadows=false
       bMiteredCorners=false
       bSeparateFillAlpha=false
       outlineMaterial=null
       outlineSize=0
       outlineColor=<rgba 0 0 0 1>
      }
     }
     highlightColor {colorUseRule=UseColor_Specified, specifiedColor=<rgba 0 0 0 0>}
     highlightShape {
      bIsDynamicallyLoaded=false
      drawAs=Image
      imageType=NoImage
      mirroring=NoMirror
      resourceName=None
      resourceObject=null
      tiling=NoTile
      imageSize=<32 32>
      margin=<ltrb 0 0 0 0>
      outlineSettings {
       bUseBrushTransparency=false
       roundingType=HalfHeightRadius
       width=0
       color {colorUseRule=UseColor_Specified, specifiedColor=<rgba 0 0 0 0>}
       cornerRadii=<0 0 0 0>
      }
      tintColor {colorUseRule=UseColor_Specified, specifiedColor=<rgba 1 1 1 1>}
      uVRegion {bIsValid=false, max=<0 0>, min=<0 0>}
     }
     selectedBackgroundColor {colorUseRule=UseColor_Foreground, specifiedColor=<rgba 1 0 1 1>}
     shadowColorAndOpacity=<rgba 0 0 0 1>
     shadowOffset=<0 0>
     strikeBrush {
      bIsDynamicallyLoaded=false
      drawAs=Image
      imageType=NoImage
      mirroring=NoMirror
      resourceName=None
      resourceObject=null
      tiling=NoTile
      imageSize=<32 32>
      margin=<ltrb 0 0 0 0>
      outlineSettings {
       bUseBrushTransparency=false
       roundingType=HalfHeightRadius
       width=0
       color {colorUseRule=UseColor_Specified, specifiedColor=<rgba 0 0 0 0>}
       cornerRadii=<0 0 0 0>
      }
      tintColor {colorUseRule=UseColor_Specified, specifiedColor=<rgba 1 1 1 1>}
      uVRegion {bIsValid=false, max=<0 0>, min=<0 0>}
     }
     underlineBrush {
      bIsDynamicallyLoaded=false
      drawAs=Image
      imageType=NoImage
      mirroring=NoMirror
      resourceName=None
      resourceObject=null
      tiling=NoTile
      imageSize=<32 32>
      margin=<ltrb 0 0 0 0>
      outlineSettings {
       bUseBrushTransparency=false
       roundingType=HalfHeightRadius
       width=0
       color {colorUseRule=UseColor_Specified, specifiedColor=<rgba 0 0 0 0>}
       cornerRadii=<0 0 0 0>
      }
      tintColor {colorUseRule=UseColor_Specified, specifiedColor=<rgba 1 1 1 1>}
      uVRegion {bIsValid=false, max=<0 0>, min=<0 0>}
     }
    }
   }
   slot {class=VerticalBoxSlot, properties {padding=<ltrb 0 0 0 16>}}
  }
  LazyImage {
   widgetClassPath=&FortLazyImage
   bIsVariable=true
   parent=VerticalBox_38
   widget=LazyImage
   properties {visibility=Collapsed}
   slot {class=VerticalBoxSlot, properties {horizontalAlignment=HAlign_Left, verticalAlignment=VAlign_Top, padding=<ltrb 0 0 0 24>}}
  }
  Text_Body {
   widgetClassPath=&CommonRichTextBlock
   bIsVariable=true
   parent=VerticalBox_38
   widget=Text_Body
   properties {
    autoWrapText=true
    bAutoCollapseWithEmptyText=true
    bOverrideDefaultStyle=true
    lineHeightPercentage=1.15
    mobileTextBlockScale=1.2
    defaultTextStyleOverride {
     overflowPolicy=Clip
     transformPolicy=None
     colorAndOpacity {colorUseRule=UseColor_Specified, specifiedColor=<rgba 0.5904 0.5839 0.617 1>}
     font {
      bForceMonospaced=false
      bMaterialIsStencil=false
      fontMaterial=null
      fontObject=&HeadingNow
      letterSpacing=0
      monospacedWidth=1
      size=17.25
      skewAmount=0
      typefaceFontName=74 Regular
      outlineSettings {
       bApplyOutlineToDropShadows=true
       bMiteredCorners=false
       bSeparateFillAlpha=false
       outlineMaterial=null
       outlineSize=0
       outlineColor=<rgba 0 0 0 1>
      }
     }
     highlightColor {colorUseRule=UseColor_Specified, specifiedColor=<rgba 0 0 0 0>}
     highlightShape {
      bIsDynamicallyLoaded=false
      drawAs=Image
      imageType=NoImage
      mirroring=NoMirror
      resourceName=None
      resourceObject=null
      tiling=NoTile
      imageSize=<32 32>
      margin=<ltrb 0 0 0 0>
      outlineSettings {
       bUseBrushTransparency=false
       roundingType=HalfHeightRadius
       width=0
       color {colorUseRule=UseColor_Specified, specifiedColor=<rgba 0 0 0 0>}
       cornerRadii=<0 0 0 0>
      }
      tintColor {colorUseRule=UseColor_Specified, specifiedColor=<rgba 1 1 1 1>}
      uVRegion {bIsValid=false, max=<0 0>, min=<0 0>}
     }
     selectedBackgroundColor {colorUseRule=UseColor_Foreground, specifiedColor=<rgba 1 0 1 1>}
     shadowColorAndOpacity=<rgba 0 0 0 1>
     shadowOffset=<0 0>
     strikeBrush {
      bIsDynamicallyLoaded=false
      drawAs=Image
      imageType=NoImage
      mirroring=NoMirror
      resourceName=None
      resourceObject=null
      tiling=NoTile
      imageSize=<32 32>
      margin=<ltrb 0 0 0 0>
      outlineSettings {
       bUseBrushTransparency=false
       roundingType=HalfHeightRadius
       width=0
       color {colorUseRule=UseColor_Specified, specifiedColor=<rgba 0 0 0 0>}
       cornerRadii=<0 0 0 0>
      }
      tintColor {colorUseRule=UseColor_Specified, specifiedColor=<rgba 1 1 1 1>}
      uVRegion {bIsValid=false, max=<0 0>, min=<0 0>}
     }
     underlineBrush {
      bIsDynamicallyLoaded=false
      drawAs=Image
      imageType=NoImage
      mirroring=NoMirror
      resourceName=None
      resourceObject=null
      tiling=NoTile
      imageSize=<32 32>
      margin=<ltrb 0 0 0 0>
      outlineSettings {
       bUseBrushTransparency=false
       roundingType=HalfHeightRadius
       width=0
       color {colorUseRule=UseColor_Specified, specifiedColor=<rgba 0 0 0 0>}
       cornerRadii=<0 0 0 0>
      }
      tintColor {colorUseRule=UseColor_Specified, specifiedColor=<rgba 1 1 1 1>}
      uVRegion {bIsValid=false, max=<0 0>, min=<0 0>}
     }
    }
   }
   slot {class=VerticalBoxSlot, properties {padding=<ltrb 0 0 0 48>}}
  }
 ]
}
