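# Scrollable dialog text widget for the MSEE season pass that shows a vertical list of header, body and image sections.
# PopulateSections clears the DynamicEntryBox, then creates a WBP_MSEE_CustomDialog_Entry per entry struct, passing its text, image and rich text decorator classes.
# SetTextModifiers stores event dates and names, which ApplyTextModifiers substitutes into supplied text via Format Text.
# PopulateSections does not itself call ApplyTextModifiers, and the construct and tick events are empty.
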WidgetBlueprint /SeasonPass/MSEE/WBP_MSEE_CustomDialogText
graphs [
 ApplyTextModifiers {nodes [
   n1 = ApplyTextModifiers!() -> (then, TextInput:Text) {is=title K2Node_FunctionEntry}
   > Return Node!(
    execute
    Result:Text=Format Text!(
     Format=@n1
     StartDate:Text=Get Var_StartDate() -> (Var_StartDate:Text)
     EndDate:Text=Get Var_EndDate() -> (Var_EndDate:Text)
     CurrencyName:Text=Get Var_CurrencyName() -> (Var_CurrencyName:Text)
     EventName:Text=Get Var_EventName() -> (Var_EventName:Text)
     TokenName:Text=Get Var_TokenName() -> (Var_TokenName:Text)
     QuestBundleName:Text=Get Var_QuestBundleName() -> (Var_QuestBundleName:Text)
     ChaseRewardName:Text=Get Var_ChaseRewardName() -> (Var_ChaseRewardName:Text)
    ) -> (Result)
   )
  ]}
 EventGraph {nodes [
   Event PreConstruct()
   Event Construct()
   Event Tick()
  ]}
 PopulateSections {nodes [
   n1 = PopulateSections!() -> (then, Entries:Array of S MSEE Custom Dialog Entry Structures) {is=title K2Node_FunctionEntry}
   > Reset!(
    execute
    self:Dynamic Entry Box Object Reference=Get DynamicEntryBox() -> (DynamicEntryBox:Dynamic Entry Box Object Reference)
    bDeleteWidgets:bool=true
   ) -> (then)
   > n2 = For Each Loop(Array:Array of S MSEE Custom Dialog Entry Structures=@n1) -> (Array Element:S MSEE Custom Dialog Entry Structure)
   > n3 = Create Entry of Class!(
    execute
    self:Dynamic Entry Box Object Reference=Get DynamicEntryBox()
    EntryClass:User Widget Class Reference=Get CustomDialogEntry() -> (CustomDialogEntry:User Widget Class Reference)
   ) -> (then, ReturnValue:User Widget Object Reference)
   > n4 = Cast To WBP_MSEE_CustomDialog_Entry!(execute, Object=@n3) -> (then, CastFailed, AsWBP MSEE Custom Dialog Entry:WBP MSEE Custom Dialog Entry Object Reference) {is=title K2Node_DynamicCast}
   n5 = Break S MSEE Custom Dialog Entry!(S_MSEE_CustomDialogEntry:"S MSEE Custom Dialog Entry Structure (by ref)"=@n2) -> (
    Header_2_#1:Text
    Body_4_#2:Text
    ImageReference_10_#3:Fort Image Reference Structure
    ImageSize_9_#4:Vector 2D Structure
    DecoratorClasses_14_#5:Array of Rich Text Block Decorator Class References
   ) {is=title K2Node_BreakStruct}
   > PopulateEntry!(
    execute
    self:WBP MSEE Custom Dialog Entry Object Reference=@n4
    Header:"Text (by ref)"=@n5
    Body:"Text (by ref)"=@n5.Body_4_#2
    ImageReference:"Fort Image Reference Structure (by ref)"=@n5.ImageReference_10_#3
    Image Size:Vector 2D Structure=@n5.ImageSize_9_#4
    InDecoratorClasses:Array of Rich Text Block Decorator Class References=@n5.DecoratorClasses_14_#5
   ) -> (then)
  ]}
 SetTextModifiers {nodes [
   n1 = SetTextModifiers!() -> (
    then
    Var_StartDate:Text
    Var_EndDate:Text
    Var_CurrencyName:Text
    Var_EventName:Text
    Var_TokenName:Text
    Var_QuestBundleName:Text
    Var_ChaseRewardName:Text
   ) {is=title K2Node_FunctionEntry}
   > n2 = Sequence!(execute) -> (then_0, then_1, then_2:Exec, then_3:Exec, then_4:Exec, then_5:Exec, then_6:Exec)
   > Set Var_StartDate(Var_StartDate:Text=@n1) -> (Output_Get:Text)
   Set Var_EndDate(execute=@n2.then_1, Var_EndDate:Text=@n1.Var_EndDate) -> (Output_Get:Text)
   Set Var_CurrencyName(execute=@n2.then_2, Var_CurrencyName:Text=@n1.Var_CurrencyName) -> (Output_Get:Text)
   Set Var_EventName(execute=@n2.then_3, Var_EventName:Text=@n1.Var_EventName) -> (Output_Get:Text)
   Set Var_TokenName(execute=@n2.then_4, Var_TokenName:Text=@n1.Var_TokenName) -> (Output_Get:Text)
   Set Var_QuestBundleName(execute=@n2.then_5, Var_QuestBundleName:Text=@n1.Var_QuestBundleName) -> (Output_Get:Text)
   Set Var_ChaseRewardName(execute=@n2.then_6, Var_ChaseRewardName:Text=@n1.Var_ChaseRewardName) -> (Output_Get:Text)
  ]}
]
variables [
 CustomDialogEntry {type=Class, default=&/WBP_MSEE_CustomDialog_Entry.WBP_MSEE_CustomDialog_Entry_C}
 Var_ChaseRewardName {type=string, default=""}
 Var_CurrencyName {type=string, default=""}
 Var_EndDate {type=string, default=""}
 Var_EventName {type=string, default=""}
 Var_QuestBundleName {type=string, default=""}
 Var_StartDate {type=string, default=""}
 Var_TokenName {type=string, default=""}
]
widgets {
 info {inheritedWidgetCount=0, parentClass=&UserWidget, rootWidgetClass=&ScrollBox, widgetCount=2}
 widgets [
  ScrollBox_0 {widgetClassPath=&ScrollBox, widget=ScrollBox_0}
  DynamicEntryBox {
   widgetClassPath=&DynamicEntryBox
   bIsVariable=true
   parent=ScrollBox_0
   widget=DynamicEntryBox
   properties {entryBoxType=Vertical, entryWidgetClass=&WBP_MSEE_CustomDialog_Entry_C}
   slot {class=ScrollBoxSlot}
  }
 ]
}
