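# MSEE About dialog: an activatable widget that wraps the UIKit internal dialog with a sectioned text body, an optional image and a legal-text footer.
# Public functions pass entries to the custom text widget, set the footer text, and load the image by reference, collapsing it when its URL is empty.
# PreConstruct sets design-time placeholder legal text and registers the body for scrolling.
# Button clicks, bound only while activated, deactivate the dialog.
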WidgetBlueprint /SeasonPass/MSEE/WBP_MSEE_AboutDialog
graphs [
 EventGraph {nodes [
   n1 = OnButtonClicked!() -> (OutputDelegate, then, ButtonIndex:int) {is=title K2Node_CustomEvent}
   > DeactivateWidget()
   n2 = Event PreConstruct()
   > n3 = Branch(Condition=@n2.IsDesignTime)
   > n4 = "SetText (Text)"(self=Get LegalText() -> (LegalText:Common Text Object Reference), InText="INVTEXT(\"[PH] Legal Text\")")
   SetScrollWidget!(
    execute=[@n4.then, @n3.else]
    ToScroll:Widget Object Reference=Get WBP_MSEE_CustomDialogText() -> (WBP_MSEE_CustomDialogText:WBP MSEE Custom Dialog Text Object Reference)
   ) -> (then)
   Event On Activated()
   > Bind Event to Dialog Button Clicked Dispatcher(
    self:WBP UIKit Dialog Internal Object Reference=Get WBP_UIKit_Dialog_Internal() -> (WBP_UIKit_Dialog_Internal:WBP UIKit Dialog Internal Object Reference)
    Delegate=@n1
   ) {is=title K2Node_AddDelegate}
   Event On Deactivated()
   > Unbind Event from Dialog Button Clicked Dispatcher(self:WBP UIKit Dialog Internal Object Reference=Get WBP_UIKit_Dialog_Internal(), Delegate=@n1) {is=title K2Node_RemoveDelegate}
  ]}
 PopulateEntries {nodes [
   n1 = PopulateEntries!() -> (then, Entries:Array of S MSEE Custom Dialog Entry Structures) {is=title K2Node_FunctionEntry}
   > PopulateSections!(
    execute
    self:WBP MSEE Custom Dialog Text Object Reference=Get WBP_MSEE_CustomDialogText()
    Entries:Array of S MSEE Custom Dialog Entry Structures=@n1
   ) -> (then)
  ]}
 SetFooterLegalText {nodes [
   n1 = SetFooterLegalText!() -> (then, InText:Text) {is=title K2Node_FunctionEntry}
   > "SetText (Text)"(self=Get LegalText(), InText=@n1)
  ]}
 SetImageByRef {nodes [
   n1 = SetImageByRef!() -> (then, ImageReference:"Fort Image Reference Structure (by ref)") {is=title K2Node_FunctionEntry}
   n2 = Break Fort Image Reference!(FortImageReference:"Fort Image Reference Structure (by ref)"=Get ImageReference() -> (ImageReference:"Fort Image Reference Structure (by ref)")) -> (Image:Soft Object Reference, ImageUrl:String, MaterialTextureParamName:Name) {is=title K2Node_BreakStruct}
   n3 = Get FortLazyImage() -> (FortLazyImage:Fort Lazy Image Object Reference)
   > SetVisibility(self=@n3, InVisibility=Select(Option 0=SelfHitTestInvisible, Option 1=Collapsed, Index=IsEmpty(InString=@n2.ImageUrl)))
   > SetBrushFromImageReference!(
    execute
    self:Fort Lazy Image Object Reference=@n3
    ImageReference:"Fort Image Reference Structure (by ref)"=@n1
    bMatchSize:bool=false
    DownloadPriority:ECMSDownloadPriority Enum=Normal
   ) -> (then)
  ]}
]
variables [
 Decorator Classes {type="array<Class>", default []}
 Rich Text Test {type=DataTable, default=null}
]
widgets {
 info {
  inheritedWidgetCount=0
  namedSlotCount=3
  parentClass=&CommonActivatableWidget
  rootWidgetClass=/FortUILibrary/Systems/Dialog/WBP_UIKit_Dialog_Internal.WBP_UIKit_Dialog_Internal_C
  widgetCount=7
 }
 widgets [
  WBP_UIKit_Dialog_Internal {
   widgetClassPath=/FortUILibrary/Systems/Dialog/WBP_UIKit_Dialog_Internal
   bIsVariable=true
   widget=WBP_UIKit_Dialog_Internal
   properties {
    buttons {
     each Button Horizontal Fill=false
     each Button Min Width=0
     each Button Size=Medium
     focuse Button=0
     space Between Buttons=30
     cTA Buttons Types [Regular]
     group Placement {
      horizontal Alignment=HAlign_Fill
      vertical Alignment=VAlign_Bottom
      grid {column=1, column Span=0, layer=0, row=2, row Span=0, nudge=<0 0>}
      padding=<ltrb 80 64 80 48>
     }
    }
    dialogSIze {
     height Override {enable=true, value=800}
     max Aspect Ratio {enable=false, value=0}
     max Desired Height {enable=true, value=800}
     max Desired Width {enable=false, value=0}
     min Aspect Ratio {enable=false, value=0}
     min Desired Height {enable=false, value=0}
     min Desired Width {enable=false, value=0}
     width Override {enable=true, value=1400}
    }
    imageProperties {
     material=/FortUILibrary/Systems/Dialog/MaterialInstances/MI_UI_DialogImage
     image Size=<512 0>
     placement {
      horizontal Alignment=HAlign_Fill
      vertical Alignment=VAlign_Fill
      grid {column=1, column Span=0, layer=20, row=0, row Span=1, nudge=<0 0>}
      padding=<ltrb 0 0 0 0>
     }
    }
   }
  }
  WBP_MSEE_CustomDialogText {
   widgetClassPath=&/WBP_MSEE_CustomDialogText
   bIsVariable=true
   namedSlotHost=WBP_UIKit_Dialog_Internal
   widget=WBP_MSEE_CustomDialogText
   properties {padding=<ltrb 80 0 80 0>}
  }
  Overlay_0 {widgetClassPath=&Overlay, namedSlotHost=WBP_UIKit_Dialog_Internal, widget=Overlay_0}
  FortLazyImage {
   widgetClassPath=&FortLazyImage
   bIsVariable=true
   parent=Overlay_0
   widget=FortLazyImage
   properties {
    visibility=Collapsed
    brush {
     bIsDynamicallyLoaded=false
     drawAs=Image
     imageType=FullColor
     mirroring=NoMirror
     resourceName=None
     resourceObject=/Game/Global/Textures/Default/Blanks/T_Blank_White_ZeroAlpha_sRGBColor
     tiling=NoTile
     imageSize=<512 886>
     margin=<ltrb 0 0 0 0>
     outlineSettings {
      bUseBrushTransparency=false
      roundingType=HalfHeightRadius
      width=0
      color {colorUseRule=UseColor_Specified, specifiedColor=<rgba 0 0 0 0>}
      cornerRadii=<0 0 0 0>
     }
     tintColor {colorUseRule=UseColor_Specified, specifiedColor=<rgba 1 1 1 1>}
     uVRegion {bIsValid=false, max=<0 0>, min=<0 0>}
    }
    loadingBackgroundBrush {
     bIsDynamicallyLoaded=false
     drawAs=Image
     imageType=FullColor
     mirroring=NoMirror
     resourceName=None
     resourceObject=null
     tiling=NoTile
     imageSize=<1 1>
     margin=<ltrb 0 0 0 0>
     outlineSettings {
      bUseBrushTransparency=false
      roundingType=HalfHeightRadius
      width=0
      color {colorUseRule=UseColor_Specified, specifiedColor=<rgba 0 0 0 0>}
      cornerRadii=<0 0 0 0>
     }
     tintColor {colorUseRule=UseColor_Specified, specifiedColor=<rgba 1 1 1 1>}
     uVRegion {bIsValid=false, max=<0 0>, min=<0 0>}
    }
   }
   slot {class=OverlaySlot, properties {horizontalAlignment=HAlign_Center, verticalAlignment=VAlign_Center}}
  }
  SafeZone {widgetClassPath=&SafeZone, bIsVariable=true, namedSlotHost=WBP_UIKit_Dialog_Internal, widget=SafeZone}
  Footer {widgetClassPath=&Overlay, bIsVariable=true, parent=SafeZone, widget=Footer, slot {class=SafeZoneSlot}}
  LegalText {
   widgetClassPath=&CommonTextBlock
   bIsVariable=true
   parent=Footer
   widget=LegalText
   properties {
    autoWrapText=true
    bIsScrollingEnabled=false
    mobileFontSizeMultiplier=1.4
    font {
     bForceMonospaced=false
     bMaterialIsStencil=false
     fontMaterial=null
     fontObject=&HeadingNow
     letterSpacing=10
     monospacedWidth=1
     size=11.25
     skewAmount=0
     typefaceFontName=74 Regular
     outlineSettings {
      bApplyOutlineToDropShadows=false
      bMiteredCorners=false
      bSeparateFillAlpha=false
      outlineMaterial=null
      outlineSize=0
      outlineColor=<rgba 0 0 0 1>
     }
    }
    shadowOffset=<0 0>
   }
   slot {class=OverlaySlot, properties {horizontalAlignment=HAlign_Fill, verticalAlignment=VAlign_Bottom, padding=<ltrb 40 40 40 40>}}
  }
 ]
}
