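# Activatable Season Pass screen for buying MSEE rewards with V-Bucks.
# Add/subtract buttons clamp the reward count, highlight reward track cards, update price with strikethrough and play pitch-stepped audio.
# Confirm purchases level and bundle offers asynchronously, first showing a dialog when right-of-withdrawal applies; its label switches to Get V-Bucks on insufficient balance.
# Theme data sets title, panel image and an async-loaded background, applied only if still requested and active.
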WidgetBlueprint /SeasonPass/MSEE/WBP_MSEE_BuyRewards
graphs [
 AudioIncrementDecrementCounter {nodes [
   n1 = AudioIncrementDecrementCounter!() -> (then, bIsIncrement:bool=~) {is=title K2Node_FunctionEntry}
   > n2 = Branch(Condition=@n1)
   > n3 = Increment Int(Value=Get AudioLevelsCounter() -> (AudioLevelsCounter:int=0))
   n4 = SelectInt(A=1, B=@n3, bPickA="Equal (Integer)"(A=@n3))
   n5 = Decrement Int(" "=@n2.else, Value=Get AudioLevelsCounter() -> (AudioLevelsCounter=0))
   n6 = SelectInt(A=-1, B=@n5, bPickA="Equal (Integer)"(A=@n5))
   n7 = Get bIsIncrement() -> (bIsIncrement:bool)
   Set AudioLevelsCounter(
    execute=[@n3."  ", @n5."  "]
    AudioLevelsCounter:int="Clamp (Integer)"(Value=SelectInt(A=@n4, B=@n6, bPickA=@n7), Min=-5, Max=5) -> (ReturnValue=0)
   ) -> (Output_Get:int=0)
   > n8 = SpawnSound2D!(
    execute
    Sound:Sound Base Object Reference=/Game/Sounds/Fort_UI_Sounds/FrontEnd/DefaultSounds/PressedButtons/MSS_UI_Button_Pressed_RewardAddSubtract
    VolumeMultiplier:float=1
    PitchMultiplier:float=1
    StartTime:float=0
    ConcurrencySettings:Sound Concurrency Object Reference
    bPersistAcrossLevelTransition:bool=false
    bAutoDestroy:bool=true
   ) -> (then, ReturnValue:Audio Component Object Reference)
   > Set Integer Parameter!(
    execute
    self:Audio Parameter Controller Interface Interface=@n8
    InName:Name=Pitch
    InInt:int=Get AudioLevelsCounter() -> (AudioLevelsCounter=0)
   ) -> (then) {is=title K2Node_CallFunction}
   > Execute Trigger Parameter!(execute, self:Audio Parameter Controller Interface Interface=@n8, InName:Name=Play) -> (then)
  ]}
 BP_GetDesiredFocusTarget {nodes [
   Get Desired Focus Target()
   n1 = Get BuyLevels_Button() -> (BuyLevels_Button:WBP UIKit Button Loud Object Reference)
   > Return Node!(
    execute
    ReturnValue:Widget Object Reference=Select(
     Option 0:Widget Object Reference=Get Cancel_Button() -> (Cancel_Button:WBP UIKit Button Regular Object Reference)
     Option 1:Widget Object Reference=@n1
     Index=Get bIsEnabled!(self:Widget Object Reference=@n1) -> (bIsEnabled:bool)
    ) -> (ReturnValue:Widget Object Reference)
   )
  ]}
 BlockInput {nodes [
   n1 = BlockInput!() -> (then, ShouldBlockInput:bool=~) {is=title K2Node_FunctionEntry}
   > SetVisibility(
    self=Get RewardTrack_Slot() -> (RewardTrack_Slot:Named Slot Object Reference)
    InVisibility=Select(Option 0:ESlateVisibility Enum=SelfHitTestInvisible, Option 1:ESlateVisibility Enum=HitTestInvisible, Index=@n1) -> (ReturnValue:ESlateVisibility Enum)
   )
  ]}
 EventGraph {nodes [
   n1 = "On Clicked (Button_Close)"!() -> (OutputDelegate, then, Button:Common Button Base Object Reference) {is=title K2Node_ComponentBoundEvent}
   > n2 = Fire Event Element Interacted(Element=@n1.Button)
   DeactivateWidget(execute=[@n12.then, @n2.then])
   n3 = "On Clicked (BuyLevels_Button)"!() -> (OutputDelegate, then, Button:Common Button Base Object Reference) {is=title K2Node_ComponentBoundEvent}
   > Fire Event Element Interacted(Element=@n3.Button)
   > n4 = ShouldShowRightOfWithdrawalModal!(execute, Context:Widget Object Reference) -> (then, ReturnValue:bool)
   > n5 = Branch(Condition=@n4)
   > n6 = DeactivateWidget()
   n7 = OnCpcDismissed!() -> (OutputDelegate, then, Accepted:bool=~) {is=title K2Node_CustomEvent}
   n8 = Self-Reference()
   OnCpcDialogRequested!(
    execute=@n6.then
    Outer:Object Reference=@n8
    BodyText:Text="LOCTABLE(\"/SeasonPass/Data/ST_Passes\", \"CPC_BattlePassLevels\")"
    OnCpcDialogDismissed:"Delegate (by ref)"=@n7
    Target:Widget Object Reference=@n8
    Context:User Widget Object Reference=@n8
   ) -> (then)
   n9 = PurchaseMultipleSeasonPassOffers!(
    execute:Exec=[@n21.then, @n5.else]
    SeasonPass:Season Pass VM Object Reference=Get SeasonPassVM() -> (SeasonPassVM:Season Pass VM Object Reference)
    OffersWithQuantities:Map of Fort Simple Offer VM Object References to Integers=MakeOfferMap!() -> (Map:Map of Fort Simple Offer VM Object References to Integers)
   ) -> (then:Exec, OnOperationCompleted:Exec, Success:bool=~) {is=title K2Node_AsyncAction}
   > ShowProgressModal!(execute, Title:Text="LOCTEXT(\"Completing Purchase\")", Description:Text) -> (then)
   HideProgressModal!(execute=@n9.OnOperationCompleted) -> (then)
   > UpdateSelectedRewards!(execute, self, IsInit:bool=false) -> (then)
   > Branch(Condition=@n9)
   > n10 = Branch(Condition=@n4)
   DeactivateWidget(execute=@n10.else)
   n11 = "On Clicked (Cancel_Button)"!() -> (OutputDelegate, then, Button:Common Button Base Object Reference) {is=title K2Node_ComponentBoundEvent}
   > n12 = Fire Event Element Interacted(Element=@n11.Button)
   n13 = "On Clicked (Button_Add_1)"!() -> (OutputDelegate, then, Button:Common Button Base Object Reference) {is=title K2Node_ComponentBoundEvent}
   > Fire Event Element Interacted(Element=@n13.Button)
   > SetRewardsToBuy!(execute, self, AdditiveAmount:int=1) -> (then)
   > AudioIncrementDecrementCounter!(execute, self, bIsIncrement:bool=true) -> (then) {is=title K2Node_CallFunction}
   > UpdateSelectedRewards(IsInit=false)
   n14 = "On Clicked (Button_Subtract_1)"!() -> (OutputDelegate, then, Button:Common Button Base Object Reference) {is=title K2Node_ComponentBoundEvent}
   > Fire Event Element Interacted(Element=@n14.Button)
   > SetRewardsToBuy(AdditiveAmount=-1)
   > AudioIncrementDecrementCounter(bIsIncrement=false)
   > UpdateSelectedRewards(IsInit=false)
   n15 = RequestLoadBackground!() -> (OutputDelegate, then, BackgroundMaterial:Soft Object Reference, Direction:EVaultWorldTransitionDirection Enum=Forward) {is=title K2Node_CustomEvent}
   > Set RequestedBackgroundSoftPath(RequestedBackgroundSoftPath:Soft Object Path Structure=To Soft Object Path!(SoftObjectReference:Soft Object Reference=@n15.BackgroundMaterial) -> (ReturnValue:Soft Object Path Structure)) -> (Output_Get:Soft Object Path Structure)
   > n16 = Async Load Asset(Asset=@n15.BackgroundMaterial)
   n17 = Get RequestedBackgroundSoftPath() -> (RequestedBackgroundSoftPath:Soft Object Path Structure)
   n18 = GetSoftObjectPath(Object=@n16)
   Branch(
    execute=@n16.Completed
    Condition="Equal Exactly (String)"(A=BreakSoftObjectPath(InSoftObjectPath=@n18), B=BreakSoftObjectPath(InSoftObjectPath=@n17))
   )
   > Branch(Condition=Get bIsActive() -> (bIsActive:bool))
   > UpdateBackground!(
    execute
    self:Fort Item Preview Background Object Reference=Get FortItemPreviewBackground() -> (FortItemPreviewBackground:Fort Item Preview Background Object Reference)
    LoadedBackground:Object Reference=@n16
    TransitionDirection:EVaultWorldTransitionDirection Enum=@n15.Direction
    PrimaryColor:Linear Color Structure=(R=1,G=1,B=1,A=0)
    SecondaryColor:Linear Color Structure=(R=0,G=0,B=0,A=0)
   ) -> (then)
   n19 = Event PreConstruct()
   > SetCurrency!(
    execute
    self:WBP Price Switcher Base Object Reference=Get WBP_PriceSwitcher_Base() -> (WBP_PriceSwitcher_Base:WBP Price Switcher Base Object Reference)
    Currency:EFortPriceCurrency Enum=Mtx
   ) -> (then)
   n20 = Get Switcher_BottomInfo!(self:WBP Price Switcher Base Object Reference=Get WBP_PriceSwitcher_Base()) -> (Switcher_BottomInfo:Common Visibility Switcher Object Reference)
   > SetVerticalAlignment(
    self:Size Box Slot Object Reference=SlotAsSizeBoxSlot!(Widget:Widget Object Reference=GetParent!(self:Widget Object Reference=@n20) -> (ReturnValue:Panel Widget Object Reference)) -> (ReturnValue:Size Box Slot Object Reference)
    InVerticalAlignment=VAlign_Center
   )
   > SetRewardTrack!(execute, self, RewardTrack:Dynamic Entry Box Object Reference=Get RewardTrack() -> (RewardTrack:Dynamic Entry Box Object Reference)) -> (then)
   > SetVisibility(self=Get BuySelectorBox() -> (BuySelectorBox:Horizontal Box Object Reference), InVisibility=Hidden)
   > Branch(Condition=@n19.IsDesignTime)
   > "SetText (Text)"(self=Get LegalText() -> (LegalText:Common Text Object Reference), InText="INVTEXT(\"[PH] Legal Text\")")
   Event On Activated()
   > Fire Event Sub View Opened!(execute, View:Widget Object Reference=Self-Reference(), AdditionalInfo:Flow Analytics Additional Info Structure, bIsOverlay:bool=false) -> (then)
   > SetIsEnabled(self=Get BuyLevels_Button(), bInIsEnabled=false)
   > BlockInput!(execute, self, ShouldBlockInput:bool=true) -> (then) {is=title K2Node_CallFunction}
   > Set RewardsToBuy(RewardsToBuy:int=0) -> (Output_Get:int=0)
   > UpdateSelectedRewards(IsInit=true)
   > SetRewardsToBuy(AdditiveAmount=Get Rewards Left to Buy() -> (Rewards Left to Buy:int=0))
   > PlayIntroAnimation()
   PlayIntroAnimation() {is=title K2Node_CustomEvent}
   n21 = Branch(execute=@n7.then, Condition=@n7.Accepted)
  ]}
 GetFrontendFlowInterruptConfig {nodes [
   GetFrontendFlowInterruptConfig() {is=title K2Node_FunctionEntry}
   > Return Node(ReturnValue:Frontend Flow Interrup Config Structure)
  ]}
 MakeOfferMap {nodes [
   MakeOfferMap() {is=title K2Node_FunctionEntry}
   > n1 = Get!(execute:Exec) -> (then:Exec, else:Exec, SeasonPassVM:Season Pass VM Object Reference)
   n2 = GetLevelsToGrant!(self:Season Pass Bundle Offer VM Object Reference=Get LevelBundleOffer!(self:Season Pass VM Object Reference=@n1) -> (LevelBundleOffer:Season Pass Bundle Offer VM Object Reference)) -> (ReturnValue:int=0)
   n3 = Get LevelsToBuy() -> (LevelsToBuy:int=0)
   n4 = Get SeasonPassVM()
   n5 = Get SeasonPassVM()
   > Return Node!(
    execute
    Map:Map of Fort Simple Offer VM Object References to Integers=Make Map!(
     Key 0:Fort Simple Offer VM Object Reference=Get LevelOffer!(self:Season Pass VM Object Reference=@n4) -> (LevelOffer:Fort Simple Offer VM Object Reference)
     Value 0:int=int - int(A=@n3, B=@n2)
     Key 1:Fort Simple Offer VM Object Reference=Get LevelBundleOffer(self=@n5)
     Value 1:int=1
    ) -> (Map:Map of Fort Simple Offer VM Object References to Integers) {is=title K2Node_MakeMap}
   )
  ]}
 SetRewardTrack {nodes [
   n1 = SetRewardTrack!() -> (then, RewardTrack:Dynamic Entry Box Object Reference) {is=title K2Node_FunctionEntry}
   > n2 = Set RewardTrack(RewardTrack:Dynamic Entry Box Object Reference=@n1) -> (Output_Get:Dynamic Entry Box Object Reference)
   > n3 = Set RewardTrackWidgets(RewardTrackWidgets:Array of Widget Object References=GetAllEntries!(self:Dynamic Entry Box Base Object Reference=@n2) -> (ReturnValue:Array of User Widget Object References)) -> (Output_Get:Array of Widget Object References)
   > n4 = Cast To WBP_MSEE_Itemcard!(execute, Object="Get (a copy)"(Array:Array of Widget Object References=@n3) -> (Output:Widget Object Reference)) -> (then, CastFailed, AsWBP MSEE Itemcard:WBP MSEE Itemcard Object Reference) {is=title K2Node_DynamicCast}
   > Set PreviewItemVM(PreviewItemVM:Fort Item VM Object Reference=Get CurrentItemVM!(self:WBP MSEE Itemcard Object Reference=@n4) -> (CurrentItemVM:Fort Item VM Object Reference)) -> (Output_Get:Fort Item VM Object Reference)
   > Set PreviewedSeasonRewardVM(PreviewedSeasonRewardVM:Season Pass Reward VM Object Reference=Get SeasonPassRewardVM!(self:WBP MSEE Itemcard Object Reference=@n4) -> (SeasonPassRewardVM:Season Pass Reward VM Object Reference)) -> (Output_Get:Season Pass Reward VM Object Reference)
  ]}
 SetRewardsToBuy {nodes [
   SetRewardsToBuy!() -> (then, AdditiveAmount:int) {is=title K2Node_FunctionEntry}
   > n1 = Sequence()
   SetIsEnabled(
    execute=@n1.then_1
    self=Get Button_Add_1() -> (Button_Add_1:WBP UIKit Button Regular Object Reference)
    bInIsEnabled="integer < integer"(A=Get RewardsToBuy() -> (RewardsToBuy:int=0), B=Get AvailableRewardsToBuy() -> (AvailableRewardsToBuy:int=0))
   )
   n2 = Get RewardsToBuy() -> (RewardsToBuy=0)
   > SetIsEnabled(
    self=Get Button_Subtract_1() -> (Button_Subtract_1:WBP UIKit Button Regular Object Reference)
    bInIsEnabled="integer >= integer"(A=int - int(A=@n2))
   )
   > UpdateConfirmButtonAction()
   n3 = Get AdditiveAmount() -> (AdditiveAmount:int=0)
   n4 = Get RewardsToBuy() -> (RewardsToBuy=0)
   n5 = Set RewardsToBuy(
    execute=@n1.then_0
    RewardsToBuy="Clamp (Integer)"(Value=int + int(A=@n4, B=@n3), Min=0, Max=Get AvailableRewardsToBuy() -> (AvailableRewardsToBuy=0)) -> (ReturnValue=0)
   ) -> (Output_Get=0)
   > SetText(
    self=Get Rewards_Text() -> (Rewards_Text:Common Rich Text Block Object Reference)
    InText=Format Text!(Format="LOCTEXT(\"{Rewards} {Rewards}|plural(one=Reward,other=Rewards)\")", Rewards:int=@n5) -> (Result)
   )
   n6 = "integer <= integer"(A=Get RewardsToBuy() -> (RewardsToBuy=0))
   > SetIsEnabled(self=Get BuyLevels_Button(), bInIsEnabled=NOT Boolean(A=@n6))
   > UpdateFinalPrice()
   > Branch(Condition=@n6)
   > RequestRefreshFocus(self:Common Activatable Widget Object Reference)
  ]}
 UpdateConfirmButtonAction {nodes [
   UpdateConfirmButtonAction() {is=title K2Node_FunctionEntry}
   n1 = "integer >= integer"(
    A=GetMtxBalance!(self:Fort Local Player VM Object Reference=Get FortLocalPlayerVM() -> (FortLocalPlayerVM:Fort Local Player VM Object Reference)) -> (ReturnValue:int=0)
    B=Get TotalCost() -> (TotalCost:int=0)
   )
   > SetText!(
    execute
    self:Fort CTAButton Object Reference=Get BuyLevels_Button()
    ButtonText:"Text (by ref)"=SelectText(A="LOCTEXT(\"Buy Rewards\")", B="LOCTEXT(\"Get V-Bucks\")", bPickA=@n1)
   ) -> (then)
   > n2 = ShouldShowRightOfWithdrawalModal()
   > SetRequiresHold!(execute, self:Common Button Base Object Reference=Get BuyLevels_Button(), bInRequiresHold:bool=AND Boolean(A=NOT Boolean(A=@n2), B=@n1)) -> (then)
  ]}
 UpdateFinalPrice {nodes [
   UpdateFinalPrice() {is=title K2Node_FunctionEntry}
   n1 = Get SeasonPassCategoryVM() -> (SeasonPassCategoryVM:Season Pass Category VM Object Reference)
   > Set LevelsToBuy(LevelsToBuy:int="Min (Integer)"(
     A=GetNumberOfIncompleteQuestEntries!(self:Season Pass Category VM Object Reference=@n1) -> (ReturnValue:int=0)
     B=Get RewardsToBuy() -> (RewardsToBuy=0)
    ) -> (ReturnValue=0)) -> (Output_Get:int=0)
   n2 = Get SeasonPassVM()
   > Set Regular Price(Regular Price:int=GetRegularPriceIntForQuantity!(self:Fort Simple Offer VM Object Reference=Get LevelOffer(self=@n2), Quantity:int=Get RewardsToBuy() -> (RewardsToBuy=0)) -> (ReturnValue:int=0)) -> (Output_Get:int=0)
   > Set Levels To Count For Final Price(Levels To Count For Final Price:int=Get LevelsToBuy() -> (LevelsToBuy=0)) -> (Output_Get:int=0)
   n3 = Get LevelOffer(self=Get SeasonPassVM())
   n4 = Get Levels To Count For Final Price() -> (Levels To Count For Final Price:int=0)
   > n5 = Set TotalCost(TotalCost:int=int + int(
     A=Get Bundle Price() -> (Bundle Price:int=0)
     B=GetPriceIntForQuantity!(self:Fort Simple Offer VM Object Reference=@n3, Quantity:int=@n4) -> (ReturnValue:int=0)
    )) -> (Output_Get:int=0)
   > SetPrice!(execute, self:WBP Price Switcher Base Object Reference=Get WBP_PriceSwitcher_Base(), Price:Text="To Text (Integer)"(Value=@n5)) -> (then)
   > SetRegularPrice!(
    execute
    self:WBP Price Switcher Base Object Reference=Get WBP_PriceSwitcher_Base()
    Price:Text="To Text (Integer)"(Value=Get Regular Price() -> (Regular Price:int=0))
   ) -> (then)
   n6 = Get Regular Price() -> (Regular Price=0)
   n7 = Get TotalCost() -> (TotalCost=0)
   > SetStrikethroughPriceVisibility!(
    execute
    self:WBP Price Switcher Base Object Reference=Get WBP_PriceSwitcher_Base()
    Visibility:ESlateVisibility Enum=Select(Option 0=SelfHitTestInvisible, Option 1=Collapsed, Index="Equal (Integer)"(A=@n6, B=@n7))
   ) -> (then)
  ]}
 UpdateRewardCards {nodes [
   n1 = UpdateRewardCards!() -> (then, Rewards:Array of Season Pass Reward VM Object References) {is=title K2Node_FunctionEntry}
   n2 = Get SeasonPassCategoryVM()
   > Set AvailableRewardsToBuy(AvailableRewardsToBuy:int="Min (Integer)"(A=Length(TargetArray:Array of Season Pass Reward VM Object References=@n1), B=GetNumberOfIncompleteQuestEntries(self=@n2) -> (ReturnValue=0)) -> (ReturnValue=0)) -> (Output_Get:int=0)
   > SetRewardsToBuy(AdditiveAmount=0) {is=title K2Node_CallFunction}
  ]}
 UpdateSelectedRewards {nodes [
   UpdateSelectedRewards!() -> (then, IsInit:bool=~) {is=title K2Node_FunctionEntry}
   > SetSize!(execute, self:Spacer Object Reference=Get OwnedSpacer() -> (OwnedSpacer:Spacer Object Reference), InSize:Vector 2D Structure) -> (then)
   > SetWidthOverride(self=Get BuySize() -> (BuySize:Size Box Object Reference), InWidthOverride=0)
   > SetVisibility(self=Get BuySelectorBox(), InVisibility=Hidden)
   > Set RewardsLeftToBuy(RewardsLeftToBuy:int=Get RewardsToBuy() -> (RewardsToBuy=0)) -> (Output_Get:int=0)
   n1 = GetAllEntries(self=Get RewardTrack())
   > Branch(Condition=Is Not Empty!(TargetArray:Array of User Widget Object References=@n1) -> (ReturnValue:bool))
   > n2 = Set Rewards(Rewards:Array of User Widget Object References=@n1) -> (Output_Get:Array of User Widget Object References)
   > n3 = For Each Loop(Array:Array of User Widget Object References=@n2) -> (Array Element:User Widget Object Reference)
   > Set RewardIndex(RewardIndex:int=@n3.Array Index) -> (Output_Get:int=0)
   > n4 = Cast To WBP_MSEE_Itemcard(Object=@n3)
   > n5 = Set ItemCard(ItemCard:WBP MSEE Itemcard Object Reference=@n4) -> (Output_Get:WBP MSEE Itemcard Object Reference)
   > ShowBuyOutline!(execute, self:WBP MSEE Itemcard Object Reference=@n5, ShowOutline:bool=false) -> (then)
   > Set OverriddenPercent!(execute, OverriddenPercent:double=-1, self:WBP MSEE Itemcard Object Reference=@n5) -> (then, Output_Get:double)
   > SetProgressAmount!(
    execute
    self:WBP MSEE Itemcard Object Reference=@n5
    Progress:double=Get ProgressPercent!(self:WBP MSEE Itemcard Object Reference=@n5) -> (ProgressPercent:double)
   ) -> (then)
   > n6 = Branch(Condition=Get bIsRewardOwned!(self:Season Pass Reward VM Object Reference=Get SeasonPassRewardVM(self=@n5)) -> (bIsRewardOwned:bool))
   n7 = Get!(self:Spacer Object Reference=Get OwnedSpacer()) -> (Size_X:double, Size_Y:double)
   n8 = Get!(self:WBP UIKit Item Card Base Object Reference=Get ItemCard() -> (ItemCard:WBP MSEE Itemcard Object Reference)) -> (ItemCardSize_X:double, ItemCardSize_Y:double)
   n9 = Get RewardTrack()
   > SetSize!(
    execute
    self=Get OwnedSpacer()
    InSize_X:double=float + float(A=float + float(A=@n7, B=@n8), B=Get!(self:Dynamic Entry Box Base Object Reference=@n9) -> (EntrySpacing_X:double, EntrySpacing_Y:double))
    InSize_Y:double=0
   ) -> (then)
   > Set LastOwnedRewardIndex(LastOwnedRewardIndex:int=Get RewardIndex() -> (RewardIndex:int=0)) -> (Output_Get:int=0)
   n10 = Branch(execute=@n6.else, Condition=Get IsInit() -> (IsInit:bool))
   > n11 = Get!(execute:Exec) -> (then:Exec, else:Exec, SeasonPassSelectionVM:Season Pass Selection VM Object Reference)
   > n12 = Get!(execute, self:Season Pass Selection VM Object Reference=@n11) -> (then, else, SelectedPage:Season Pass Page VM Object Reference)
   n13 = Get Tracks!(self:Season Pass Page VM Object Reference=@n12) -> (Tracks:Array of Season Pass Track VM Object References)
   > Branch(Condition=Is Not Empty(TargetArray:Array of Season Pass Track VM Object References=@n13))
   n14 = Get Rewards!(self:Season Pass Track VM Object Reference="Get (a copy)"(Array:Array of Season Pass Track VM Object References=@n13) -> (Output:Season Pass Track VM Object Reference)) -> (Rewards:Array of Season Pass Reward VM Object References)
   > Branch(Condition=Is Not Empty(TargetArray:Array of Season Pass Reward VM Object References=@n14))
   n15 = Find Item!(
    TargetArray:Array of Season Pass Reward VM Object References=@n14
    ItemToFind:"Season Pass Reward VM Object Reference (by ref)"=Get SelectedReward!(self:Season Pass Selection VM Object Reference=Get SeasonPassSelectionVM() -> (SeasonPassSelectionVM:Season Pass Selection VM Object Reference)) -> (SelectedReward:Season Pass Reward VM Object Reference)
   ) -> (ReturnValue)
   n16 = Get RewardIndex() -> (RewardIndex=0)
   n17 = int + int(A=Get LastOwnedRewardIndex() -> (LastOwnedRewardIndex:int=0))
   n18 = Get RewardIndex() -> (RewardIndex=0)
   > Branch(Condition=OR Boolean(A="integer <= integer"(A=@n16, B=@n15), B="Equal (Integer)"(A=@n18, B=@n17)))
   n19 = Get RewardsLeftToBuy() -> (RewardsLeftToBuy:int=0)
   > Increment Int(Value=@n19)
   > n20 = Set Rewards Left to Buy(Rewards Left to Buy:int=@n19) -> (Output_Get:int=0)
   n21 = "integer > integer"(A=Get RewardsLeftToBuy() -> (RewardsLeftToBuy=0))
   ShowBuyOutline(execute=[@n20.then, @n10.else], self=Get ItemCard(), ShowOutline=@n21)
   n22 = Get ItemCard()
   > Set OverriddenPercent(OverriddenPercent=SelectFloat(A=1, B=0, bPickA=@n21), self=@n22)
   > SetProgressAmount(self=@n22, Progress=0)
   > Branch(Condition=@n21)
   > SetVisibility(self=Get BuySelectorBox(), InVisibility=HitTestInvisible)
   n23 = Get FirstSelectedItemPassed() -> (FirstSelectedItemPassed:bool)
   n24 = Get!(self:Dynamic Entry Box Base Object Reference=Get RewardTrack()) -> (EntrySpacing_X:double, EntrySpacing_Y:double)
   n25 = Get BuySize()
   n26 = Get WidthOverride!(self:Size Box Object Reference=@n25) -> (WidthOverride:float)
   n27 = Get!(self:WBP UIKit Item Card Base Object Reference=Get ItemCard()) -> (ItemCardSize_X:double, ItemCardSize_Y:double)
   > SetWidthOverride(self=@n25, InWidthOverride=float + float(A=SelectFloat(A=@n24, B=0, bPickA=@n23), B=float + float(A=@n27, B=@n26)))
   > n28 = Branch(Condition=Get IsInit())
   Set FirstSelectedItemPassed(execute=[@n28.then, @n29."  "], FirstSelectedItemPassed:bool=true) -> (Output_Get:bool)
   n29 = Decrement Int(" "=@n28.else, Value=Get RewardsLeftToBuy() -> (RewardsLeftToBuy=0))
  ]}
 VM_GetThemeData {nodes [
   n1 = VM_GetThemeData!() -> (then, Theme data:Season Pass Theme Data Structure) {is=title K2Node_FunctionEntry}
   n2 = Break Season Pass Theme Data!(SeasonPassThemeData:"Season Pass Theme Data Structure (by ref)"=@n1) -> (
    TitleColor:Linear Color Structure
    TitleMaterial:Soft Object Reference
    TitleOutline:Font Outline Settings Structure
    SeasonColor:Linear Color Structure
    ProgressionColor:Linear Color Structure
    OverlayColor:Linear Color Structure
    VaultBackgroundMaterial:Soft Object Reference
    GiftboxVaultBackgroundMaterial:Soft Object Reference
    UIBackgroundMaterial:Soft Object Reference
    UpsellBackgroundMaterial:Soft Object Reference
    CrewCategoryBackgroundMaterial:Soft Object Reference
    CustomLogo:Soft Object Reference
    CustomSeasonTitle:Text
    CustomSeasonTitleColor:Linear Color Structure
    Upsell2BackgrounMaterial:Soft Object Reference
    ImageReferenceTable:Fort Image Reference Table Structure
    ColorReferenceTable:Fort Color Reference Table Structure
    TextReferenceTable:Fort Text Reference Table Structure
   ) {is=title K2Node_BreakStruct}
   > RequestLoadBackground!(execute, self, BackgroundMaterial:Soft Object Reference=@n2.VaultBackgroundMaterial, Direction:EVaultWorldTransitionDirection Enum=None) -> (then) {is=title K2Node_CallFunction}
   > SetText!(
    execute
    self:Rich Text Block Object Reference=Get RichText_Title() -> (RichText_Title:Common Rich Text Block Object Reference)
    InText:"Text (by ref)"=TryLookupText!(Table:"Fort Text Reference Table Structure (by ref)"=@n2.TextReferenceTable, Name:String=BuyMenu_Header) -> (OutTextReference:Text, ReturnValue:bool)
   ) -> (then)
   n3 = TryLookupImage!(Table:"Fort Image Reference Table Structure (by ref)"=@n2.ImageReferenceTable, Name:String=BuyRewards_PanelBackground) -> (OutImageReference:Fort Image Reference Structure, ReturnValue:bool)
   > n4 = Branch(Condition=@n3.ReturnValue)
   > SetBrushFromLazyDisplayAsset!(
    execute
    self:Common Lazy Image Object Reference=Get Levels_Panel_Background() -> (Levels_Panel_Background:Fort Lazy Image Object Reference)
    LazyObject:"Soft Object Reference (by ref)"=Break Fort Image Reference!(FortImageReference:"Fort Image Reference Structure (by ref)"=@n3) -> (Image:Soft Object Reference, ImageUrl:String, MaterialTextureParamName:Name) {is=title K2Node_BreakStruct}
    bMatchTextureSize:bool=false
   ) -> (then)
   > Set Visibility(Visibility=HitTestInvisible, self=Get Levels_Panel_Background())
   Set Visibility(execute=@n4.else, Visibility=Collapsed, self=Get Levels_Panel_Background())
  ]}
]
variables [
 AudioLevelsCounter {type=integer}
 AvailableRewardsToBuy {type=integer}
 FromInspectScreen {type=boolean}
 LevelsToBuy {type=integer}
 PreviewedSeasonRewardVM {type=SeasonPassRewardVM, default=null}
 PreviewItemVM {type=FortItemVM, default=null}
 RequestedBackgroundSoftPath {type=SoftObjectPath, default=""}
 Rewards Left to Buy {type=integer}
 RewardsToBuy {type=integer}
 RewardTrack {type=DynamicEntryBox, default=null}
 RewardTrackWidgets {type="array<Widget>", default []}
 TotalCost {type=integer}
]
widgets {
 info {
  inheritedWidgetCount=0
  namedSlotCount=1
  parentClass=&CommonActivatableWidget
  rootWidgetClass=&Overlay
  widgetCount=39
 }
 widgets [
  Overlay_69 {widgetClassPath=&Overlay, widget=Overlay_69}
  WBP_CaptureForPostBufferUpdate {
   widgetClassPath=&WBP_CaptureForPostBufferUpdate
   bIsVariable=true
   parent=Overlay_69
   slot {class=OverlaySlot, properties {horizontalAlignment=HAlign_Fill, verticalAlignment=VAlign_Fill}}
  }
  FortItemPreviewBackground {
   widgetClassPath=/Script/FortniteUI.FortItemPreviewBackground
   bIsVariable=true
   parent=Overlay_69
   widget=FortItemPreviewBackground
   properties {previewSceneChangerTag {tagName=UI.ItemPreview.SceneChanger.MPItemShop}}
   slot {class=OverlaySlot, properties {horizontalAlignment=HAlign_Fill, verticalAlignment=VAlign_Fill}}
  }
  Background_Custom {
   widgetClassPath=&NamedSlot
   bIsVariable=true
   parent=Overlay_69
   widget=Background_Custom
   slot {class=OverlaySlot, properties {horizontalAlignment=HAlign_Fill, verticalAlignment=VAlign_Fill}}
  }
  SafeZone {
   widgetClassPath=&SafeZone
   bIsVariable=true
   parent=Overlay_69
   widget=SafeZone
   slot {class=OverlaySlot, properties {horizontalAlignment=HAlign_Fill, verticalAlignment=VAlign_Fill}}
  }
  SizeBox_2 {
   widgetClassPath=&SizeBox
   parent=SafeZone
   widget=SizeBox_2
   properties {bOverride_WidthOverride=true, widthOverride=1920}
   slot {class=SafeZoneSlot}
  }
  GridPanel_0 {
   widgetClassPath=&GridPanel
   parent=SizeBox_2
   widget=GridPanel_0
   properties {columnFill [0.8,1,0.8], rowFill [0,1,0]}
   slot {class=SizeBoxSlot, properties {padding=<ltrb 80 40 80 40>}}
  }
  WBP_PlayerInfo {
   widgetClassPath=/Game/UI/Frontend/ItemShop/MultiProduct/WBP_PlayerInfo
   bIsVariable=true
   parent=GridPanel_0
   widget=WBP_PlayerInfo
   slot {class=GridSlot, properties {column=2, horizontalAlignment=HAlign_Right, verticalAlignment=VAlign_Top}}
  }
  MainContent {
   widgetClassPath=&VerticalBox
   parent=GridPanel_0
   widget=MainContent
   slot {class=GridSlot, properties {columnSpan=3, rowSpan=2, verticalAlignment=VAlign_Center}}
  }
  RichText_Title {
   widgetClassPath=&CommonRichTextBlock
   bIsVariable=true
   parent=MainContent
   widget=RichText_Title
   properties {
    bOverrideDefaultStyle=true
    mobileTextBlockScale=0.8
    text=Buy Rewards
    textTransformPolicy=ToUpper
    defaultTextStyleOverride {
     overflowPolicy=Clip
     transformPolicy=None
     colorAndOpacity {colorUseRule=UseColor_Specified, specifiedColor=<rgba 1 1 1 1>}
     font {
      bForceMonospaced=false
      bMaterialIsStencil=false
      fontMaterial=null
      fontObject=&HeadingNow
      letterSpacing=0
      monospacedWidth=1
      size=42
      skewAmount=0
      typefaceFontName=86 Bold Italic
      outlineSettings {
       bApplyOutlineToDropShadows=false
       bMiteredCorners=false
       bSeparateFillAlpha=false
       outlineMaterial=null
       outlineSize=0
       outlineColor=<rgba 0 0 0 1>
      }
     }
     highlightColor {colorUseRule=UseColor_Specified, specifiedColor=<rgba 0 0 0 1>}
     highlightShape {
      bIsDynamicallyLoaded=false
      drawAs=Image
      imageType=NoImage
      mirroring=NoMirror
      resourceName=None
      resourceObject=null
      tiling=NoTile
      imageSize=<32 32>
      margin=<ltrb 0 0 0 0>
      outlineSettings {
       bUseBrushTransparency=false
       roundingType=HalfHeightRadius
       width=0
       color {colorUseRule=UseColor_Specified, specifiedColor=<rgba 0 0 0 0>}
       cornerRadii=<0 0 0 0>
      }
      tintColor {colorUseRule=UseColor_Specified, specifiedColor=<rgba 1 1 1 1>}
      uVRegion {bIsValid=false, max=<0 0>, min=<0 0>}
     }
     selectedBackgroundColor {colorUseRule=UseColor_Foreground, specifiedColor=<rgba 1 0 1 1>}
     shadowColorAndOpacity=<rgba 0 0 0 1>
     shadowOffset=<0 0>
     strikeBrush {
      bIsDynamicallyLoaded=false
      drawAs=Image
      imageType=NoImage
      mirroring=NoMirror
      resourceName=None
      resourceObject=null
      tiling=NoTile
      imageSize=<32 32>
      margin=<ltrb 0 0 0 0>
      outlineSettings {
       bUseBrushTransparency=false
       roundingType=HalfHeightRadius
       width=0
       color {colorUseRule=UseColor_Specified, specifiedColor=<rgba 0 0 0 0>}
       cornerRadii=<0 0 0 0>
      }
      tintColor {colorUseRule=UseColor_Specified, specifiedColor=<rgba 1 1 1 1>}
      uVRegion {bIsValid=false, max=<0 0>, min=<0 0>}
     }
     underlineBrush {
      bIsDynamicallyLoaded=false
      drawAs=Image
      imageType=NoImage
      mirroring=NoMirror
      resourceName=None
      resourceObject=null
      tiling=NoTile
      imageSize=<32 32>
      margin=<ltrb 0 0 0 0>
      outlineSettings {
       bUseBrushTransparency=false
       roundingType=HalfHeightRadius
       width=0
       color {colorUseRule=UseColor_Specified, specifiedColor=<rgba 0 0 0 0>}
       cornerRadii=<0 0 0 0>
      }
      tintColor {colorUseRule=UseColor_Specified, specifiedColor=<rgba 1 1 1 1>}
      uVRegion {bIsValid=false, max=<0 0>, min=<0 0>}
     }
    }
   }
   slot {class=VerticalBoxSlot, properties {horizontalAlignment=HAlign_Center, verticalAlignment=VAlign_Top}}
  }
  Controllers {
   widgetClassPath=&HorizontalBox
   parent=MainContent
   widget=Controllers
   slot {class=VerticalBoxSlot, properties {horizontalAlignment=HAlign_Center, verticalAlignment=VAlign_Top, padding=<ltrb 0 0 0 48>}}
  }
  Button_Subtract_1 {
   widgetClassPath=&WBP_UIKit_Button_Regular
   bIsVariable=true
   parent=Controllers
   widget=Button_Subtract_1
   properties {
    bIsFocusable=false
    minHeight=48
    minWidth=112
    mobileMinHeight=88
    mobileMinWidth=88
    text=-1
    triggeringInputAction {dataTable=&InputActionDataTable, rowName=GenericLeftShoulder}
   }
   slot {class=HorizontalBoxSlot, properties {horizontalAlignment=HAlign_Center, verticalAlignment=VAlign_Center, padding=<ltrb 16 0 0 0>}}
  }
  SizeBox_0 {
   widgetClassPath=&SizeBox
   parent=Controllers
   widget=SizeBox_0
   properties {bOverride_MinDesiredWidth=true, minDesiredWidth=400, widthOverride=600}
   slot {class=HorizontalBoxSlot, properties {verticalAlignment=VAlign_Center, padding=<ltrb 48 0 48 0>, size {sizeRule=Fill, value=1}}}
  }
  Overlay_0 {widgetClassPath=&Overlay, parent=SizeBox_0, widget=Overlay_0, slot {class=SizeBoxSlot}}
  SizeBox_1 {
   widgetClassPath=&SizeBox
   parent=Overlay_0
   widget=SizeBox_1
   properties {bOverride_MinDesiredWidth=true, minDesiredWidth=400}
   slot {class=OverlaySlot, properties {horizontalAlignment=HAlign_Fill, verticalAlignment=VAlign_Fill}}
  }
  Levels_Panel_Background {
   widgetClassPath=&FortLazyImage
   bIsVariable=true
   parent=SizeBox_1
   widget=Levels_Panel_Background
   properties {
    brush {
     bIsDynamicallyLoaded=false
     drawAs=Image
     imageType=FullColor
     mirroring=NoMirror
     resourceName=None
     resourceObject=null
     tiling=NoTile
     imageSize=<32 32>
     margin=<ltrb 0 0 0 0>
     outlineSettings {
      bUseBrushTransparency=false
      roundingType=HalfHeightRadius
      width=0
      color {colorUseRule=UseColor_Specified, specifiedColor=<rgba 0 0 0 0>}
      cornerRadii=<0 0 0 0>
     }
     tintColor {colorUseRule=UseColor_Specified, specifiedColor=<rgba 1 1 1 1>}
     uVRegion {bIsValid=false, max=<0 0>, min=<0 0>}
    }
    loadingBackgroundBrush {
     bIsDynamicallyLoaded=false
     drawAs=Image
     imageType=FullColor
     mirroring=NoMirror
     resourceName=None
     resourceObject=null
     tiling=NoTile
     imageSize=<512 256>
     margin=<ltrb 0 0 0 0>
     outlineSettings {
      bUseBrushTransparency=false
      roundingType=HalfHeightRadius
      width=0
      color {colorUseRule=UseColor_Specified, specifiedColor=<rgba 0 0 0 0>}
      cornerRadii=<0 0 0 0>
     }
     tintColor {colorUseRule=UseColor_Specified, specifiedColor=<rgba 1 1 1 1>}
     uVRegion {bIsValid=false, max=<0 0>, min=<0 0>}
    }
   }
   slot {class=SizeBoxSlot}
  }
  Controllers_Info {
   widgetClassPath=&HorizontalBox
   parent=Overlay_0
   widget=Controllers_Info
   properties {bOverride_Cursor=true}
   slot {class=OverlaySlot, properties {horizontalAlignment=HAlign_Fill, verticalAlignment=VAlign_Fill, padding=<ltrb 48 48 48 48>}}
  }
  Rewards_Text {
   widgetClassPath=&CommonRichTextBlock
   bIsVariable=true
   parent=Controllers_Info
   widget=Rewards_Text
   properties {
    bOverrideDefaultStyle=true
    justification=Center
    mobileTextBlockScale=1.2
    text=XX Rewards
    wrapTextAt=500
    defaultTextStyleOverride {
     overflowPolicy=Clip
     transformPolicy=None
     colorAndOpacity {colorUseRule=UseColor_Specified, specifiedColor=<rgba 1 1 1 1>}
     font {
      bForceMonospaced=false
      bMaterialIsStencil=false
      fontMaterial=null
      fontObject=&HeadingNow
      letterSpacing=0
      monospacedWidth=1
      size=15
      skewAmount=0
      typefaceFontName=75 Medium
      outlineSettings {
       bApplyOutlineToDropShadows=false
       bMiteredCorners=false
       bSeparateFillAlpha=false
       outlineMaterial=null
       outlineSize=0
       outlineColor=<rgba 0 0 0 1>
      }
     }
     highlightColor {colorUseRule=UseColor_Specified, specifiedColor=<rgba 0 0 0 1>}
     highlightShape {
      bIsDynamicallyLoaded=false
      drawAs=Image
      imageType=NoImage
      mirroring=NoMirror
      resourceName=None
      resourceObject=null
      tiling=NoTile
      imageSize=<32 32>
      margin=<ltrb 0 0 0 0>
      outlineSettings {
       bUseBrushTransparency=false
       roundingType=HalfHeightRadius
       width=0
       color {colorUseRule=UseColor_Specified, specifiedColor=<rgba 0 0 0 0>}
       cornerRadii=<0 0 0 0>
      }
      tintColor {colorUseRule=UseColor_Specified, specifiedColor=<rgba 1 1 1 1>}
      uVRegion {bIsValid=false, max=<0 0>, min=<0 0>}
     }
     selectedBackgroundColor {colorUseRule=UseColor_Foreground, specifiedColor=<rgba 1 0 1 1>}
     shadowColorAndOpacity=<rgba 0 0 0 1>
     shadowOffset=<0 0>
     strikeBrush {
      bIsDynamicallyLoaded=false
      drawAs=Image
      imageType=NoImage
      mirroring=NoMirror
      resourceName=None
      resourceObject=null
      tiling=NoTile
      imageSize=<32 32>
      margin=<ltrb 0 0 0 0>
      outlineSettings {
       bUseBrushTransparency=false
       roundingType=HalfHeightRadius
       width=0
       color {colorUseRule=UseColor_Specified, specifiedColor=<rgba 0 0 0 0>}
       cornerRadii=<0 0 0 0>
      }
      tintColor {colorUseRule=UseColor_Specified, specifiedColor=<rgba 1 1 1 1>}
      uVRegion {bIsValid=false, max=<0 0>, min=<0 0>}
     }
     underlineBrush {
      bIsDynamicallyLoaded=false
      drawAs=Image
      imageType=NoImage
      mirroring=NoMirror
      resourceName=None
      resourceObject=null
      tiling=NoTile
      imageSize=<32 32>
      margin=<ltrb 0 0 0 0>
      outlineSettings {
       bUseBrushTransparency=false
       roundingType=HalfHeightRadius
       width=0
       color {colorUseRule=UseColor_Specified, specifiedColor=<rgba 0 0 0 0>}
       cornerRadii=<0 0 0 0>
      }
      tintColor {colorUseRule=UseColor_Specified, specifiedColor=<rgba 1 1 1 1>}
      uVRegion {bIsValid=false, max=<0 0>, min=<0 0>}
     }
    }
   }
   slot {class=HorizontalBoxSlot, properties {horizontalAlignment=HAlign_Left, verticalAlignment=VAlign_Center, size {sizeRule=Fill, value=1}}}
  }
  WBP_PriceSwitcher_Base {
   widgetClassPath=/FortUILibrary/Systems/ItemShopTile/WBP_PriceSwitcher_Base
   bIsVariable=true
   parent=Controllers_Info
   widget=WBP_PriceSwitcher_Base
   properties {
    hasRealPrice=true
    isAdditionalRealPrice=true
    previewMobileState=true
    previewRealPriceText=""
    previewRegularPrice=1800
    priceAndOwnershipHorizontalAlignment=HAlign_Right
    priceDisplayMode=PriceOnly
    realPriceHorizontalAlignment=HAlign_Center
    realPriceVerticalAlignment=VAlign_Center
    realPriceWidget=&self:WidgetTree.Text_RealPrice
    widgetHeight=60
    widgetHeightWithRealPrice=88
    additionalPriceTextFont {
     bForceMonospaced=false
     bMaterialIsStencil=false
     fontMaterial=/FortUILibrary/Systems/ItemShopTile/Materials/PriceSwitcher/MI_UIKit_PriceSwitcher_Text
     fontObject=&HeadingNow
     letterSpacing=0
     monospacedWidth=1
     size=10.5
     skewAmount=0
     typefaceFontName=75 Medium
     outlineSettings {
      bApplyOutlineToDropShadows=false
      bMiteredCorners=false
      bSeparateFillAlpha=false
      outlineMaterial=null
      outlineSize=0
      outlineColor=<rgba 0 0 0 1>
     }
    }
    additionalRegularPriceTextFont {
     bForceMonospaced=false
     bMaterialIsStencil=false
     fontMaterial=&MI_UIKit_PriceSwitcher_Text
     fontObject=&HeadingNow
     letterSpacing=0
     monospacedWidth=1
     size=15
     skewAmount=0
     typefaceFontName=75 Medium
     outlineSettings {
      bApplyOutlineToDropShadows=false
      bMiteredCorners=false
      bSeparateFillAlpha=false
      outlineMaterial=null
      outlineSize=0
      outlineColor=<rgba 0 0 0 1>
     }
    }
    currencyImageSize=<20 20>
    ownedTextFont {
     bForceMonospaced=false
     bMaterialIsStencil=false
     fontMaterial=&MI_UIKit_PriceSwitcher_Text
     fontObject=&HeadingNow
     letterSpacing=0
     monospacedWidth=1
     size=15
     skewAmount=0
     typefaceFontName=74 Regular
     outlineSettings {
      bApplyOutlineToDropShadows=false
      bMiteredCorners=false
      bSeparateFillAlpha=false
      outlineMaterial=null
      outlineSize=0
      outlineColor=<rgba 0 0 0 1>
     }
    }
    priceTextFont {
     bForceMonospaced=false
     bMaterialIsStencil=false
     fontMaterial=&MI_UIKit_PriceSwitcher_Text
     fontObject=&HeadingNow
     letterSpacing=0
     monospacedWidth=1
     size=15
     skewAmount=0
     typefaceFontName=75 Medium
     outlineSettings {
      bApplyOutlineToDropShadows=false
      bMiteredCorners=false
      bSeparateFillAlpha=false
      outlineMaterial=null
      outlineSize=0
      outlineColor=<rgba 0 0 0 1>
     }
    }
    realPriceTextFont {
     bForceMonospaced=false
     bMaterialIsStencil=false
     fontMaterial=&MI_UIKit_PriceSwitcher_Text
     fontObject=&HeadingNow
     letterSpacing=0
     monospacedWidth=1
     size=15
     skewAmount=0
     typefaceFontName=74 Regular
     outlineSettings {
      bApplyOutlineToDropShadows=false
      bMiteredCorners=false
      bSeparateFillAlpha=false
      outlineMaterial=null
      outlineSize=0
      outlineColor=<rgba 0 0 0 1>
     }
    }
    regularPriceTextFont {
     bForceMonospaced=false
     bMaterialIsStencil=false
     fontMaterial=&MI_UIKit_PriceSwitcher_Text
     fontObject=&HeadingNow
     letterSpacing=0
     monospacedWidth=1
     size=15
     skewAmount=0
     typefaceFontName=75 Medium
     outlineSettings {
      bApplyOutlineToDropShadows=false
      bMiteredCorners=false
      bSeparateFillAlpha=false
      outlineMaterial=null
      outlineSize=0
      outlineColor=<rgba 0 0 0 1>
     }
    }
    specialOfferTextFont {
     bForceMonospaced=false
     bMaterialIsStencil=false
     fontMaterial=&MI_UIKit_PriceSwitcher_Text
     fontObject=&HeadingNow
     letterSpacing=0
     monospacedWidth=1
     size=15.75
     skewAmount=0
     typefaceFontName=74 Regular
     outlineSettings {
      bApplyOutlineToDropShadows=false
      bMiteredCorners=false
      bSeparateFillAlpha=false
      outlineMaterial=null
      outlineSize=0
      outlineColor=<rgba 0 0 0 1>
     }
    }
   }
   slot {class=HorizontalBoxSlot, properties {horizontalAlignment=HAlign_Right, verticalAlignment=VAlign_Center, padding=<ltrb 36 0 0 0>}}
  }
  Text_RealPrice {
   widgetClassPath=&CommonTextBlock
   bIsVariable=true
   parent=Controllers_Info
   widget=Text_RealPrice
   properties {
    bAutoCollapseWithEmptyText=true
    mobileFontSizeMultiplier=1.2
    text="[ph] (equivalent to $8.99*)"
    font {
     bForceMonospaced=false
     bMaterialIsStencil=false
     fontMaterial=null
     fontObject=&NotoSans
     letterSpacing=0
     monospacedWidth=1
     size=24
     skewAmount=0
     typefaceFontName=NotoSans-Regular
     outlineSettings {
      bApplyOutlineToDropShadows=false
      bMiteredCorners=false
      bSeparateFillAlpha=false
      outlineMaterial=null
      outlineSize=0
      outlineColor=<rgba 0 0 0 1>
     }
    }
    shadowOffset=<0 0>
   }
   slot {class=HorizontalBoxSlot, properties {horizontalAlignment=HAlign_Left, verticalAlignment=VAlign_Center}}
  }
  Button_Add_1 {
   widgetClassPath=&WBP_UIKit_Button_Regular
   bIsVariable=true
   parent=Controllers
   widget=Button_Add_1
   properties {
    bIsFocusable=false
    minHeight=48
    minWidth=112
    mobileMinHeight=88
    mobileMinWidth=88
    text=+1
    triggeringInputAction {dataTable=&InputActionDataTable, rowName=GenericRightShoulder}
   }
   slot {class=HorizontalBoxSlot, properties {horizontalAlignment=HAlign_Center, verticalAlignment=VAlign_Center, padding=<ltrb 0 0 16 0>}}
  }
  GridPanel_1 {
   widgetClassPath=&GridPanel
   parent=MainContent
   widget=GridPanel_1
   properties {bOverride_Cursor=true}
   slot {class=VerticalBoxSlot, properties {horizontalAlignment=HAlign_Center, padding=<ltrb 0 0 0 36>}}
  }
  RewardTrack_Slot {
   widgetClassPath=&NamedSlot
   bIsVariable=true
   parent=GridPanel_1
   widget=RewardTrack_Slot
   properties {navigation=&self:WidgetTree.RewardTrack_Slot.WidgetNavigation_0}
   slot {class=GridSlot, properties {layer=-1}}
  }
  BuySelectorBox {
   widgetClassPath=&HorizontalBox
   bIsVariable=true
   parent=GridPanel_1
   widget=BuySelectorBox
   properties {visibility=Hidden}
   slot {class=GridSlot, properties {row=1, padding=<ltrb 0 8 0 0>}}
  }
  OwnedSpacer {widgetClassPath=&Spacer, bIsVariable=true, parent=BuySelectorBox, widget=OwnedSpacer, slot {class=HorizontalBoxSlot}}
  BuySize {
   widgetClassPath=&SizeBox
   bIsVariable=true
   parent=BuySelectorBox
   widget=BuySize
   properties {bOverride_WidthOverride=true}
   slot {class=HorizontalBoxSlot}
  }
  BuyBorder {
   widgetClassPath=&Border
   bIsVariable=true
   parent=BuySize
   widget=BuyBorder
   properties {
    horizontalAlignment=HAlign_Center
    verticalAlignment=VAlign_Center
    background {
     bIsDynamicallyLoaded=false
     drawAs=Image
     imageType=FullColor
     mirroring=NoMirror
     resourceName=None
     resourceObject=/FortUILibrary/UIKit/Buttons/MaterialInstances/CTAButtons/MI_UIKit_BgLoud
     tiling=NoTile
     imageSize=<32 32>
     margin=<ltrb 0 0 0 0>
     outlineSettings {
      bUseBrushTransparency=false
      roundingType=HalfHeightRadius
      width=0
      color {colorUseRule=UseColor_Specified, specifiedColor=<rgba 0 0 0 0>}
      cornerRadii=<0 0 0 0>
     }
     tintColor {colorUseRule=UseColor_Specified, specifiedColor=<rgba 1 1 1 1>}
     uVRegion {bIsValid=false, max=<0 0>, min=<0 0>}
    }
    padding=<ltrb 8 4 8 4>
   }
   slot {class=SizeBoxSlot}
  }
  BuyMarker_Text {
   widgetClassPath=&CommonRichTextBlock
   bIsVariable=true
   parent=BuyBorder
   widget=BuyMarker_Text
   properties {
    bOverrideDefaultStyle=true
    justification=Center
    mobileTextBlockScale=1.2
    text=Buy
    textTransformPolicy=ToUpper
    wrapTextAt=500
    defaultTextStyleOverride {
     overflowPolicy=Clip
     transformPolicy=None
     colorAndOpacity {colorUseRule=UseColor_Specified, specifiedColor=<rgba 0 0 0 1>}
     font {
      bForceMonospaced=false
      bMaterialIsStencil=false
      fontMaterial=null
      fontObject=&HeadingNow
      letterSpacing=0
      monospacedWidth=1
      size=15
      skewAmount=0
      typefaceFontName=75 Medium
      outlineSettings {
       bApplyOutlineToDropShadows=false
       bMiteredCorners=false
       bSeparateFillAlpha=false
       outlineMaterial=null
       outlineSize=0
       outlineColor=<rgba 0 0 0 1>
      }
     }
     highlightColor {colorUseRule=UseColor_Specified, specifiedColor=<rgba 0 0 0 1>}
     highlightShape {
      bIsDynamicallyLoaded=false
      drawAs=Image
      imageType=NoImage
      mirroring=NoMirror
      resourceName=None
      resourceObject=null
      tiling=NoTile
      imageSize=<32 32>
      margin=<ltrb 0 0 0 0>
      outlineSettings {
       bUseBrushTransparency=false
       roundingType=HalfHeightRadius
       width=0
       color {colorUseRule=UseColor_Specified, specifiedColor=<rgba 0 0 0 0>}
       cornerRadii=<0 0 0 0>
      }
      tintColor {colorUseRule=UseColor_Specified, specifiedColor=<rgba 1 1 1 1>}
      uVRegion {bIsValid=false, max=<0 0>, min=<0 0>}
     }
     selectedBackgroundColor {colorUseRule=UseColor_Foreground, specifiedColor=<rgba 1 0 1 1>}
     shadowColorAndOpacity=<rgba 0 0 0 1>
     shadowOffset=<0 0>
     strikeBrush {
      bIsDynamicallyLoaded=false
      drawAs=Image
      imageType=NoImage
      mirroring=NoMirror
      resourceName=None
      resourceObject=null
      tiling=NoTile
      imageSize=<32 32>
      margin=<ltrb 0 0 0 0>
      outlineSettings {
       bUseBrushTransparency=false
       roundingType=HalfHeightRadius
       width=0
       color {colorUseRule=UseColor_Specified, specifiedColor=<rgba 0 0 0 0>}
       cornerRadii=<0 0 0 0>
      }
      tintColor {colorUseRule=UseColor_Specified, specifiedColor=<rgba 1 1 1 1>}
      uVRegion {bIsValid=false, max=<0 0>, min=<0 0>}
     }
     underlineBrush {
      bIsDynamicallyLoaded=false
      drawAs=Image
      imageType=NoImage
      mirroring=NoMirror
      resourceName=None
      resourceObject=null
      tiling=NoTile
      imageSize=<32 32>
      margin=<ltrb 0 0 0 0>
      outlineSettings {
       bUseBrushTransparency=false
       roundingType=HalfHeightRadius
       width=0
       color {colorUseRule=UseColor_Specified, specifiedColor=<rgba 0 0 0 0>}
       cornerRadii=<0 0 0 0>
      }
      tintColor {colorUseRule=UseColor_Specified, specifiedColor=<rgba 1 1 1 1>}
      uVRegion {bIsValid=false, max=<0 0>, min=<0 0>}
     }
    }
   }
   slot {class=BorderSlot, properties {horizontalAlignment=HAlign_Center, verticalAlignment=VAlign_Center, padding=<ltrb 8 4 8 4>}}
  }
  UnselectedSpacer {
   widgetClassPath=&Spacer
   parent=BuySelectorBox
   widget=UnselectedSpacer
   slot {class=HorizontalBoxSlot, properties {size {sizeRule=Fill, value=1}}}
  }
  BuyButtons_HBox {
   widgetClassPath=&HorizontalBox
   parent=MainContent
   widget=BuyButtons_HBox
   slot {class=VerticalBoxSlot, properties {horizontalAlignment=HAlign_Center, padding=<ltrb 0 48 0 0>}}
  }
  BuyLevels_Button {
   widgetClassPath=/FortUILibrary/UIKit/Buttons/WBP_UIKit_Button_Loud
   bIsVariable=true
   parent=BuyButtons_HBox
   widget=BuyLevels_Button
   properties {bRequiresHold=true, text=Buy Rewards}
   slot {class=HorizontalBoxSlot}
  }
  Cancel_Button {
   widgetClassPath=&WBP_UIKit_Button_Regular
   bIsVariable=true
   parent=BuyButtons_HBox
   widget=Cancel_Button
   properties {bHideInputAction=true, buttonType=Quiet, text=Cancel}
   slot {class=HorizontalBoxSlot, properties {padding=<ltrb 32 0 0 0>}}
  }
  LegalandActionBar {
   widgetClassPath=&HorizontalBox
   bIsVariable=true
   parent=GridPanel_0
   widget=LegalandActionBar
   properties {bOverride_Cursor=true}
   slot {class=GridSlot, properties {columnSpan=3, row=3, verticalAlignment=VAlign_Bottom}}
  }
  LegalText {
   widgetClassPath=&CommonTextBlock
   bIsVariable=true
   parent=LegalandActionBar
   widget=LegalText
   properties {
    bIsScrollingEnabled=false
    mobileFontSizeMultiplier=1.4
    wrapTextAt=1200
    font {
     bForceMonospaced=false
     bMaterialIsStencil=false
     fontMaterial=null
     fontObject=&HeadingNow
     letterSpacing=10
     monospacedWidth=1
     size=11.25
     skewAmount=0
     typefaceFontName=74 Regular
     outlineSettings {
      bApplyOutlineToDropShadows=false
      bMiteredCorners=false
      bSeparateFillAlpha=false
      outlineMaterial=null
      outlineSize=0
      outlineColor=<rgba 0 0 0 1>
     }
    }
    shadowOffset=<0 0>
   }
   slot {class=HorizontalBoxSlot, properties {verticalAlignment=VAlign_Bottom}}
  }
  Spacer_329 {
   widgetClassPath=&Spacer
   parent=LegalandActionBar
   widget=Spacer_329
   slot {class=HorizontalBoxSlot, properties {verticalAlignment=VAlign_Center, size {sizeRule=Fill, value=1}}}
  }
  WBP_ActionBarContainer {
   widgetClassPath=/Game/UI/Frontend/BottomBar/WBP_ActionBarContainer
   bIsVariable=true
   parent=LegalandActionBar
   widget=WBP_ActionBarContainer
   slot {class=HorizontalBoxSlot, properties {horizontalAlignment=HAlign_Right, verticalAlignment=VAlign_Bottom}}
  }
  Button_Bar {widgetClassPath=&HorizontalBox, namedSlotHost=WBP_ActionBarContainer, widget=Button_Bar}
  WBP_AthenaBottomBar_WithNotifications {
   widgetClassPath=/Game/UI/Frontend/BottomBar/WBP_AthenaBottomBar_WithNotifications
   bIsVariable=true
   parent=Button_Bar
   widget=WBP_AthenaBottomBar_WithNotifications
   slot {class=HorizontalBoxSlot, properties {padding=<ltrb 24 0 0 0>}}
  }
  Button_Close {
   widgetClassPath=/FortUILibrary/UIKit/Buttons/WBP_UIKit_Button_Quiet
   bIsVariable=true
   parent=Button_Bar
   widget=Button_Close
   properties {
    bIsFocusable=false
    text=Back
    visibility=Visible
    triggeringInputAction {dataTable=&InputActionDataTable, rowName=GenericBack}
   }
   slot {class=HorizontalBoxSlot, properties {verticalAlignment=VAlign_Center}}
  }
 ]
}
