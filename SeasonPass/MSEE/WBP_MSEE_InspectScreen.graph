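# Season pass (MSEE) inspect screen, an activatable widget that previews the selected reward with its item description.
# Previous/Next buttons step SelectedReward through the page's track rewards on the selection view model, hiding at either end.
# ShowChaseWidget swaps the buy panel and wallet text for chase reward progress and info; Buy pushes the BuyRewards menu.
# Theme data supplies text lookups and an async-loaded background, applied only if still requested and active.
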WidgetBlueprint /SeasonPass/MSEE/WBP_MSEE_InspectScreen
animations [
 Intro {
  binding_count=5
  length_seconds=0.1
  object=&self:Intro
  sequence {bindings [
    InteractionSwitcher {
     id=#1
     possessed_class=CommonVisibilitySwitcher
     tracks [
      {
       type=MovieSceneFloatTrack
       display_name=Render Opacity
       sections [{type=MovieSceneFloatSection, end_frame=2, start_frame=0, channels [None {type=MovieSceneScriptingFloatChannel, keys [[0,0],[2,1]]}]}]
      }
      {
       type=MovieScene2DTransformTrack
       sections [{
         type=MovieScene2DTransformSection
         end_frame=2
         start_frame=0
         channels [
          Translation.X {type=MovieSceneScriptingFloatChannel, default=0}
          Translation.Y {type=MovieSceneScriptingFloatChannel, keys [[0,32],[2,0]]}
         ]
        }]
      }
     ]
    }
    NavButtons_Overlay {
     id=#2
     possessed_class=Overlay
     tracks [
      {
       type=MovieSceneFloatTrack
       display_name=Render Opacity
       sections [{type=MovieSceneFloatSection, end_frame=2, start_frame=0, channels [None {type=MovieSceneScriptingFloatChannel, keys [[0,0],[2,1]]}]}]
      }
      {
       type=MovieScene2DTransformTrack
       sections [{
         type=MovieScene2DTransformSection
         end_frame=2
         start_frame=0
         channels [
          Translation.X {type=MovieSceneScriptingFloatChannel, const [0,2]}
          Translation.Y {type=MovieSceneScriptingFloatChannel, const [0,2]}
          Angle {type=MovieSceneScriptingFloatChannel, const [0,2]}
          Scale.X {type=MovieSceneScriptingFloatChannel, keys [[0,1.1],[2,1]]}
          Scale.Y {type=MovieSceneScriptingFloatChannel, const [1,2]}
          Shear.X {type=MovieSceneScriptingFloatChannel, const [0,2]}
          Shear.Y {type=MovieSceneScriptingFloatChannel, const [0,2]}
         ]
        }]
      }
     ]
    }
    ItemDescription {
     id=#3
     possessed_class=WBP_UIKit_ItemDescription_Compact_C
     tracks [
      {
       type=MovieSceneFloatTrack
       display_name=Render Opacity
       sections [{type=MovieSceneFloatSection, end_frame=2, start_frame=0, channels [None {type=MovieSceneScriptingFloatChannel, keys [[0,0],[2,1]]}]}]
      }
      {
       type=MovieScene2DTransformTrack
       sections [{
         type=MovieScene2DTransformSection
         end_frame=2
         start_frame=0
         channels [
          Translation.X {type=MovieSceneScriptingFloatChannel, keys [[0,-64],[2,0]]}
          Translation.Y {type=MovieSceneScriptingFloatChannel, keys [[0,-32],[2,0]]}
         ]
        }]
      }
     ]
    }
    ChaseReward_Info {
     id=#4
     possessed_class=CommonRichTextBlock
     tracks [{
       type=MovieSceneFloatTrack
       display_name=Render Opacity
       sections [{type=MovieSceneFloatSection, end_frame=2, start_frame=0, channels [None {type=MovieSceneScriptingFloatChannel, keys [[0,0],[0.8,0],[2,1]]}]}]
      }]
    }
    WBP_ActionBarContainer {
     id=#5
     possessed_class=WBP_ActionBarContainer_C
     tracks [{
       type=MovieSceneFloatTrack
       display_name=Render Opacity
       sections [{type=MovieSceneFloatSection, end_frame=2, start_frame=0, channels [None {type=MovieSceneScriptingFloatChannel, keys [[0,0],[2,1]]}]}]
      }]
    }
   ]}
 }
 HoverPrev {
  binding_count=1
  length_seconds=0.25
  object=&self:HoverPrev
  sequence {bindings [Button_Previous_IconSlot {
     id=#6
     possessed_class=NamedSlot
     tracks [{
       type=MovieScene2DTransformTrack
       sections [{
         type=MovieScene2DTransformSection
         end_frame=5
         start_frame=0
         channels [
          Translation.X {type=MovieSceneScriptingFloatChannel, const [0,1]}
          Translation.Y {type=MovieSceneScriptingFloatChannel, const [0,1]}
          Angle {type=MovieSceneScriptingFloatChannel, const [0,1]}
          Scale.X {type=MovieSceneScriptingFloatChannel, keys [[0,1],[5,1.1]]}
          Scale.Y {type=MovieSceneScriptingFloatChannel, keys [[0,1],[5,1.1]]}
          Shear.X {type=MovieSceneScriptingFloatChannel, const [0,1]}
          Shear.Y {type=MovieSceneScriptingFloatChannel, const [0,1]}
         ]
        }]
      }]
    }]}
 }
 HoverNext {
  binding_count=1
  length_seconds=0.25
  object=&self:HoverNext
  sequence {bindings [Button_Next_IconSlot {
     id=#7
     possessed_class=NamedSlot
     tracks [{
       type=MovieScene2DTransformTrack
       sections [{
         type=MovieScene2DTransformSection
         end_frame=5
         start_frame=0
         channels [
          Translation.X {type=MovieSceneScriptingFloatChannel, const [0,1]}
          Translation.Y {type=MovieSceneScriptingFloatChannel, const [0,1]}
          Angle {type=MovieSceneScriptingFloatChannel, const [0,1]}
          Scale.X {type=MovieSceneScriptingFloatChannel, keys [[0,1],[5,1.1]]}
          Scale.Y {type=MovieSceneScriptingFloatChannel, keys [[0,1],[5,1.1]]}
          Shear.X {type=MovieSceneScriptingFloatChannel, const [0,1]}
          Shear.Y {type=MovieSceneScriptingFloatChannel, const [0,1]}
         ]
        }]
      }]
    }]}
 }
 PressedNext {
  binding_count=1
  length_seconds=0.2961
  object=&self:PressedNext
  sequence {bindings [Button_Next_IconSlot {
     id=#8
     possessed_class=NamedSlot
     tracks [{
       type=MovieScene2DTransformTrack
       sections [{
         type=MovieScene2DTransformSection
         channels [
          Translation.X {type=MovieSceneScriptingFloatChannel, keys [[0,0],[1.4743,5],[4.4227,0],[5.9217,0]]}
          Translation.Y {type=MovieSceneScriptingFloatChannel, default=0}
         ]
        }]
      }]
    }]}
 }
 PressedPrev {
  binding_count=1
  length_seconds=0.2961
  object=&self:PressedPrev
  sequence {bindings [Button_Previous {
     id=#9
     possessed_class=WBP_UIKit_Button_Generic_C
     tracks [{
       type=MovieScene2DTransformTrack
       sections [{
         type=MovieScene2DTransformSection
         channels [
          Translation.X {type=MovieSceneScriptingFloatChannel, keys [[0,0],[1.4743,-5],[4.4227,0],[5.9217,0]]}
          Translation.Y {type=MovieSceneScriptingFloatChannel, default=0}
         ]
        }]
      }]
    }]}
 }
]
graphs [
 BP_GetDesiredFocusTarget {nodes [
   Get Desired Focus Target()
   > Return Node!(execute, ReturnValue:Widget Object Reference=Get Buy_Button() -> (Buy_Button:WBP UIKit Button Loud Object Reference))
  ]}
 EventGraph {nodes [
   "On Clicked (Buy_Button)"!() -> (OutputDelegate, then, Button:Common Button Base Object Reference) {is=title K2Node_ComponentBoundEvent}
   > n1 = PushPrimaryContentSoftClassAsync!(
    execute:Exec
    SoftWidgetClass:Common Activatable Widget Soft Class Reference=Get BuyMenu() -> (BuyMenu:Common Activatable Widget Soft Class Reference)
   ) -> (then:Exec, OnPrimaryContentPushed:Exec, ActivatableWidget:Common Activatable Widget Object Reference) {is=title K2Node_AsyncAction}
   n2 = Cast To WBP_MSEE_BuyRewards!(execute=@n1.OnPrimaryContentPushed, Object=@n1) -> (then, CastFailed, AsWBP MSEE Buy Rewards:WBP MSEE Buy Rewards Object Reference) {is=title K2Node_DynamicCast}
   > Set FromInspectScreen!(execute, FromInspectScreen:bool=true, self:WBP MSEE Buy Rewards Object Reference=@n2) -> (then, Output_Get:bool)
   > Set PlayFromBuyMenuIntro(PlayFromBuyMenuIntro:bool=true) -> (Output_Get:bool)
   "On Clicked (Button_Back)"!() -> (OutputDelegate, then, Button:Common Button Base Object Reference) {is=title K2Node_ComponentBoundEvent}
   > DeactivateWidget()
   n3 = "On Unfocused (Button_Previous)"!() -> (OutputDelegate, then, Button:Common Button Base Object Reference) {is=title K2Node_ComponentBoundEvent}
   PlayAnimation(
    execute=[@n3.then, @n4.then]
    InAnimation=Get HoverPrev() -> (HoverPrev:Widget Animation Object Reference)
    PlayMode=Reverse
    PlaybackSpeed=2
   )
   n4 = "On Unhovered (Button_Previous)"!() -> (OutputDelegate, then, Button:Common Button Base Object Reference) {is=title K2Node_ComponentBoundEvent}
   "On Clicked (Button_Next)"!() -> (OutputDelegate, then, Button:Common Button Base Object Reference) {is=title K2Node_ComponentBoundEvent}
   > SwitchItem_Next()
   "On Clicked (Button_Previous)"!() -> (OutputDelegate, then, Button:Common Button Base Object Reference) {is=title K2Node_ComponentBoundEvent}
   > SwitchItem_Previous()
   n5 = "On Hovered (Button_Previous)"!() -> (OutputDelegate, then, Button:Common Button Base Object Reference) {is=title K2Node_ComponentBoundEvent}
   PlayAnimation(execute=[@n5.then, @n10.then], InAnimation=Get HoverPrev())
   n6 = "On Hovered (Button_Next)"!() -> (OutputDelegate, then, Button:Common Button Base Object Reference) {is=title K2Node_ComponentBoundEvent}
   PlayAnimation(execute=[@n6.then, @n8.then], InAnimation=Get HoverNext() -> (HoverNext:Widget Animation Object Reference))
   n7 = "On Unhovered (Button_Next)"!() -> (OutputDelegate, then, Button:Common Button Base Object Reference) {is=title K2Node_ComponentBoundEvent}
   PlayAnimation(execute=[@n7.then, @n9.then], InAnimation=Get HoverNext(), PlayMode=Reverse, PlaybackSpeed=2)
   n8 = "On Focused (Button_Next)"!() -> (OutputDelegate, then, Button:Common Button Base Object Reference) {is=title K2Node_ComponentBoundEvent}
   n9 = "On Unfocused (Button_Next)"!() -> (OutputDelegate, then, Button:Common Button Base Object Reference) {is=title K2Node_ComponentBoundEvent}
   n10 = "On Focused (Button_Previous)"!() -> (OutputDelegate, then, Button:Common Button Base Object Reference) {is=title K2Node_ComponentBoundEvent}
   n11 = RequestLoadBackground!() -> (OutputDelegate, then, BackgroundMaterial:Soft Object Reference, Direction:EVaultWorldTransitionDirection Enum=Forward) {is=title K2Node_CustomEvent}
   > Set RequestedBackgroundSoftPath(RequestedBackgroundSoftPath:Soft Object Path Structure=To Soft Object Path!(SoftObjectReference:Soft Object Reference=@n11.BackgroundMaterial) -> (ReturnValue:Soft Object Path Structure)) -> (Output_Get:Soft Object Path Structure)
   > n12 = Async Load Asset(Asset=@n11.BackgroundMaterial)
   n13 = GetSoftObjectPath(Object=@n12)
   n14 = Get RequestedBackgroundSoftPath() -> (RequestedBackgroundSoftPath:Soft Object Path Structure)
   Branch(
    execute=@n12.Completed
    Condition="Equal Exactly (String)"(A=BreakSoftObjectPath(InSoftObjectPath=@n13), B=BreakSoftObjectPath(InSoftObjectPath=@n14))
   )
   > Branch(Condition=Get bIsActive() -> (bIsActive:bool))
   > UpdateBackground!(
    execute
    self:Fort Item Preview Background Object Reference=Get FortItemPreviewBackground() -> (FortItemPreviewBackground:Fort Item Preview Background Object Reference)
    LoadedBackground:Object Reference=@n12
    TransitionDirection:EVaultWorldTransitionDirection Enum=@n11.Direction
    PrimaryColor:Linear Color Structure=(R=1,G=1,B=1,A=0)
    SecondaryColor:Linear Color Structure=(R=0,G=0,B=0,A=0)
   ) -> (then)
   n15 = OnPressed_Next() {is=title K2Node_CustomEvent}
   > PlayAnimation(InAnimation=Get PressedNext() -> (PressedNext:Widget Animation Object Reference))
   n16 = OnPressed_Prev() {is=title K2Node_CustomEvent}
   > PlayAnimation(InAnimation=Get PressedPrev() -> (PressedPrev:Widget Animation Object Reference))
   PlayIntroAnim() {is=title K2Node_CustomEvent}
   > IntroAnimFinished()
   FromBuyMenuIntro() {is=title K2Node_CustomEvent}
   > FromBuyIntroFinished()
   n17 = IntroAnimFinished() {is=title K2Node_CustomEvent}
   IsScreenActive!(
    execute=[@n17.then, @n18.then]
    self:WBP MSEE Chase Reward Progress Object Reference=Get Chase Widget() -> (Chase Widget:WBP MSEE Chase Reward Progress Object Reference)
    IsActive:bool=Get ShowChaseWidget() -> (ShowChaseWidget:bool)
   ) -> (then)
   > Bind Event to On Pressed(
    self:WBP UIKit Button Generic Object Reference=Get Button_Previous() -> (Button_Previous:WBP UIKit Button Generic Object Reference)
    Delegate=@n16
   ) {is=title K2Node_AddDelegate}
   > Assign On Pressed!(
    execute:Exec
    self:WBP UIKit Button Generic Object Reference=Get Button_Next() -> (Button_Next:WBP UIKit Button Generic Object Reference)
    Delegate:"Delegate (by ref)"=@n15
   ) -> (then:Exec) {is=title K2Node_AssignDelegate}
   n18 = FromBuyIntroFinished() {is=title K2Node_CustomEvent}
   n19 = Event PreConstruct()
   > SetupPlaceholderInfo!(execute, self, IsDesignTime:bool=@n19.IsDesignTime) -> (then)
   > UpdateForCustom()
   > Set with Broadcast IsUsingCompact!(execute, IsUsingCompact:bool=IsUsingCompactUI()) -> (then, Output_Get:bool)
   Event On Activated()
   > ActivateWidget(self=Get WBP_ItemPreview() -> (WBP_ItemPreview:WBP Item Preview Object Reference))
   > SetPreviewData!(
    execute
    self:Fort Item Preview Widget Object Reference=Get WBP_ItemPreview()
    Data:"Fort Cosmetic Preview Data Structure (by ref)"=GetPreviewData!(self:Season Pass Selection VM Object Reference=Get SeasonPassSelectionVM() -> (SeasonPassSelectionVM:Season Pass Selection VM Object Reference)) -> (ReturnValue:"Fort Cosmetic Preview Data Structure (by ref)")
   ) -> (then)
   > n20 = Branch(Condition=Get PlayFromBuyMenuIntro() -> (PlayFromBuyMenuIntro:bool))
   > FromBuyMenuIntro() {is=title K2Node_CallFunction}
   PlayIntroAnim(execute=@n20.else) {is=title K2Node_CallFunction}
   Event On Deactivated()
   > IsScreenActive(self=Get Chase Widget(), IsActive=false)
   n21 = Get WBP_ItemPreview()
   > ClearPreviewItem(self:Fort Item Preview Widget Object Reference=@n21)
   > DeactivateWidget(self=@n21)
   > Unbind Event from On Pressed(self:WBP UIKit Button Generic Object Reference=Get Button_Previous(), Delegate=Create Event()) {is=title K2Node_RemoveDelegate}
   > Unbind Event from On Pressed(self=Get Button_Next(), Delegate=Create Event())
  ]}
 GetChaseRewardInfo {nodes [
   n1 = GetChaseRewardInfo!() -> (then, ThemeData:Season Pass Theme Data Structure, ItemName:Text, IsOwned:bool=~) {is=title K2Node_FunctionEntry}
   n2 = Text Lookup!(SeasonPassThemeData:"Season Pass Theme Data Structure (by ref)"=@n1, Ref:String=ChaseReward_Obtained) -> (ReturnValue:Text) {is=title K2Node_MacroInstance}
   n3 = Get ItemName() -> (ItemName:Text)
   n4 = Text Lookup(SeasonPassThemeData=@n1, Ref=ChaseReward_Info)
   > Return Node!(
    execute
    ChaseRewardOwned:Text=SelectText(A=Format Text!(Format=@n2, Item:Text=@n3) -> (Result), B=Format Text(Format=@n4, Item=@n3), bPickA=Get IsOwned() -> (IsOwned:bool))
   )
  ]}
 GetCostInfo {nodes [
   n1 = GetCostInfo!() -> (then, RewardVM:Season Pass Reward VM Object Reference, ThemeData:Season Pass Theme Data Structure) {is=title K2Node_FunctionEntry}
   > Return Node!(
    execute
    RequireText:Text=Format Text!(
     Format=Text Lookup(SeasonPassThemeData=@n1.ThemeData, Ref=Inspect_Cost)
     Cost:int=Get ProgressCountForUnlock!(self:Season Pass Reward VM Object Reference=@n1) -> (ProgressCountForUnlock:int=0)
    ) -> (Result)
   )
  ]}
 GetCurrentItemIndex {nodes [
   GetCurrentItemIndex() {is=title K2Node_FunctionEntry}
   > n1 = Get!(execute:Exec) -> (then:Exec, else:Exec, SeasonPassSelectionVM:Season Pass Selection VM Object Reference)
   > n2 = Get!(execute, self:Season Pass Selection VM Object Reference=@n1) -> (then, else, SelectedPage:Season Pass Page VM Object Reference)
   n3 = Get Rewards!(self:Season Pass Track VM Object Reference="Get (a copy)"(Array:Array of Season Pass Track VM Object References=Get Tracks!(self:Season Pass Page VM Object Reference=@n2) -> (Tracks:Array of Season Pass Track VM Object References)) -> (Output:Season Pass Track VM Object Reference)) -> (Rewards:Array of Season Pass Reward VM Object References)
   > Return Node!(
    execute
    CurrentIndex:int=Find Item!(
     TargetArray:Array of Season Pass Reward VM Object References=@n3
     ItemToFind:"Season Pass Reward VM Object Reference (by ref)"=Get SeasonPassRewardVM() -> (SeasonPassRewardVM:Season Pass Reward VM Object Reference)
    ) -> (ReturnValue)
    Rewards:Array of Season Pass Reward VM Object References=@n3
   )
  ]}
 GetFrontendFlowInterruptConfig {nodes [
   GetFrontendFlowInterruptConfig() {is=title K2Node_FunctionEntry}
   > Return Node!(execute, ReturnValue:Frontend Flow Interrup Config Structure)
  ]}
 GetTextByRef {nodes [
   n1 = GetTextByRef!() -> (then, Ref:String, SeasonPassThemeData:"Season Pass Theme Data Structure (by ref)") {is=title K2Node_FunctionEntry}
   n2 = TryLookupText!(
    Table:"Fort Text Reference Table Structure (by ref)"=Break Season Pass Theme Data!(SeasonPassThemeData:"Season Pass Theme Data Structure (by ref)"=Copy!(Item:Season Pass Theme Data Structure=@n1.SeasonPassThemeData) -> (ReturnValue:Season Pass Theme Data Structure) {is=title K2Node_Copy}) -> (
     TitleColor:Linear Color Structure
     TitleMaterial:Soft Object Reference
     TitleOutline:Font Outline Settings Structure
     SeasonColor:Linear Color Structure
     ProgressionColor:Linear Color Structure
     OverlayColor:Linear Color Structure
     VaultBackgroundMaterial:Soft Object Reference
     GiftboxVaultBackgroundMaterial:Soft Object Reference
     UIBackgroundMaterial:Soft Object Reference
     UpsellBackgroundMaterial:Soft Object Reference
     CrewCategoryBackgroundMaterial:Soft Object Reference
     CustomLogo:Soft Object Reference
     CustomSeasonTitle:Text
     CustomSeasonTitleColor:Linear Color Structure
     Upsell2BackgrounMaterial:Soft Object Reference
     ImageReferenceTable:Fort Image Reference Table Structure
     ColorReferenceTable:Fort Color Reference Table Structure
     TextReferenceTable:Fort Text Reference Table Structure
    ) {is=title K2Node_BreakStruct}.TextReferenceTable
    Name:String=Get Ref() -> (Ref:String)
   ) -> (OutTextReference:Text, ReturnValue:bool)
   > Return Node(ReturnValue:Text=SelectText(A=@n2, bPickA=@n2.ReturnValue))
  ]}
 GetWalletInfo {nodes [
   n1 = GetWalletInfo!() -> (then, Wallet:int, ThemeData:Season Pass Theme Data Structure) {is=title K2Node_FunctionEntry}
   > Return Node!(
    execute
    WalletText:Text=Format Text!(Format=Text Lookup(SeasonPassThemeData=@n1.ThemeData, Ref=Inspect_Wallet), Wallet:int=Get Wallet() -> (Wallet:int=0)) -> (Result)
   )
  ]}
 IsItemOwned {nodes [
   n1 = IsItemOwned!() -> (then, IsOwned:bool=~) {is=title K2Node_FunctionEntry}
   > SetVisibility(
    self=Get BuyReward_VBox() -> (BuyReward_VBox:Vertical Box Object Reference)
    InVisibility=Select(Option 0=SelfHitTestInvisible, Option 1=Collapsed, Index=@n1)
   )
   > SetVisibility(
    self=Get Claimed_Text() -> (Claimed_Text:Common Rich Text Block Object Reference)
    InVisibility=Select(Option 0=Collapsed, Option 1=SelfHitTestInvisible, Index=@n1)
   )
   > ShouldShowChaseWidget()
  ]}
 OnRewardVMUpdated {nodes [
   OnRewardVMUpdated!() -> (then, RewardVM:Season Pass Reward VM Object Reference) {is=title K2Node_FunctionEntry}
   > UpdateArrowButtonVisibility()
  ]}
 SetChaseInspectWidget {nodes [
   n1 = SetChaseInspectWidget!() -> (then, Chase Widget:WBP MSEE Chase Reward Progress Object Reference) {is=title K2Node_FunctionEntry}
   > Set Chase Widget(Chase Widget:WBP MSEE Chase Reward Progress Object Reference=@n1) -> (Output_Get:WBP MSEE Chase Reward Progress Object Reference)
  ]}
 SetupPlaceholderInfo {nodes [
   n1 = SetupPlaceholderInfo!() -> (then, IsDesignTime:bool=~) {is=title K2Node_FunctionEntry}
   > SetText(
    self=[
     Get Wallet_Info() -> (Wallet_Info:Common Rich Text Block Object Reference)
     Get ChaseReward_Info() -> (ChaseReward_Info:Common Rich Text Block Object Reference)
     Get ItemCostInfo_Text() -> (ItemCostInfo_Text:Common Rich Text Block Object Reference)
    ]
    InText=SelectText(A="INVTEXT(\"[PH] Placeholder Text Spot\")", bPickA=@n1)
   )
  ]}
 ShouldShowChaseWidget {nodes [
   ShouldShowChaseWidget() {is=title K2Node_FunctionEntry}
   > SetActiveWidget(
    self=Get InteractionSwitcher() -> (InteractionSwitcher:Common Visibility Switcher Object Reference)
    Widget=Select(
     Option 0:Widget Object Reference=Get BuyReward_VBox()
     Option 1:Widget Object Reference=Get ChaseProgress_Slot() -> (ChaseProgress_Slot:Named Slot Object Reference)
     Index=Get ShowChaseWidget()
    ) -> (ReturnValue:Widget Object Reference)
   )
   > SetActiveWidget(
    self=Get InfoText_Switcher() -> (InfoText_Switcher:Common Visibility Switcher Object Reference)
    Widget=Select(Option 0=Get Wallet_Info(), Option 1=Get ChaseReward_Info(), Index=Get ShowChaseWidget())
   )
   > n1 = Branch(Condition=Get ShowChaseWidget())
   > SetVisibility(self=Get Button_Previous(), InVisibility=Collapsed)
   > SetVisibility(self=Get Button_Next(), InVisibility=Collapsed)
   UpdateArrowButtonVisibility(execute=@n1.else)
  ]}
 SwitchItem_Next {nodes [
   SwitchItem_Next() {is=title K2Node_FunctionEntry}
   > n1 = GetCurrentItemIndex!(execute, self) -> (then, CurrentIndex:int=0, Rewards:Array of Season Pass Reward VM Object References) {is=title K2Node_CallFunction}
   > n2 = Branch(Condition="integer >= integer"(A=@n1, B=Last Index!(TargetArray:Array of Season Pass Reward VM Object References=@n1.Rewards) -> (ReturnValue=0)))
   Set SelectedReward!(
    execute=@n2.else
    SelectedReward:Season Pass Reward VM Object Reference="Get (a copy)"(Array:Array of Season Pass Reward VM Object References=@n1.Rewards, Dimension 1=int + int(A=@n1)) -> (Output:Season Pass Reward VM Object Reference)
    self:Season Pass Selection VM Object Reference=Get SeasonPassSelectionVM()
   ) -> (then, Output_Get:Season Pass Reward VM Object Reference)
  ]}
 SwitchItem_Previous {nodes [
   SwitchItem_Previous() {is=title K2Node_FunctionEntry}
   > n1 = GetCurrentItemIndex() -> (CurrentIndex=0)
   > n2 = Branch(Condition="integer <= integer"(A=@n1))
   Set SelectedReward(execute=@n2.else, SelectedReward="Get (a copy)"(Array=@n1.Rewards, Dimension 1=int - int(A=@n1)), self=Get SeasonPassSelectionVM())
  ]}
 TextLookup {nodes [
   n1 = Inputs!() -> (SeasonPassThemeData:"Season Pass Theme Data Structure (by ref)", Ref:String)
   n2 = TryLookupText(Table=Break Season Pass Theme Data(SeasonPassThemeData=@n1).TextReferenceTable, Name=@n1.Ref)
   Outputs!(ReturnValue:Text=SelectText(A=@n2, bPickA=@n2.ReturnValue))
  ]}
 UpdateArrowButtonVisibility {nodes [
   UpdateArrowButtonVisibility() {is=title K2Node_FunctionEntry}
   > n1 = GetCurrentItemIndex() -> (CurrentIndex=0)
   > SetVisibility(
    self=Get Button_Previous()
    InVisibility=Select(Option 0:ESlateVisibility Enum=SelfHitTestInvisible, Option 1:ESlateVisibility Enum=Collapsed, Index="Equal (Integer)"(B=@n1)) -> (ReturnValue:ESlateVisibility Enum)
   )
   n2 = Last Index(TargetArray=@n1.Rewards) -> (ReturnValue=0)
   > SetVisibility(self=Get Button_Next(), InVisibility=Select(Option 0=SelfHitTestInvisible, Option 1=Collapsed, Index="Equal (Integer)"(A=@n2, B=@n1)))
  ]}
 UpdateForCustom {nodes [
   UpdateForCustom() {is=title K2Node_FunctionEntry}
   n1 = Get Button_Previous_IconSlot() -> (Button_Previous_IconSlot:Named Slot Object Reference)
   > SetIcon!(
    execute
    self:WBP UIKit Button Generic Base Object Reference=Get Button_Previous()
    Icon:Object Reference=Select(
     Option 0:Object Reference=/FortUILibrary/Foundation/Buttons/Materials/Arrows/M_CycleButton_Arrow_Left
     Option 1:Object Reference
     Index=HasAnyChildren!(self:Panel Widget Object Reference=@n1) -> (ReturnValue:bool)
    ) -> (ReturnValue:Object Reference)
   ) -> (then)
   n2 = Get Button_Next_IconSlot() -> (Button_Next_IconSlot:Named Slot Object Reference)
   > SetIcon(
    self=Get Button_Next()
    Icon=Select(Option 0=/FortUILibrary/Foundation/Buttons/Materials/Arrows/M_CycleButton_Arrow_Rgt, Index=HasAnyChildren(self=@n2))
   )
   > SetDecorators!(
    execute
    self:Rich Text Block Object Reference=[Get Wallet_Info(), Get ChaseReward_Info(), Get ItemCostInfo_Text()]
    InDecoratorClasses:Array of Rich Text Block Decorator Class References=Get In Decorator Classes() -> (In Decorator Classes:Array of Rich Text Block Decorator Class References)
   ) -> (then)
  ]}
 UpdateItemDetails {nodes [
   n1 = UpdateItemDetails!() -> (then, FortItemVM:Fort Item VM Object Reference) {is=title K2Node_FunctionEntry}
   > Is Valid(InputObject=@n1)
   > SetFortItemDetailsVM!(
    execute
    self:WBP UIKit Item Description Compact Object Reference=Get ItemDescription() -> (ItemDescription:WBP UIKit Item Description Compact Object Reference)
    Viewmodel:Fort Item Details VM Object Reference=Get ItemDetails!(self:Fort Item VM Object Reference=@n1) -> (ItemDetails:Fort Item Details VM Object Reference)
   ) -> (then)
  ]}
 VM_GetThemeData {nodes [
   n1 = VM_GetThemeData!() -> (then, Theme data:Season Pass Theme Data Structure) {is=title K2Node_FunctionEntry}
   > Set Theme Data(Theme Data:Season Pass Theme Data Structure=@n1) -> (Output_Get:Season Pass Theme Data Structure)
   > RequestLoadBackground!(
    execute
    self
    BackgroundMaterial:Soft Object Reference=Break Season Pass Theme Data(SeasonPassThemeData=@n1).VaultBackgroundMaterial
    Direction:EVaultWorldTransitionDirection Enum=None
   ) -> (then) {is=title K2Node_CallFunction}
  ]}
]
variables [
 BuyMenu {type=Class, default=&/WBP_MSEE_BuyRewards.WBP_MSEE_BuyRewards_C}
 Chase Widget {type=WBP_MSEE_ChaseReward_Progress_C, default=null}
 In Decorator Classes {type="array<Class>", default []}
 IsUsingCompact {type=boolean}
 ItemDescriptionCompactUIPadding {type=Margin, default=<ltrb 64 64 0 0>}
 ItemDescriptionDefaultPadding {type=Margin, default=<ltrb 64 64 0 0>}
 ModelPreviewCompactUIPadding {type=Margin, default=<ltrb 128 128 128 0>}
 ModelPreviewDefaultPadding {type=Margin, default=<ltrb 128 0 128 0>}
 PlayFromBuyMenuIntro {type=boolean}
 RequestedBackgroundSoftPath {type=SoftObjectPath, default=""}
 ShowChaseWidget {type=boolean}
 Text Style Set {type=DataTable, default=null}
 Theme Data {
  type=SeasonPassThemeData
  default {
   crewCategoryBackgroundMaterial=null
   customLogo=null
   customSeasonTitle=""
   giftboxVaultBackgroundMaterial=null
   primaryAssetId=""
   titleMaterial=null
   uIBackgroundMaterial=null
   upsell2BackgrounMaterial=null
   upsellBackgroundMaterial=null
   vaultBackgroundMaterial=null
   benefitsByType []
   colorReferenceTable {namedColors []}
   customSeasonTitleColor=<rgba 1 1 1 1>
   imageReferenceTable {namedImages []}
   overlayColor=<rgba 0 0 0 1>
   progressionColor=<rgba 1 1 1 1>
   seasonColor=<rgba 1 1 1 1>
   textReferenceTable {namedTexts []}
   titleColor=<rgba 1 1 1 1>
   titleOutline {
    bApplyOutlineToDropShadows=false
    bMiteredCorners=false
    bSeparateFillAlpha=false
    outlineMaterial=null
    outlineSize=0
    outlineColor=<rgba 0 0 0 1>
   }
  }
 }
]
widgets {
 info {
  inheritedWidgetCount=0
  namedSlotCount=3
  parentClass=&CommonActivatableWidget
  rootWidgetClass=&Overlay
  widgetCount=31
 }
 widgets [
  Overlay_69 {widgetClassPath=&Overlay, widget=Overlay_69}
  WBP_CaptureForPostBufferUpdate {
   widgetClassPath=&WBP_CaptureForPostBufferUpdate
   bIsVariable=true
   parent=Overlay_69
   slot {class=OverlaySlot, properties {horizontalAlignment=HAlign_Fill, verticalAlignment=VAlign_Fill}}
  }
  FortItemPreviewBackground {
   widgetClassPath=/Script/FortniteUI.FortItemPreviewBackground
   bIsVariable=true
   parent=Overlay_69
   widget=FortItemPreviewBackground
   properties {previewSceneChangerTag {tagName=UI.ItemPreview.SceneChanger.MPItemShop}}
   slot {class=OverlaySlot, properties {horizontalAlignment=HAlign_Fill, verticalAlignment=VAlign_Fill}}
  }
  Background_Custom {
   widgetClassPath=&NamedSlot
   bIsVariable=true
   parent=Overlay_69
   widget=Background_Custom
   slot {class=OverlaySlot, properties {horizontalAlignment=HAlign_Fill, verticalAlignment=VAlign_Fill}}
  }
  SafeZone {
   widgetClassPath=&SafeZone
   bIsVariable=true
   parent=Overlay_69
   widget=SafeZone
   slot {class=OverlaySlot, properties {horizontalAlignment=HAlign_Fill, verticalAlignment=VAlign_Fill}}
  }
  SizeBox_2 {
   widgetClassPath=&SizeBox
   parent=SafeZone
   widget=SizeBox_2
   properties {maxAspectRatio=3.9, minAspectRatio=1.7778}
   slot {class=SafeZoneSlot}
  }
  GridPanel_0 {
   widgetClassPath=&GridPanel
   parent=SizeBox_2
   widget=GridPanel_0
   properties {columnFill [1,1,0.8], rowFill [0,1,0]}
   slot {class=SizeBoxSlot, properties {padding=<ltrb 80 40 80 40>}}
  }
  WBP_ItemPreview {
   widgetClassPath=/FortniteCoreUI/Frontend/ItemPreview/WBP_ItemPreview
   bIsVariable=true
   parent=GridPanel_0
   widget=WBP_ItemPreview
   properties {
    bAllowEquippedVariants=false
    bAllowPawnOverrideOnExclusiveItems=true
    bCanDisplayUserAccessories=false
    bRestrictZoomToWidget=true
    itemViewObject=&self:WidgetTree.WBP_ItemPreview.ItemViewObject
    visibility=Visible
    defaultPreviewContextTags {gameplayTags [{tagName=UI.ItemPreview.PreviewType.OnPawn}]}
    previewActionButtons [
     &self:WidgetTree.Button_PreviewAction
     &self:WidgetTree.Button_PreviewAction1
     &self:WidgetTree.Button_PreviewAction2
    ]
    previewSceneChangerTag {tagName=UI.ItemPreview.SceneChanger.MPItemShop}
   }
   slot {class=GridSlot, properties {columnSpan=3, rowSpan=2, padding=<ltrb 128 0 128 0>}}
  }
  InfoText_Switcher {
   widgetClassPath=&CommonVisibilitySwitcher
   bIsVariable=true
   parent=GridPanel_0
   widget=InfoText_Switcher
   slot {class=GridSlot, properties {column=2, verticalAlignment=VAlign_Top, padding=<ltrb 0 64 64 0>}}
  }
  Wallet_Info {
   widgetClassPath=&CommonRichTextBlock
   bIsVariable=true
   parent=InfoText_Switcher
   widget=Wallet_Info
   properties {
    bOverrideDefaultStyle=true
    justification=Right
    lineHeightPercentage=1.15
    mobileTextBlockScale=1.2
    textStyleSet=/Game/Athena/Items/Cosmetics/Metadata/ItemDescriptionTextStyle
    wrapTextAt=400
    defaultTextStyleOverride {
     overflowPolicy=Clip
     transformPolicy=None
     colorAndOpacity {colorUseRule=UseColor_Specified, specifiedColor=<rgba 1 1 1 0.7>}
     font {
      bForceMonospaced=false
      bMaterialIsStencil=false
      fontMaterial=null
      fontObject=&HeadingNow
      letterSpacing=0
      monospacedWidth=1
      size=15.75
      skewAmount=0
      typefaceFontName=74 Regular
      outlineSettings {
       bApplyOutlineToDropShadows=true
       bMiteredCorners=false
       bSeparateFillAlpha=false
       outlineMaterial=null
       outlineSize=0
       outlineColor=<rgba 0 0 0 1>
      }
     }
     highlightColor {colorUseRule=UseColor_Specified, specifiedColor=<rgba 0 0 0 0>}
     highlightShape {
      bIsDynamicallyLoaded=false
      drawAs=Image
      imageType=NoImage
      mirroring=NoMirror
      resourceName=None
      resourceObject=null
      tiling=NoTile
      imageSize=<32 32>
      margin=<ltrb 0 0 0 0>
      outlineSettings {
       bUseBrushTransparency=false
       roundingType=HalfHeightRadius
       width=0
       color {colorUseRule=UseColor_Specified, specifiedColor=<rgba 0 0 0 0>}
       cornerRadii=<0 0 0 0>
      }
      tintColor {colorUseRule=UseColor_Specified, specifiedColor=<rgba 1 1 1 1>}
      uVRegion {bIsValid=false, max=<0 0>, min=<0 0>}
     }
     selectedBackgroundColor {colorUseRule=UseColor_Foreground, specifiedColor=<rgba 1 0 1 1>}
     shadowColorAndOpacity=<rgba 0 0 0 1>
     shadowOffset=<0 0>
     strikeBrush {
      bIsDynamicallyLoaded=false
      drawAs=Image
      imageType=NoImage
      mirroring=NoMirror
      resourceName=None
      resourceObject=null
      tiling=NoTile
      imageSize=<32 32>
      margin=<ltrb 0 0 0 0>
      outlineSettings {
       bUseBrushTransparency=false
       roundingType=HalfHeightRadius
       width=0
       color {colorUseRule=UseColor_Specified, specifiedColor=<rgba 0 0 0 0>}
       cornerRadii=<0 0 0 0>
      }
      tintColor {colorUseRule=UseColor_Specified, specifiedColor=<rgba 1 1 1 1>}
      uVRegion {bIsValid=false, max=<0 0>, min=<0 0>}
     }
     underlineBrush {
      bIsDynamicallyLoaded=false
      drawAs=Image
      imageType=NoImage
      mirroring=NoMirror
      resourceName=None
      resourceObject=null
      tiling=NoTile
      imageSize=<32 32>
      margin=<ltrb 0 0 0 0>
      outlineSettings {
       bUseBrushTransparency=false
       roundingType=HalfHeightRadius
       width=0
       color {colorUseRule=UseColor_Specified, specifiedColor=<rgba 0 0 0 0>}
       cornerRadii=<0 0 0 0>
      }
      tintColor {colorUseRule=UseColor_Specified, specifiedColor=<rgba 1 1 1 1>}
      uVRegion {bIsValid=false, max=<0 0>, min=<0 0>}
     }
    }
   }
   slot {class=CommonVisibilitySwitcherSlot}
  }
  ChaseReward_Info {
   widgetClassPath=&CommonRichTextBlock
   bIsVariable=true
   parent=InfoText_Switcher
   widget=ChaseReward_Info
   properties {
    bOverrideDefaultStyle=true
    justification=Right
    lineHeightPercentage=1.15
    mobileTextBlockScale=1.2
    textStyleSet=&ItemDescriptionTextStyle
    wrapTextAt=525
    defaultTextStyleOverride {
     overflowPolicy=Clip
     transformPolicy=None
     colorAndOpacity {colorUseRule=UseColor_Specified, specifiedColor=<rgba 1 1 1 0.7>}
     font {
      bForceMonospaced=false
      bMaterialIsStencil=false
      fontMaterial=null
      fontObject=&HeadingNow
      letterSpacing=0
      monospacedWidth=1
      size=15.75
      skewAmount=0
      typefaceFontName=74 Regular
      outlineSettings {
       bApplyOutlineToDropShadows=true
       bMiteredCorners=false
       bSeparateFillAlpha=false
       outlineMaterial=null
       outlineSize=0
       outlineColor=<rgba 0 0 0 1>
      }
     }
     highlightColor {colorUseRule=UseColor_Specified, specifiedColor=<rgba 0 0 0 0>}
     highlightShape {
      bIsDynamicallyLoaded=false
      drawAs=Image
      imageType=NoImage
      mirroring=NoMirror
      resourceName=None
      resourceObject=null
      tiling=NoTile
      imageSize=<32 32>
      margin=<ltrb 0 0 0 0>
      outlineSettings {
       bUseBrushTransparency=false
       roundingType=HalfHeightRadius
       width=0
       color {colorUseRule=UseColor_Specified, specifiedColor=<rgba 0 0 0 0>}
       cornerRadii=<0 0 0 0>
      }
      tintColor {colorUseRule=UseColor_Specified, specifiedColor=<rgba 1 1 1 1>}
      uVRegion {bIsValid=false, max=<0 0>, min=<0 0>}
     }
     selectedBackgroundColor {colorUseRule=UseColor_Foreground, specifiedColor=<rgba 1 0 1 1>}
     shadowColorAndOpacity=<rgba 0 0 0 1>
     shadowOffset=<0 0>
     strikeBrush {
      bIsDynamicallyLoaded=false
      drawAs=Image
      imageType=NoImage
      mirroring=NoMirror
      resourceName=None
      resourceObject=null
      tiling=NoTile
      imageSize=<32 32>
      margin=<ltrb 0 0 0 0>
      outlineSettings {
       bUseBrushTransparency=false
       roundingType=HalfHeightRadius
       width=0
       color {colorUseRule=UseColor_Specified, specifiedColor=<rgba 0 0 0 0>}
       cornerRadii=<0 0 0 0>
      }
      tintColor {colorUseRule=UseColor_Specified, specifiedColor=<rgba 1 1 1 1>}
      uVRegion {bIsValid=false, max=<0 0>, min=<0 0>}
     }
     underlineBrush {
      bIsDynamicallyLoaded=false
      drawAs=Image
      imageType=NoImage
      mirroring=NoMirror
      resourceName=None
      resourceObject=null
      tiling=NoTile
      imageSize=<32 32>
      margin=<ltrb 0 0 0 0>
      outlineSettings {
       bUseBrushTransparency=false
       roundingType=HalfHeightRadius
       width=0
       color {colorUseRule=UseColor_Specified, specifiedColor=<rgba 0 0 0 0>}
       cornerRadii=<0 0 0 0>
      }
      tintColor {colorUseRule=UseColor_Specified, specifiedColor=<rgba 1 1 1 1>}
      uVRegion {bIsValid=false, max=<0 0>, min=<0 0>}
     }
    }
   }
   slot {class=CommonVisibilitySwitcherSlot}
  }
  NavButtons_Overlay {
   widgetClassPath=&Overlay
   parent=GridPanel_0
   widget=NavButtons_Overlay
   slot {class=GridSlot, properties {columnSpan=3, rowSpan=3}}
  }
  Button_Previous {
   widgetClassPath=&WBP_UIKit_Button_Generic
   bIsVariable=true
   parent=NavButtons_Overlay
   widget=Button_Previous
   properties {
    bHideInputActionWithKeyboard=true
    bIsFocusable=false
    backgroundMaterial=/FortUILibrary/UIKit/Buttons/MaterialInstances/CTAButtons/MI_UIKit_BgRegular
    enableFocusedSelectedState=false
    icon=/FortUILibrary/Textures/Icons/ToBeDeprecated/T_UI_Icon_Global_Chevron_Left
    iconMaterial=/FortUILibrary/Textures/Material/MI_UIKit_Icon_Base
    minHeight=64
    minWidth=64
    outlineMaterial=/FortUILibrary/UIKit/Buttons/MaterialInstances/CTAButtons/MI_UIKit_ButtonOutline
    priority=100
    promptHorizontalAlignment=HAlign_Center
    promptVerticalAlignment=VAlign_Bottom
    useBackgroundBlock=false
    useTextBlock=false
    buttonPadding=<ltrb 0 0 0 0>
    iconSize=<64 64>
    promptNudge=<0 64>
    secondTextPadding=<ltrb 0 0 0 0>
    statusIndicatorPadding=<ltrb 0 0 0 0>
    triggeringInputAction {dataTable=&InputActionDataTable, rowName=GenericLeftShoulder}
   }
   slot {class=OverlaySlot, properties {verticalAlignment=VAlign_Center}}
  }
  Button_Previous_IconSlot {widgetClassPath=&NamedSlot, bIsVariable=true, namedSlotHost=Button_Previous, widget=Button_Previous_IconSlot}
  Button_Next {
   widgetClassPath=&WBP_UIKit_Button_Generic
   bIsVariable=true
   parent=NavButtons_Overlay
   widget=Button_Next
   properties {
    bHideInputActionWithKeyboard=true
    bIsFocusable=false
    bShouldUseFallbackDefaultInputAction=false
    backgroundMaterial=&MI_UIKit_BgRegular
    enableFocusedSelectedState=false
    icon=/FortUILibrary/Textures/Icons/T_UI_Icon_UIKit_Chevron
    iconMaterial=&MI_UIKit_Icon_Base
    minHeight=64
    minWidth=64
    outlineMaterial=&MI_UIKit_ButtonOutline
    priority=100
    promptHorizontalAlignment=HAlign_Center
    promptVerticalAlignment=VAlign_Bottom
    useBackgroundBlock=false
    useTextBlock=false
    buttonPadding=<ltrb 0 0 0 0>
    iconSize=<64 64>
    promptNudge=<0 64>
    triggeringInputAction {dataTable=&InputActionDataTable, rowName=GenericRightShoulder}
   }
   slot {class=OverlaySlot, properties {horizontalAlignment=HAlign_Right, verticalAlignment=VAlign_Center}}
  }
  Button_Next_IconSlot {widgetClassPath=&NamedSlot, bIsVariable=true, namedSlotHost=Button_Next, widget=Button_Next_IconSlot}
  ItemDescription {
   widgetClassPath=/FortUILibrary/UIKit/Items/WBP_UIKit_ItemDescription_Compact
   bIsVariable=true
   parent=GridPanel_0
   widget=ItemDescription
   properties {titleMaxHeight=250, input Action Row {dataTable=&InputActionDataTable, rowName=GenericFaceButtonTop_Hold}}
   slot {class=GridSlot, properties {padding=<ltrb 64 64 0 0>}}
  }
  Info_VBox {
   widgetClassPath=&VerticalBox
   bIsVariable=true
   parent=GridPanel_0
   widget=Info_VBox
   slot {class=GridSlot, properties {columnSpan=3, row=2, verticalAlignment=VAlign_Bottom, padding=<ltrb 0 16 0 32>}}
  }
  InteractionSwitcher {
   widgetClassPath=&CommonVisibilitySwitcher
   bIsVariable=true
   parent=Info_VBox
   widget=InteractionSwitcher
   slot {class=VerticalBoxSlot}
  }
  BuyReward_VBox {
   widgetClassPath=&VerticalBox
   bIsVariable=true
   parent=InteractionSwitcher
   widget=BuyReward_VBox
   slot {class=CommonVisibilitySwitcherSlot}
  }
  ItemCostInfo_Text {
   widgetClassPath=&CommonRichTextBlock
   bIsVariable=true
   parent=BuyReward_VBox
   widget=ItemCostInfo_Text
   properties {
    autoWrapText=true
    bOverrideDefaultStyle=true
    justification=Center
    lineHeightPercentage=1.15
    mobileTextBlockScale=1.2
    textStyleSet=&ItemDescriptionTextStyle
    wrapTextAt=700
    defaultTextStyleOverride {
     overflowPolicy=Clip
     transformPolicy=None
     colorAndOpacity {colorUseRule=UseColor_Specified, specifiedColor=<rgba 1 1 1 1>}
     font {
      bForceMonospaced=false
      bMaterialIsStencil=false
      fontMaterial=null
      fontObject=&HeadingNow
      letterSpacing=0
      monospacedWidth=1
      size=18
      skewAmount=0
      typefaceFontName=74 Regular
      outlineSettings {
       bApplyOutlineToDropShadows=true
       bMiteredCorners=false
       bSeparateFillAlpha=false
       outlineMaterial=null
       outlineSize=0
       outlineColor=<rgba 0 0 0 1>
      }
     }
     highlightColor {colorUseRule=UseColor_Specified, specifiedColor=<rgba 0 0 0 0>}
     highlightShape {
      bIsDynamicallyLoaded=false
      drawAs=Image
      imageType=NoImage
      mirroring=NoMirror
      resourceName=None
      resourceObject=null
      tiling=NoTile
      imageSize=<32 32>
      margin=<ltrb 0 0 0 0>
      outlineSettings {
       bUseBrushTransparency=false
       roundingType=HalfHeightRadius
       width=0
       color {colorUseRule=UseColor_Specified, specifiedColor=<rgba 0 0 0 0>}
       cornerRadii=<0 0 0 0>
      }
      tintColor {colorUseRule=UseColor_Specified, specifiedColor=<rgba 1 1 1 1>}
      uVRegion {bIsValid=false, max=<0 0>, min=<0 0>}
     }
     selectedBackgroundColor {colorUseRule=UseColor_Foreground, specifiedColor=<rgba 1 0 1 1>}
     shadowColorAndOpacity=<rgba 0 0 0 1>
     shadowOffset=<0 0>
     strikeBrush {
      bIsDynamicallyLoaded=false
      drawAs=Image
      imageType=NoImage
      mirroring=NoMirror
      resourceName=None
      resourceObject=null
      tiling=NoTile
      imageSize=<32 32>
      margin=<ltrb 0 0 0 0>
      outlineSettings {
       bUseBrushTransparency=false
       roundingType=HalfHeightRadius
       width=0
       color {colorUseRule=UseColor_Specified, specifiedColor=<rgba 0 0 0 0>}
       cornerRadii=<0 0 0 0>
      }
      tintColor {colorUseRule=UseColor_Specified, specifiedColor=<rgba 1 1 1 1>}
      uVRegion {bIsValid=false, max=<0 0>, min=<0 0>}
     }
     underlineBrush {
      bIsDynamicallyLoaded=false
      drawAs=Image
      imageType=NoImage
      mirroring=NoMirror
      resourceName=None
      resourceObject=null
      tiling=NoTile
      imageSize=<32 32>
      margin=<ltrb 0 0 0 0>
      outlineSettings {
       bUseBrushTransparency=false
       roundingType=HalfHeightRadius
       width=0
       color {colorUseRule=UseColor_Specified, specifiedColor=<rgba 0 0 0 0>}
       cornerRadii=<0 0 0 0>
      }
      tintColor {colorUseRule=UseColor_Specified, specifiedColor=<rgba 1 1 1 1>}
      uVRegion {bIsValid=false, max=<0 0>, min=<0 0>}
     }
    }
   }
   slot {class=VerticalBoxSlot, properties {verticalAlignment=VAlign_Top, padding=<ltrb 0 0 0 10>}}
  }
  Buy_Button {
   widgetClassPath=/FortUILibrary/UIKit/Buttons/WBP_UIKit_Button_Loud
   bIsVariable=true
   parent=BuyReward_VBox
   widget=Buy_Button
   properties {text=Buy Rewards, triggeringInputAction {dataTable=&InputActionDataTable, rowName=GenericFaceButtonLeft}}
   slot {class=VerticalBoxSlot, properties {horizontalAlignment=HAlign_Center, padding=<ltrb 0 0 0 8>}}
  }
  ChaseProgress_Slot {
   widgetClassPath=&NamedSlot
   bIsVariable=true
   parent=InteractionSwitcher
   widget=ChaseProgress_Slot
   slot {class=CommonVisibilitySwitcherSlot, properties {horizontalAlignment=HAlign_Center}}
  }
  Claimed_Text {
   widgetClassPath=&CommonRichTextBlock
   bIsVariable=true
   parent=Info_VBox
   widget=Claimed_Text
   properties {
    autoWrapText=true
    bOverrideDefaultStyle=true
    justification=Center
    lineHeightPercentage=1.15
    mobileTextBlockScale=1.2
    text="Reward Claimed!"
    textStyleSet=&ItemDescriptionTextStyle
    wrapTextAt=700
    defaultTextStyleOverride {
     overflowPolicy=Clip
     transformPolicy=None
     colorAndOpacity {colorUseRule=UseColor_Specified, specifiedColor=<rgba 0.0437 1 0.2307 1>}
     font {
      bForceMonospaced=false
      bMaterialIsStencil=false
      fontMaterial=null
      fontObject=&HeadingNow
      letterSpacing=0
      monospacedWidth=1
      size=19.5
      skewAmount=0
      typefaceFontName=75 Medium
      outlineSettings {
       bApplyOutlineToDropShadows=true
       bMiteredCorners=false
       bSeparateFillAlpha=false
       outlineMaterial=null
       outlineSize=0
       outlineColor=<rgba 0 0 0 1>
      }
     }
     highlightColor {colorUseRule=UseColor_Specified, specifiedColor=<rgba 0 0 0 0>}
     highlightShape {
      bIsDynamicallyLoaded=false
      drawAs=Image
      imageType=NoImage
      mirroring=NoMirror
      resourceName=None
      resourceObject=null
      tiling=NoTile
      imageSize=<32 32>
      margin=<ltrb 0 0 0 0>
      outlineSettings {
       bUseBrushTransparency=false
       roundingType=HalfHeightRadius
       width=0
       color {colorUseRule=UseColor_Specified, specifiedColor=<rgba 0 0 0 0>}
       cornerRadii=<0 0 0 0>
      }
      tintColor {colorUseRule=UseColor_Specified, specifiedColor=<rgba 1 1 1 1>}
      uVRegion {bIsValid=false, max=<0 0>, min=<0 0>}
     }
     selectedBackgroundColor {colorUseRule=UseColor_Foreground, specifiedColor=<rgba 1 0 1 1>}
     shadowColorAndOpacity=<rgba 0 0 0 1>
     shadowOffset=<0 0>
     strikeBrush {
      bIsDynamicallyLoaded=false
      drawAs=Image
      imageType=NoImage
      mirroring=NoMirror
      resourceName=None
      resourceObject=null
      tiling=NoTile
      imageSize=<32 32>
      margin=<ltrb 0 0 0 0>
      outlineSettings {
       bUseBrushTransparency=false
       roundingType=HalfHeightRadius
       width=0
       color {colorUseRule=UseColor_Specified, specifiedColor=<rgba 0 0 0 0>}
       cornerRadii=<0 0 0 0>
      }
      tintColor {colorUseRule=UseColor_Specified, specifiedColor=<rgba 1 1 1 1>}
      uVRegion {bIsValid=false, max=<0 0>, min=<0 0>}
     }
     underlineBrush {
      bIsDynamicallyLoaded=false
      drawAs=Image
      imageType=NoImage
      mirroring=NoMirror
      resourceName=None
      resourceObject=null
      tiling=NoTile
      imageSize=<32 32>
      margin=<ltrb 0 0 0 0>
      outlineSettings {
       bUseBrushTransparency=false
       roundingType=HalfHeightRadius
       width=0
       color {colorUseRule=UseColor_Specified, specifiedColor=<rgba 0 0 0 0>}
       cornerRadii=<0 0 0 0>
      }
      tintColor {colorUseRule=UseColor_Specified, specifiedColor=<rgba 1 1 1 1>}
      uVRegion {bIsValid=false, max=<0 0>, min=<0 0>}
     }
    }
   }
   slot {class=VerticalBoxSlot, properties {padding=<ltrb 0 8 0 0>}}
  }
  WBP_ActionBarContainer {
   widgetClassPath=/Game/UI/Frontend/BottomBar/WBP_ActionBarContainer
   bIsVariable=true
   parent=GridPanel_0
   widget=WBP_ActionBarContainer
   slot {class=GridSlot, properties {columnSpan=3, horizontalAlignment=HAlign_Right, row=3}}
  }
  Button_Bar {widgetClassPath=&HorizontalBox, namedSlotHost=WBP_ActionBarContainer, widget=Button_Bar}
  ActionButtonsWrapper {
   widgetClassPath=&HorizontalBox
   bIsVariable=true
   parent=Button_Bar
   widget=ActionButtonsWrapper
   slot {class=HorizontalBoxSlot, properties {verticalAlignment=VAlign_Center}}
  }
  Button_PreviewAction {
   widgetClassPath=/FortUILibrary/UIKit/Buttons/WBP_UIKit_Button_Quiet
   parent=ActionButtonsWrapper
   widget=Button_PreviewAction
   properties {
    bAutoUpdateFromInputAction=true
    bDisplayInputActionWhenNotInteractable=false
    bIsFocusable=false
    bShouldUseFallbackDefaultInputAction=false
    minWidth=48
    text=""
   }
   slot {class=HorizontalBoxSlot, properties {padding=<ltrb 0 0 24 0>}}
  }
  Button_PreviewAction1 {
   widgetClassPath=&WBP_UIKit_Button_Quiet
   parent=ActionButtonsWrapper
   widget=Button_PreviewAction1
   properties {
    bAutoUpdateFromInputAction=true
    bDisplayInputActionWhenNotInteractable=false
    bIsFocusable=false
    bShouldUseFallbackDefaultInputAction=false
    minWidth=48
    text=""
   }
   slot {class=HorizontalBoxSlot, properties {padding=<ltrb 0 0 24 0>}}
  }
  Button_PreviewAction2 {
   widgetClassPath=&WBP_UIKit_Button_Quiet
   parent=ActionButtonsWrapper
   widget=Button_PreviewAction2
   properties {
    bAutoUpdateFromInputAction=true
    bDisplayInputActionWhenNotInteractable=false
    bIsFocusable=false
    bShouldUseFallbackDefaultInputAction=false
    minWidth=48
    text=""
   }
   slot {class=HorizontalBoxSlot, properties {padding=<ltrb 0 0 24 0>}}
  }
  Button_Back {
   widgetClassPath=&WBP_UIKit_Button_Quiet
   bIsVariable=true
   parent=Button_Bar
   widget=Button_Back
   properties {
    bIsFocusable=false
    text=Back
    visibility=Visible
    triggeringInputAction {dataTable=&InputActionDataTable, rowName=GenericBack}
   }
   slot {class=HorizontalBoxSlot, properties {verticalAlignment=VAlign_Center}}
  }
 ]
}
