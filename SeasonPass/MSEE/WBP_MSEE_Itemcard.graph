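# MSEE Season Pass item card, extending the UIKit item card base with cost text, a buy outline and an owned indicator.
# PreConstruct sets design-time preview values, then InitItemcard async-loads the background material, falling back to a default, and caches its dynamic instance.
# That instance's scalar parameters drive the progress overlay, which OverriddenPercent can override, and the hover inner glow.
# Hover and click broadcast the reward and item view models through dispatchers.
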WidgetBlueprint /SeasonPass/MSEE/WBP_MSEE_Itemcard
animations [Focus {
  binding_count=1
  length_seconds=0.05
  object=&self:Focus
  sequence {bindings [SizeBox {
     id=#1
     possessed_class=SizeBox
     tracks [{
       type=MovieSceneFloatTrack
       display_name=Height Override
       sections [{
         type=MovieSceneFloatSection
         end_frame=1
         start_frame=0
         channels [None {type=MovieSceneScriptingFloatChannel, keys [[0,160],[0.9997,200],[1,180]]}]
        }]
      }]
    }]}
 }]
graphs [
 DesignTimeSetup {nodes [
   n1 = DesignTimeSetup!() -> (then, Condition:bool=~) {is=title K2Node_FunctionEntry}
   > Branch(Condition=@n1)
   > SetCost!(execute, self, Cost:int=42) -> (then)
   > n2 = GetDynamicMaterial(self=Get Background() -> (Background:Image Object Reference))
   > SetScalarParameterValue(self=@n2, ParameterName=OverlayProgress, Value=Get ProgressPercentPreview() -> (ProgressPercentPreview:double))
  ]}
 EventGraph {nodes [
   n1 = AsyncLoadBackground!() -> (OutputDelegate, then, MaterialInterfaceValue:Material Interface Soft Object Reference) {is=title K2Node_CustomEvent}
   > n2 = Async Load Asset(Asset=Select(
     Option 0:Material Interface Soft Object Reference=/FortUILibrary/Systems/Items/MaterialInstances/MI_UIKit_Item_Background
     Option 1:Material Interface Soft Object Reference=@n1.MaterialInterfaceValue
     Index=IsValidSoftObjectReference(SoftObjectReference=@n1.MaterialInterfaceValue)
    ) -> (ReturnValue:Material Interface Soft Object Reference))
   n3 = Cast To MaterialInterface!(execute=@n2.Completed, Object=@n2) -> (then, CastFailed, AsMaterial Interface:Material Interface Object Reference) {is=title K2Node_DynamicCast}
   > SetBrushFromMaterial(self=Get Background(), Material=@n3)
   > n4 = GetDynamicMaterial(self=Get Background())
   > Set with Broadcast DynamicBackgroundMaterial!(execute, DynamicBackgroundMaterial:Material Instance Dynamic Object Reference=@n4) -> (then, Output_Get:Material Instance Dynamic Object Reference)
   > SetProgressAmount!(
    execute
    self
    Progress:double=SelectFloat(A=Get ProgressPercentPreview(), B=Get ProgressPercent() -> (ProgressPercent:double), bPickA=Get IsDesignTime() -> (IsDesignTime:bool))
   ) -> (then)
   Event On Clicked()
   > Parent: On Clicked() {is=title K2Node_CallParentFunction}
   > Call Item Clicked!(
    execute:Exec
    self:WBP MSEE Itemcard Object Reference
    RewardVM:Season Pass Reward VM Object Reference=Get SeasonPassRewardVM() -> (SeasonPassRewardVM:Season Pass Reward VM Object Reference)
    FortItemVM:Fort Item VM Object Reference=Get CurrentItemVM() -> (CurrentItemVM:Fort Item VM Object Reference)
   ) -> (then:Exec) {is=title K2Node_CallDelegate}
   n5 = Event PreConstruct()
   > Parent: PreConstruct(IsDesignTime=@n5.IsDesignTime)
   > Set IsDesignTime(IsDesignTime:bool=@n5.IsDesignTime) -> (Output_Get:bool)
   > DesignTimeSetup!(execute, self, Condition:bool=@n5.IsDesignTime) -> (then) {is=title K2Node_CallFunction}
   > InitItemcard()
   Event On Unhovered()
   > Parent: On Unhovered() {is=title K2Node_CallParentFunction}
   > n6 = Get!(execute:Exec) -> (then:Exec, else:Exec, DynamicBackgroundMaterial:Material Instance Dynamic Object Reference)
   > SetScalarParameterValue(self=@n6, ParameterName=InnerGlowOpacity, Value=0)
   Event Destruct()
   > Parent: Destruct()
   Event On Selected()
   > Parent: On Selected() {is=title K2Node_CallParentFunction}
   Event OnInitialized()
   > Parent: OnInitialized() {is=title K2Node_CallParentFunction}
   Event On Hovered()
   > Parent: On Hovered() {is=title K2Node_CallParentFunction}
   > Call Item Hovered!(
    execute:Exec
    self:WBP MSEE Itemcard Object Reference
    RewardVM:Season Pass Reward VM Object Reference=Get SeasonPassRewardVM()
    FortItemVM:Fort Item VM Object Reference=Get CurrentItemVM()
   ) -> (then:Exec) {is=title K2Node_CallDelegate}
   > n7 = Get()
   > SetScalarParameterValue(self=@n7, ParameterName=InnerGlowOpacity, Value=1)
   HandleInputMethodChanged!() -> (OutputDelegate, then, NewInputMethod:ECommonInputType Enum) {is=title K2Node_CustomEvent}
   Event On Focused() {is=title K2Node_Event}
   Event Construct()
  ]}
 Get Item VM {nodes [
   n1 = Get Item VM!() -> (then, Reward:Season Pass Reward VM Object Reference) {is=title K2Node_FunctionEntry}
   > Is Valid(InputObject=@n1)
   > Return Node!(
    execute
    Result:Fort Item VM Object Reference="Get (a copy)"(Array:Array of Fort Item VM Object References=Get ItemVMs!(self:Season Pass Reward VM Object Reference=@n1) -> (ItemVMs:Array of Fort Item VM Object References)) -> (Output:Fort Item VM Object Reference)
   )
  ]}
 GetCost {nodes [
   n1 = GetCost!() -> (then, IsOwned:bool=~, LevelCost:int) {is=title K2Node_FunctionEntry}
   > Return Node!(execute, Cost:int=SelectInt(A=0, B=@n1.LevelCost, bPickA=@n1))
  ]}
 GetProgressAmount {nodes [
   n1 = GetProgressAmount!() -> (then, DynamicBackground:Material Instance Dynamic Object Reference, Percent:double=~) {is=title K2Node_FunctionEntry}
   > Return Node!(execute, ProgressPercent:double=@n1.Percent)
  ]}
 InitItemcard {nodes [
   InitItemcard() {is=title K2Node_FunctionEntry}
   > AsyncLoadBackground!(
    execute
    self
    MaterialInterfaceValue:Material Interface Soft Object Reference=Get BackgroundMaterial() -> (BackgroundMaterial:Material Interface Soft Object Reference)
   ) -> (then) {is=title K2Node_CallFunction}
   > SetItemIsEquipped!(execute, self:WBP UIKit Item Card Base Object Reference, Is Equipped:bool=false) -> (then)
   > SetCornerRadius!(
    execute
    self:WBP UIKit Item Card Base Object Reference
    UseOverrideValues:bool=false
    RadiusInPixels:bool=true
    CornerRadius_TopLeft:double=20
    CornerRadius_TopRight:double=20
    CornerRadius_BottomLeft:double=20
    CornerRadius_BottomRight:double=20
    OutlineCornerOffset:double=20
   ) -> (then)
   > ShowBuyOutline!(execute, self, ShowOutline:bool=false) -> (then)
  ]}
 ItemClicked {nodes [ItemClicked!() -> (then, RewardVM:Season Pass Reward VM Object Reference, FortItemVM:Fort Item VM Object Reference) {is=title K2Node_FunctionEntry}]}
 ItemHovered {nodes [ItemHovered!() -> (then, RewardVM:Season Pass Reward VM Object Reference, FortItemVM:Fort Item VM Object Reference) {is=title K2Node_FunctionEntry}]}
 SetCost {nodes [
   n1 = SetCost!() -> (then, Cost:int) {is=title K2Node_FunctionEntry}
   > "SetText (Text)"(
    self=Get Cost_Text() -> (Cost_Text:Common Text Object Reference)
    InText=SelectText(B="To Text (Integer)"(Value=@n1), bPickA="integer <= integer"(A=@n1))
   )
  ]}
 SetItemImage {nodes [
   n1 = SetItemImage!() -> (then, Item Image:Soft Object Reference) {is=title K2Node_FunctionEntry}
   > Parent: SetItemImage!(execute, Item Image:Soft Object Reference=@n1) -> (then) {is=title K2Node_CallParentFunction}
   n2 = Get LazyImage_Item() -> (LazyImage_Item:Fort Lazy Image Object Reference)
   n3 = SlotAsGridSlot(Widget=@n2)
   n4 = OR Boolean(A=NOT Boolean(A=IsAssetNull!(Asset:Soft Object Reference=@n1) -> (ReturnValue:bool)), B=InEditor!() -> (ReturnValue:bool))
   > SetHorizontalAlignment(
    self:Grid Slot Object Reference=@n3
    InHorizontalAlignment=Select(Option 0:EHorizontalAlignment Enum=HAlign_Center, Option 1:EHorizontalAlignment Enum=HAlign_Fill, Index=@n4) -> (ReturnValue:EHorizontalAlignment Enum)
   )
   > SetVerticalAlignment(
    self:Grid Slot Object Reference=@n3
    InVerticalAlignment=Select(Option 0:EVerticalAlignment Enum=VAlign_Center, Option 1:EVerticalAlignment Enum=VAlign_Fill, Index=@n4) -> (ReturnValue:EVerticalAlignment Enum)
   )
   > n5 = GetDynamicMaterial(self=@n2)
   > SetScalarParameterValue(self=@n5, ParameterName=UseVignette, Value=0)
  ]}
 SetProgressAmount {nodes [
   n1 = SetProgressAmount!() -> (then, Progress:double=~) {is=title K2Node_FunctionEntry}
   > n2 = Branch(Condition="float > float"(A=Get OverriddenPercent() -> (OverriddenPercent:double), B=-1))
   n3 = Get DynamicBackgroundMaterial() -> (DynamicBackgroundMaterial:Material Instance Dynamic Object Reference)
   > SetScalarParameterValue(self=@n3, ParameterName=OverlayProgress, Value=Get OverriddenPercent())
   SetScalarParameterValue(execute=@n2.else, self=@n3, ParameterName=OverlayProgress, Value=@n1)
  ]}
 ShowBuyOutline {nodes [
   n1 = ShowBuyOutline!() -> (then, ShowOutline:bool=~) {is=title K2Node_FunctionEntry}
   n2 = Get BuyOutline() -> (BuyOutline:Common Border Object Reference)
   > SetVisibility(
    self=@n2
    InVisibility=Select(Option 0:ESlateVisibility Enum=Collapsed, Option 1:ESlateVisibility Enum=SelfHitTestInvisible, Index=@n1) -> (ReturnValue:ESlateVisibility Enum)
   )
   > SetBrushFromMaterial(self:Border Object Reference=@n2, Material=Get BuySelectionOutline() -> (BuySelectionOutline:Material Interface Object Reference))
  ]}
]
variables [
 BackgroundMaterial {type=MaterialInterface, default=/FortUILibrary/Systems/Items/Materials/M_UIKit_Item_Background}
 BuySelectionOutline {type=MaterialInterface, default=/FortUILibrary/Systems/Items/MaterialInstances/MI_UIKit_EquippedStateOutline}
 DynamicBackgroundMaterial {type=MaterialInstanceDynamic}
 HideEnhanced {type=boolean, default=true}
 IsDesignTime {type=boolean}
 ItemClicked {}
 ItemHovered {}
 OverriddenPercent {type=number, default=-1}
 ProgressPercent {type=number}
 ProgressPercentPreview {type=number, default=0.05}
 TestIndex {type=integer}
 TestMidIndex {type=integer}
]
widgets {
 info {
  inheritedWidgetCount=0
  namedSlotCount=1
  parentClass=/FortUILibrary/Systems/Items/WBP_UIKit_ItemCard_Base.WBP_UIKit_ItemCard_Base_C
  rootWidgetClass=null
  widgetCount=5
 }
 widgets [
  Status {widgetClassPath=&Overlay, widget=Status}
  StatusIndicator_Owned {
   widgetClassPath=/FortUILibrary/UIKit/StatusIndicator/WBP_UIKit_StatusIndicator
   bIsVariable=true
   parent=Status
   widget=StatusIndicator_Owned
   properties {icon=/FortUILibrary/Textures/MaterialInstances/MI_UIKit_Icon_Check, iconSize=Compact, visibility=Collapsed}
   slot {class=OverlaySlot, properties {horizontalAlignment=HAlign_Right, padding=<ltrb 8 8 8 8>}}
  }
  CommonVisualAttachment_71 {
   widgetClassPath=&CommonVisualAttachment
   parent=Status
   widget=CommonVisualAttachment_71
   properties {contentAnchor=<0.5 0>}
   slot {class=OverlaySlot, properties {horizontalAlignment=HAlign_Center, verticalAlignment=VAlign_Bottom}}
  }
  Cost_Text {
   widgetClassPath=&CommonTextBlock
   bIsVariable=true
   parent=CommonVisualAttachment_71
   widget=Cost_Text
   properties {
    justification=Center
    mobileFontSizeMultiplier=1.4
    scrollStyle=&TextScrollStyle-Base_C
    textTransformPolicy=ToUpper
    font {
     bForceMonospaced=false
     bMaterialIsStencil=false
     fontMaterial=null
     fontObject=&HeadingNow
     letterSpacing=0
     monospacedWidth=1
     size=15.75
     skewAmount=0
     typefaceFontName=75 Medium
     outlineSettings {
      bApplyOutlineToDropShadows=false
      bMiteredCorners=false
      bSeparateFillAlpha=false
      outlineMaterial=null
      outlineSize=0
      outlineColor=<rgba 0 0 0 1>
     }
    }
    shadowOffset=<0 0>
   }
   slot {class=SizeBoxSlot, properties {horizontalAlignment=HAlign_Center, padding=<ltrb 0 12 0 0>}}
  }
  BuyOutline {
   widgetClassPath=&CommonBorder
   bIsVariable=true
   parent=Status
   widget=BuyOutline
   properties {
    horizontalAlignment=HAlign_Left
    verticalAlignment=VAlign_Bottom
    background {
     bIsDynamicallyLoaded=false
     drawAs=Image
     imageType=NoImage
     mirroring=NoMirror
     resourceName=None
     resourceObject=&MI_UIKit_EquippedStateOutline
     tiling=NoTile
     imageSize=<32 32>
     margin=<ltrb 0 0 0 0>
     outlineSettings {
      bUseBrushTransparency=false
      roundingType=HalfHeightRadius
      width=0
      color {colorUseRule=UseColor_Specified, specifiedColor=<rgba 0 0 0 0>}
      cornerRadii=<0 0 0 0>
     }
     tintColor {colorUseRule=UseColor_Specified, specifiedColor=<rgba 1 1 1 1>}
     uVRegion {bIsValid=false, max=<0 0>, min=<0 0>}
    }
    contentColorAndOpacity=<rgba 1 1 1 0.995>
    padding=<ltrb 0 0 0 0>
   }
   slot {class=OverlaySlot, properties {horizontalAlignment=HAlign_Fill, verticalAlignment=VAlign_Fill}}
  }
 ]
}
